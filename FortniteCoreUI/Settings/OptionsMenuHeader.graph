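# OptionsMenuHeader is a list-entry widget for options menu headers: a text label over a thin divider and bottom spacer.
# On list item assignment it takes text and type from FortSettingInfo, disabling certain sensitivity types while controlling a remote-controlled pawn.
# Mouse leave and a gamepad-aware centering event swap text styles, and most event paths reapply a mobile-aware font size.
# Construct also sets spacer size by BR-only mode.
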WidgetBlueprint /FortniteCoreUI/Settings/OptionsMenuHeader
graphs [
 Center On Widget {nodes [Center On Widget() {is=title K2Node_FunctionEntry}]}
 EventGraph {nodes [
   Event Construct()
   > n1 = UpdateSize()
   n2 = Get TextHeader() -> (TextHeader:Common Text Object Reference)
   n3 = Get Font!(self:Text Object Reference=@n2) -> (Font:Slate Font Info Structure)
   SetFontSize!(
    execute=[@n6.then, @n5.then, @n9.then, @n10.then, @n1.then, @n11.then]
    self
    Text:Common Text Object Reference=@n2
    DefaultSize:int=Truncate(A=Break Slate Font Info!(SlateFontInfo:"Slate Font Info Structure (by ref)"=@n3) -> (Size:float=~, LetterSpacing:int, SkewAmount:float=~, bForceMonospaced:bool=~, bMaterialIsStencil:bool=~, MonospacedWidth:float=~) {is=title K2Node_BreakStruct})
    MobileSize:int=Get FontSizeMobile() -> (FontSizeMobile:int=0)
   ) -> (then)
   K2Node_Event_2 = CenterOnWidget_1() {is=title K2Node_CustomEvent}
   > n4 = Branch(Condition=IsUsingGamepad!(self:Fort Global UIContext Object Reference=Get FortGlobalUIContext!() -> (ReturnValue:Fort Global UIContext Object Reference) {is=title K2Node_GetBlueprintContextEx}) -> (ReturnValue:bool))
   > n5 = SetStyle!(
    execute
    self:Common Text Object Reference=Get TextHeader()
    InStyle:Common Text Style Class Reference=/FortUILibrary/Styles/Typography/Foundation/TextStyle-Base-S-B-S_Gorse.TextStyle-Base-S-B-S_Gorse_C
   ) -> (then)
   n6 = SetStyle(
    execute=@n4.else
    self=Get TextHeader()
    InStyle=/FortUILibrary/Styles/Typography/Foundation/TextStyle-Base-S-B-Blue-40.TextStyle-Base-S-B-Blue-40_C
   )
   n7 = Event OnListItemObjectSet()
   > n8 = Cast To FortSettingInfo!(execute, Object=@n7.ListItemObject) -> (then, CastFailed, AsFort Setting Info:Fort Setting Info Object Reference) {is=title K2Node_DynamicCast}
   > "SetText (Text)"(self=Get TextHeader(), InText=Get OptionDisplayText!(self:Fort Setting Info Object Reference=@n8) -> (OptionDisplayText:Text))
   > Set LocalSettingType(LocalSettingType:ESettingType Enum=Get SettingType!(self:Fort Setting Info Object Reference=@n8) -> (SettingType:ESettingType Enum=None)) -> (Output_Get:ESettingType Enum=None)
   > n9 = HandlePawnSet()
   Event OnMouseLeave!() -> (OutputDelegate, then, MouseEvent:"Pointer Event Structure (by ref)") {is=title K2Node_Event}
   > n10 = SetStyle(self=Get TextHeader(), InStyle=&TextStyle-Base-S-B-Blue-40_C)
   n11 = Event OnMouseEnter()
  ]}
 HandlePawnSet {nodes [
   HandlePawnSet() {is=title K2Node_FunctionEntry}
   n1 = GetOwningPlayer()
   > Set bRCPawn(bRCPawn:bool=Cast To FortRemoteControlledPawnAthena!(Object=Get Controlled Pawn!(self:Controller Object Reference=@n1) -> (ReturnValue:Pawn Object Reference) {is=title K2Node_CallFunction}) -> (AsFort Remote Controlled Pawn Athena:Fort Remote Controlled Pawn Athena Object Reference, bSuccess:bool=~) {is=title K2Node_DynamicCast}.bSuccess) -> (Output_Get:bool)
   > n2 = Switch on ESettingType!(execute, Selection:ESettingType Enum=Get LocalSettingType() -> (LocalSettingType:ESettingType Enum=None)) -> (
    None:Exec
    Header:Exec
    WindowMode:Exec
    DisplayResolution:Exec
    FrameRateLimit:Exec
    VideoQuality:Exec
    ThreeDResolution:Exec
    ViewDistance:Exec
    Shadows:Exec
    AntiAliasing:Exec
    Textures:Exec
    Effects:Exec
    PostProcessing:Exec
    VSync:Exec
    MotionBlur:Exec
    ShowGrass:Exec
    MobileFPSType:Exec
    ShowFPS:Exec
    AllowLowPower:Exec
    AllowVideoPlayback:Exec
    AllowDynamicResolution:Exec
    RenderingAPI:Exec
    UseGPUCrashDebugging:Exec
    RegionHeader:Exec
    Language:Exec
    Region:Exec
    MouseSensitivityYaw:Exec
    MouseSensitivityPitch:Exec
    MouseSensitivityMultiplierForAircraft:Exec
    TouchDragSensitivity:Exec
    TouchLookSmoothing:Exec
    ControllerLookSensitivityYaw:Exec
    ControllerLookSensitivityPitch:Exec
    MouseTargetingMultiplier:Exec
    MouseScopedMultiplier:Exec
    GamepadTargetingMultiplier:Exec
    GamepadScopedMultiplier:Exec
    GamepadBuildingMultiplier:Exec
    GamepadEditModeMultiplier:Exec
    TouchLookAccelerationMultiplier:Exec
    TouchDragTargetingSensitivity:Exec
    TouchDragScopedSensitivity:Exec
    TouchBuildingMultiplier:Exec
    TouchEditModeMultiplier:Exec
    TouchVerticalSensitivity:Exec
    InvertPitchMouse:Exec
    InvertPitchController:Exec
    InvertYaw:Exec
    InvertPitchForMotion:Exec
    InvertPitchForAircraftPrimary:Exec
    InvertPitchForAircraftSecondary:Exec
    InvertPitchForAircraftTouch:Exec
    InvertPitchForFightersMouse:Exec
    InvertPitchForFightersGamepad:Exec
    InvertPitchForFightersTouch:Exec
    InvertYawForMotion:Exec
    GyroEnabled:Exec
    GyroYawAxis:Exec
    GyroSensitivity:Exec
    GyroTargetingSensitivity:Exec
    GyroScopedSensitivity:Exec
    GyroHarvestingToolSensitivity:Exec
    SafeZone:Exec
    AnonymousMode:Exec
    AnonymousCharacterMode:Exec
    HideOtherPlayerNames:Exec
    HiddenMatchmakingDelay:Exec
    AutoJoinGameVoiceChannel:Exec
    ShowVoiceIndicators:Exec
    HUDScale:Exec
    ShowViewerCount:Exec
    FirstPersonCamera:Exec
    PeripheralLighting:Exec
    PingPlaceDangerMarkerWhenTargeting:Exec
    ShowGlobalChat:Exec
    ConsoleUnlockedFPS:Exec
    ToggleDoubleTapAutorun:Exec
    AutorunLockZonePreference:Exec
    ToggleSprint:Exec
    SprintByDefault:Exec
    SprintCancelsReload:Exec
    TapInteract:Exec
    InWorldInteract:Exec
    ToggleTargeting:Exec
    HoldToSwapPickup:Exec
    AutoEquipBetterItems:Exec
    EquipFirstBuildingPieceWhenSwappingQuickbars:Exec
    EquipFirstBuildingPieceWhenSwappingQuickbarsAthena:Exec
    DisablePreEditsWhenPlacingBuilding:Exec
    AimAssist:Exec
    EditModeAimAssist:Exec
    TouchEdit:Exec
    EditConfirmOnRelease:Exec
    QuickEdit:Exec
    TurboBuild:Exec
    CreativeTurboDelete:Exec
    AutoChangeMaterial:Exec
    GamepadAutoRun:Exec
    CrossplayPreference:Exec
    AutoOpenDoors:Exec
    AutoPickupWeapons:Exec
    AutoPickupWeaponsConsolePC:Exec
    AutoSortConsumablesToRight:Exec
    EnableTryBuildOnFocus:Exec
    EditButtonHoldTime:Exec
    SlideButtonHoldTime:Exec
    AccessoriesHeader:Exec
    ForceFeedback:Exec
    ContextTutorial:Exec
    ReplayRecording:Exec
    ReplayRecordingLargeTeams:Exec
    ReplayRecordingCreativeMode:Exec
    UsePowerSavingMode:Exec
    ShadowPlayHighlights:Exec
    ShowTemperature:Exec
    BuildingPossession:Exec
    GammaValue:Exec
    MasterVolume:Exec
    MusicVolume:Exec
    SoundFXVolume:Exec
    DialogVolume:Exec
    VoiceChatVolume:Exec
    CinematicsVolume:Exec
    Subtitles:Exec
    Quality:Exec
    VoiceChat:Exec
    PushToTalk:Exec
    ProximityVoiceChat:Exec
    VoiceChatInputDevice:Exec
    VoiceChatOutputDevice:Exec
    AllowBackgroundAudio:Exec
    ColorBlindMode:Exec
    ColorBlindStrength:Exec
    IgnoreGamepadInput:Exec
    LockPrimaryInputMethodToMouse:Exec
    EnableRudderControl:Exec
    RudderDeadZone:Exec
    RudderMaxThrottle:Exec
    FOVMinimum:Exec
    FOVMaximum:Exec
    VisualizeSoundEffects:Exec
    VisualizeSoundEffectsHeader:Exec
    MoveStickDeadZone:Exec
    LookStickDeadZone:Exec
    LookSensitivityPreset:Exec
    LookSensitivityPresetAds:Exec
    LookBuildModeMultiplier:Exec
    LookEditModeMultiplier:Exec
    UseAdvancedOptions:Exec
    LookHorizontalSpeed:Exec
    LookVerticalSpeed:Exec
    LookHorizontalSpeedAds:Exec
    LookVerticalSpeedAds:Exec
    LookHorizontalBoostSpeed:Exec
    LookVerticalBoostSpeed:Exec
    LookBoostAccelerationTime:Exec
    LookHorizontalBoostSpeedAds:Exec
    LookVerticalBoostSpeedAds:Exec
    LookBoostAccelerationTimeAds:Exec
    InstantBoostWhenBuilding:Exec
    LookEaseTime:Exec
    LookInputCurve:Exec
    AimAssistStrength:Exec
    UseLegacyControls:Exec
    PlayerSurveysAllowed:Exec
    LocalNotifications:Exec
    FireMode:Exec
   ) {is=title K2Node_SwitchEnum}
   SetIsEnabled(
    execute=[
     @n2.ControllerLookSensitivityPitch
     @n2.ControllerLookSensitivityYaw
     @n2.GamepadScopedMultiplier
     @n2.GamepadTargetingMultiplier
     @n2.MouseScopedMultiplier
     @n2.MouseSensitivityPitch
     @n2.MouseSensitivityYaw
     @n2.MouseTargetingMultiplier
     @n2.TouchDragScopedSensitivity
     @n2.TouchDragSensitivity
     @n2.TouchDragTargetingSensitivity
    ]
    bInIsEnabled=NOT Boolean(A=Get bRCPawn() -> (bRCPawn:bool))
   )
  ]}
 SetFontSize {nodes [
   n1 = SetFontSize!() -> (then, Text:Common Text Object Reference, DefaultSize:int, MobileSize:int) {is=title K2Node_FunctionEntry}
   > n2 = Set TempFont(TempFont:Slate Font Info Structure=Get Font(self=@n1)) -> (Output_Get:Slate Font Info Structure)
   n3 = IsMobileGame()
   > n4 = Set members in Slate Font Info!(
    execute:Exec
    StructRef:"Slate Font Info Structure (by ref)"=@n2
    Size:float="To Float (Integer)"(InInt=Select(Option 0:int=@n1.DefaultSize, Option 1:int=@n1.MobileSize, Index=@n3) -> (ReturnValue:int))
   ) -> (then:Exec, StructOut:"Slate Font Info Structure (by ref)") {is=title K2Node_SetFieldsInStruct}
   > SetFont(self=@n1, InFontInfo=@n4)
  ]}
 UpdateSize {nodes [
   UpdateSize() {is=title K2Node_FunctionEntry}
   > SetSize!(
    execute
    self:Spacer Object Reference=Get SpacerBottom() -> (SpacerBottom:Spacer Object Reference)
    InSize:Vector 2D Structure=Select(Option 0:Vector 2D Structure=(X=1,Y=8), Option 1:Vector 2D Structure=(X=1,Y=60), Index=IsBROnly!() -> (ReturnValue:bool)) -> (ReturnValue:Vector 2D Structure)
   ) -> (then)
   SetHeightOverride(
    self=Get SizeBoxContainer() -> (SizeBoxContainer:Size Box Object Reference)
    InHeightOverride=Select(Option 0:double=48, Option 1:double=48, Index=IsBROnly()) -> (ReturnValue:double)
   )
  ]}
]
variables [
 FontSizeMobile {type=integer, default=28}
 LocalSettingType {type=ESettingType, default=None}
]
widgets {
 info {inheritedWidgetCount=0, parentClass=&CommonUserWidget, rootWidgetClass=&SizeBox, widgetCount=6}
 widgets [
  SizeBoxContainer {
   widgetClassPath=&SizeBox
   bIsVariable=true
   widget=SizeBoxContainer
   properties {
    bOverride_MinDesiredHeight=true
    bOverride_MinDesiredWidth=true
    heightOverride=48
    minDesiredHeight=48
    minDesiredWidth=456
    navigation=&self:WidgetTree.SizeBoxContainer.WidgetNavigation_1
   }
  }
  VerticalBox_1 {widgetClassPath=&VerticalBox, parent=SizeBoxContainer, widget=VerticalBox_1, slot {class=SizeBoxSlot}}
  Overlay_1 {widgetClassPath=&Overlay, parent=VerticalBox_1, widget=Overlay_1, slot {class=VerticalBoxSlot}}
  TextHeader {
   widgetClassPath=&CommonTextBlock
   bIsVariable=true
   parent=Overlay_1
   widget=TextHeader
   properties {
    style=&TextStyle-Base-S-B-Blue-40_C
    text=Header Text
    colorAndOpacity {colorUseRule=UseColor_Specified, specifiedColor=<rgba 0.008 0.011 0.0262 1>}
    font {
     bForceMonospaced=false
     bMaterialIsStencil=false
     fontMaterial=null
     fontObject=&NotoSans
     letterSpacing=0
     monospacedWidth=1
     size=18
     skewAmount=0
     typefaceFontName=NotoSans-Bold
     outlineSettings {
      bApplyOutlineToDropShadows=false
      bMiteredCorners=false
      bSeparateFillAlpha=false
      outlineMaterial=null
      outlineSize=0
      outlineColor=<rgba 0 0 0 1>
     }
    }
    shadowOffset=<0 0>
   }
   slot {class=OverlaySlot, properties {horizontalAlignment=HAlign_Fill}}
  }
  Image_0 {
   widgetClassPath=&Image
   bIsVariable=true
   parent=VerticalBox_1
   widget=Image_0
   properties {brush {
     bIsDynamicallyLoaded=false
     drawAs=Image
     imageType=NoImage
     mirroring=NoMirror
     resourceName=None
     resourceObject=null
     tiling=NoTile
     imageSize=<32 2>
     margin=<ltrb 0 0 0 0>
     outlineSettings {
      bUseBrushTransparency=false
      roundingType=HalfHeightRadius
      width=0
      color {colorUseRule=UseColor_Specified, specifiedColor=<rgba 0 0 0 0>}
      cornerRadii=<0 0 0 0>
     }
     tintColor {colorUseRule=UseColor_Specified, specifiedColor=<rgba 0.008 0.011 0.0262 1>}
     uVRegion {bIsValid=false, max=<0 0>, min=<0 0>}
    }}
   slot {class=VerticalBoxSlot, properties {verticalAlignment=VAlign_Bottom, size {sizeRule=Fill, value=1}}}
  }
  SpacerBottom {
   widgetClassPath=&Spacer
   bIsVariable=true
   parent=VerticalBox_1
   widget=SpacerBottom
   properties {size=<1 0>}
   slot {class=VerticalBoxSlot}
  }
 ]
}
