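# Gamepad settings side panel for custom bindings, with radio groups for configs and platforms and a Reset button showing a key hint.
# On activation it wraps header text and, in Campaign, registers Apply, Reset and Back input handlers, hiding the first two; deactivation removes them.
# Handlers, Reset clicks and focus-path entry only broadcast dispatchers, leaving the actual work to listeners.
# Desired focus is the selected config button.
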WidgetBlueprint /FortniteCoreUI/Settings/CustomBindings_v3/GamepadInfoSelectorPanel_v3
graphs [
 BP_GetDesiredFocusTarget {nodes [
   Get Desired Focus Target()
   n1 = Get FortRadioButtonGroup_Configs() -> (FortRadioButtonGroup_Configs:Fort Radio Button Group Object Reference)
   > Return Node!(
    execute
    ReturnValue:Widget Object Reference=GetSelectedButton!(self:Common Button Group Legacy Object Reference=Get CreatedButtonsGroup!(self:Fort Radio Button Group Object Reference=@n1) -> (CreatedButtonsGroup:Common Button Group Legacy Object Reference)) -> (ReturnValue:Common Button Legacy Object Reference)
   )
  ]}
 EventGraph {nodes [
   "On Clicked (ResetButton)"() {is=title K2Node_ComponentBoundEvent}
   > Call Gamepad Info Selector Reset!(execute:Exec, self:Gamepad Info Selector Panel V 3 Object Reference) -> (then:Exec) {is=title K2Node_CallDelegate}
   Event On Activated()
   > SetAutoWrapText!(
    execute
    self:Text Object Reference=[
     Get TextHeader!(self:Options Menu Header Object Reference=Get OptionsMenuHeader_Platforms() -> (OptionsMenuHeader_Platforms:Options Menu Header Object Reference)) -> (TextHeader:Common Text Object Reference)
     Get TextHeader(self=Get OptionsMenuHeader_Configs() -> (OptionsMenuHeader_Configs:Options Menu Header Object Reference))
    ]
    InAutoTextWrap:bool=true
   ) -> (then)
   > Branch(Condition=IsInSubGame!(SubGame:ESubGame Enum=Campaign) -> (ReturnValue:bool))
   > SetInputActionHandler!(
    execute
    self:Common Activatable Panel Legacy Object Reference
    InputActionRow:Data Table Row Handle Structure=Get Apply() -> (Apply:Data Table Row Handle Structure)
    CommitedEvent:Delegate=Create Event()
   ) -> (then)
   n1 = Get!() -> (Apply_DataTable:Data Table Object Reference, Apply_RowName:Name)
   > SetActionHandlerState!(
    execute
    self:Common Activatable Panel Legacy Object Reference
    DataTable:Data Table Object Reference=@n1
    RowName:Name=@n1.Apply_RowName
    State:EInputActionState Enum=HiddenAndDisabled
   ) -> (then)
   > SetInputActionHandler(InputActionRow=Get Reset() -> (Reset:Data Table Row Handle Structure), CommitedEvent=Create Event())
   n2 = Get!() -> (Reset_DataTable:Data Table Object Reference, Reset_RowName:Name)
   > SetActionHandlerState(DataTable=@n2, RowName=@n2.Reset_RowName, State=Hidden)
   > SetInputActionHandler(InputActionRow=Get Back() -> (Back:Data Table Row Handle Structure), CommitedEvent=Create Event())
   Event On Deactivated()
   > n3 = Sequence!(execute) -> (then_0, then_1, then_2:Exec)
   n4 = Get Back()
   > Branch(Condition=HasInputActionHandler!(self:Common Activatable Panel Legacy Object Reference, InputActionRow:Data Table Row Handle Structure=@n4) -> (ReturnValue:bool))
   > RemoveInputActionHandler!(execute, self:Common Activatable Panel Legacy Object Reference, InputActionRow:Data Table Row Handle Structure=@n4) -> (then)
   n5 = Get Reset()
   Branch(execute=@n3.then_2, Condition=HasInputActionHandler(InputActionRow=@n5))
   > RemoveInputActionHandler(InputActionRow=@n5)
   n6 = Get Apply()
   Branch(execute=@n3.then_1, Condition=HasInputActionHandler(InputActionRow=@n6))
   > RemoveInputActionHandler(InputActionRow=@n6)
   Event OnAddedToFocusPath!() -> (OutputDelegate, then, InFocusEvent:Focus Event Structure) {is=title K2Node_Event}
   > Call Gamepad Info Selector Focus!(execute:Exec, self:Gamepad Info Selector Panel V 3 Object Reference) -> (then:Exec) {is=title K2Node_CallDelegate}
  ]}
 GamepadInfoSelectorApply {nodes [GamepadInfoSelectorApply() {is=title K2Node_FunctionEntry}]}
 GamepadInfoSelectorBack {nodes [GamepadInfoSelectorBack() {is=title K2Node_FunctionEntry}]}
 GamepadInfoSelectorFocus {nodes [GamepadInfoSelectorFocus() {is=title K2Node_FunctionEntry}]}
 GamepadInfoSelectorReset {nodes [GamepadInfoSelectorReset() {is=title K2Node_FunctionEntry}]}
 HandleApply {nodes [
   HandleApply() {is=title K2Node_FunctionEntry}
   > Call Gamepad Info Selector Apply!(execute:Exec, self:Gamepad Info Selector Panel V 3 Object Reference) -> (then:Exec) {is=title K2Node_CallDelegate}
   > Return Node!(execute, PassThrough:bool=false)
  ]}
 HandleBack {nodes [
   HandleBack() {is=title K2Node_FunctionEntry}
   > Call Gamepad Info Selector Back!(execute:Exec, self:Gamepad Info Selector Panel V 3 Object Reference) -> (then:Exec) {is=title K2Node_CallDelegate}
   > Return Node(PassThrough=false)
  ]}
 HandleReset {nodes [
   HandleReset() {is=title K2Node_FunctionEntry}
   > Call Gamepad Info Selector Reset()
   > Return Node(PassThrough=false)
  ]}
]
variables [
 Apply {type=DataTableRowHandle, default {dataTable=&InputActionDataTable, rowName=Apply}}
 Back {type=DataTableRowHandle, default {dataTable=&InputActionDataTable, rowName=Back}}
 GamepadInfoSelectorApply {}
 GamepadInfoSelectorBack {}
 GamepadInfoSelectorFocus {}
 GamepadInfoSelectorReset {}
 Reset {type=DataTableRowHandle, default {dataTable=&InputActionDataTable, rowName=ResetToDefault}}
]
widgets {
 info {inheritedWidgetCount=0, parentClass=/Script/FortniteUI.FortGamePadInfoSelectorPanel, rootWidgetClass=&SizeBox, widgetCount=14}
 widgets [
  SizeBox_0 {widgetClassPath=&SizeBox, widget=SizeBox_0, properties {bOverride_WidthOverride=true, widthOverride=300}}
  VerticalBox_1 {widgetClassPath=&VerticalBox, parent=SizeBox_0, widget=VerticalBox_1, slot {class=SizeBoxSlot}}
  VerticalBox_Configs {widgetClassPath=&VerticalBox, parent=VerticalBox_1, widget=VerticalBox_Configs, slot {class=VerticalBoxSlot}}
  OptionsMenuHeader_Configs {
   widgetClassPath=/FortniteCoreUI/Settings/OptionsMenuHeader
   bIsVariable=true
   parent=VerticalBox_Configs
   widget=OptionsMenuHeader_Configs
   properties {bIsFocusable=false}
   slot {class=VerticalBoxSlot}
  }
  FortRadioButtonGroup_Configs {
   widgetClassPath=/Game/UI/RadioButtonGroup/FortRadioButtonGroup
   bIsVariable=true
   parent=VerticalBox_Configs
   widget=FortRadioButtonGroup_Configs
   properties {widthOverride=300}
   slot {class=VerticalBoxSlot}
  }
  ResetButton {
   widgetClassPath=&Button
   parent=VerticalBox_Configs
   widget=ResetButton
   properties {
    isFocusable=false
    widgetStyle {
     clickedSlateSound {resourceObject=null}
     disabled {
      bIsDynamicallyLoaded=false
      drawAs=NoDrawType
      imageType=NoImage
      mirroring=NoMirror
      resourceName=None
      resourceObject=null
      tiling=NoTile
      imageSize=<0 0>
      margin=<ltrb 0 0 0 0>
      outlineSettings {
       bUseBrushTransparency=false
       roundingType=HalfHeightRadius
       width=0
       color {colorUseRule=UseColor_Specified, specifiedColor=<rgba 0 0 0 0>}
       cornerRadii=<0 0 0 0>
      }
      tintColor {colorUseRule=UseColor_Specified, specifiedColor=<rgba 1 1 1 1>}
      uVRegion {bIsValid=false, max=<0 0>, min=<0 0>}
     }
     disabledForeground {colorUseRule=UseColor_Specified, specifiedColor=<rgba 0.5271 0.5271 0.5271 1>}
     hovered {
      bIsDynamicallyLoaded=false
      drawAs=NoDrawType
      imageType=NoImage
      mirroring=NoMirror
      resourceName=None
      resourceObject=null
      tiling=NoTile
      imageSize=<32 32>
      margin=<ltrb 0 0 0 0>
      outlineSettings {
       bUseBrushTransparency=true
       roundingType=FixedRadius
       width=1
       color {colorUseRule=UseColor_Specified, specifiedColor=<rgba 0.7243 0.7243 0.7243 1>}
       cornerRadii=<4 4 4 4>
      }
      tintColor {colorUseRule=UseColor_Specified, specifiedColor=<rgba 0.7243 0.7243 0.7243 1>}
      uVRegion {bIsValid=false, max=<0 0>, min=<0 0>}
     }
     hoveredForeground {colorUseRule=UseColor_Specified, specifiedColor=<rgba 0 1 1 1>}
     hoveredSlateSound {resourceObject=null}
     normal {
      bIsDynamicallyLoaded=false
      drawAs=NoDrawType
      imageType=NoImage
      mirroring=NoMirror
      resourceName=None
      resourceObject=null
      tiling=NoTile
      imageSize=<32 32>
      margin=<ltrb 0 0 0 0>
      outlineSettings {
       bUseBrushTransparency=true
       roundingType=FixedRadius
       width=1
       color {colorUseRule=UseColor_Specified, specifiedColor=<rgba 0.6951 0.6951 0.6951 1>}
       cornerRadii=<4 4 4 4>
      }
      tintColor {colorUseRule=UseColor_Specified, specifiedColor=<rgba 0.4955 0.4955 0.4955 1>}
      uVRegion {bIsValid=false, max=<0 0>, min=<0 0>}
     }
     normalForeground {colorUseRule=UseColor_Specified, specifiedColor=<rgba 0 1 1 1>}
     normalPadding=<ltrb 0 0 0 0>
     pressed {
      bIsDynamicallyLoaded=false
      drawAs=NoDrawType
      imageType=NoImage
      mirroring=NoMirror
      resourceName=None
      resourceObject=null
      tiling=NoTile
      imageSize=<32 32>
      margin=<ltrb 0 0 0 0>
      outlineSettings {
       bUseBrushTransparency=true
       roundingType=FixedRadius
       width=1
       color {colorUseRule=UseColor_Specified, specifiedColor=<rgba 0.7243 0.7243 0.7243 1>}
       cornerRadii=<4 4 4 4>
      }
      tintColor {colorUseRule=UseColor_Specified, specifiedColor=<rgba 0.3843 0.3843 0.3843 1>}
      uVRegion {bIsValid=false, max=<0 0>, min=<0 0>}
     }
     pressedForeground {colorUseRule=UseColor_Specified, specifiedColor=<rgba 0 1 1 1>}
     pressedPadding=<ltrb 12 1.5 7 6>
     pressedSlateSound {resourceObject=null}
    }
   }
   slot {class=VerticalBoxSlot}
  }
  CommonBorder_Reset {
   widgetClassPath=&CommonBorder
   bIsVariable=true
   parent=ResetButton
   widget=CommonBorder_Reset
   properties {
    visibility=HitTestInvisible
    background {
     bIsDynamicallyLoaded=false
     drawAs=Box
     imageType=NoImage
     mirroring=NoMirror
     resourceName=None
     resourceObject=null
     tiling=NoTile
     imageSize=<32 32>
     margin=<ltrb 0 0 0 0>
     outlineSettings {
      bUseBrushTransparency=false
      roundingType=HalfHeightRadius
      width=0
      color {colorUseRule=UseColor_Specified, specifiedColor=<rgba 0 0 0 0>}
      cornerRadii=<0 0 0 0>
     }
     tintColor {colorUseRule=UseColor_Specified, specifiedColor=<rgba 0 0.0021 0.0365 0.9>}
     uVRegion {bIsValid=false, max=<0 0>, min=<0 0>}
    }
    padding=<ltrb 4 4 4 4>
   }
   slot {class=ButtonSlot, properties {horizontalAlignment=HAlign_Fill, verticalAlignment=VAlign_Fill, padding=<ltrb 10 5 10 0>}}
  }
  HorizontalBox_Reset {
   widgetClassPath=&HorizontalBox
   parent=CommonBorder_Reset
   widget=HorizontalBox_Reset
   properties {visibility=HitTestInvisible}
   slot {class=BorderSlot, properties {padding=<ltrb 4 4 4 4>}}
  }
  FortRichTextBlock_Reset1 {
   widgetClassPath=/Script/FortniteUI.FortRichTextBlockLegacy
   bIsVariable=true
   parent=HorizontalBox_Reset
   widget=FortRichTextBlock_Reset1
   properties {styleSet=/FortUILibrary/Foundation/Buttons/TextStyles/TextStyle_Gamepad_Bindings, text="<Font.Bold>Press </>"}
   slot {class=HorizontalBoxSlot, properties {verticalAlignment=VAlign_Center}}
  }
  KeybindWidget_Reset {
   widgetClassPath=/FortUILibrary/Input/KeybindWidget
   bIsVariable=true
   parent=HorizontalBox_Reset
   widget=KeybindWidget_Reset
   properties {boundKeyFallback=Gamepad_FaceButton_Left, inputTypeOverride=Gamepad}
   slot {class=HorizontalBoxSlot}
  }
  FortRichTextBlock_Reset2 {
   widgetClassPath=&FortRichTextBlockLegacy
   bIsVariable=true
   parent=HorizontalBox_Reset
   widget=FortRichTextBlock_Reset2
   properties {styleSet=&TextStyle_Gamepad_Bindings, text="<Font.Bold> To Reset</>"}
   slot {class=HorizontalBoxSlot, properties {verticalAlignment=VAlign_Center}}
  }
  VerticalBox_Platforms {
   widgetClassPath=&VerticalBox
   bIsVariable=true
   parent=VerticalBox_1
   widget=VerticalBox_Platforms
   slot {class=VerticalBoxSlot, properties {padding=<ltrb 0 50 0 0>}}
  }
  OptionsMenuHeader_Platforms {
   widgetClassPath=&OptionsMenuHeader
   bIsVariable=true
   parent=VerticalBox_Platforms
   widget=OptionsMenuHeader_Platforms
   properties {bIsFocusable=false}
   slot {class=VerticalBoxSlot}
  }
  FortRadioButtonGroup_Platforms {
   widgetClassPath=&FortRadioButtonGroup
   bIsVariable=true
   parent=VerticalBox_Platforms
   widget=FortRadioButtonGroup_Platforms
   properties {widthOverride=300}
   slot {class=VerticalBoxSlot}
  }
 ]
}
