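# Gamepad binding list button that shows a key icon beside a rich-text description, with a dash when no key is bound.
# SetData stores key, text and action name; UpdateButton renders them from a given button instance.
# Instances double as list data: OnListItemObjectSet casts the item to this class and copies its display.
# Helpers set content alignment and a non-interactable style, PreConstruct previews at design time, and a dispatcher carries binding changes.
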WidgetBlueprint /FortUILibrary/Foundation/Buttons/GamepadKeyTextButton
graphs [
 ChangeTheBinding_Button {nodes [ChangeTheBinding_Button!() -> (then, ActionName:Name, Key:Key Structure) {is=title K2Node_FunctionEntry}]}
 EventGraph {nodes [
   n1 = SetDataAndUpdateButton!() -> (OutputDelegate, then, Key:Key Structure, DescText:Text, ActionName:Name) {is=title K2Node_CustomEvent}
   > SetData!(execute, self, Key:Key Structure=@n1.Key, DescText:Text=@n1.DescText, ActionName:Name=@n1.ActionName) -> (then)
   > UpdateButton!(execute, self, Button:Gamepad Key Text Button Object Reference=Self-Reference()) -> (then)
   n2 = SetData!() -> (OutputDelegate, then, Key:Key Structure, DescText:Text, ActionName:Name) {is=title K2Node_CustomEvent}
   > Set KeyData(KeyData:Key Structure=@n2.Key) -> (Output_Get:Key Structure)
   > Set TextData(TextData:Text=@n2.DescText) -> (Output_Get:Text)
   > Set ActionNameData(ActionNameData:Name=@n2.ActionName) -> (Output_Get:Name=None)
   n3 = UpdateButton!() -> (OutputDelegate, then, Button:Gamepad Key Text Button Object Reference) {is=title K2Node_CustomEvent}
   > SetText(
    self:Fort Rich Text Block Legacy Object Reference=Get FortRichTextBlock_Desc() -> (FortRichTextBlock_Desc:Fort Rich Text Block Legacy Object Reference)
    InText:Text=Get TextData!(self:Gamepad Key Text Button Object Reference=@n3.Button) -> (TextData:Text)
   )
   n4 = Get KeyData!(self:Gamepad Key Text Button Object Reference=@n3.Button) -> (KeyData:Key Structure)
   > SetBoundKey!(
    execute
    self:Fort Keybind Widget Object Reference=Get BoundKey() -> (BoundKey:Keybind Widget Object Reference)
    NewBoundAction:Key Structure=@n4
   ) -> (then)
   n5 = Get KeyNone() -> (KeyNone:Key Structure)
   > SetVisibility(
    self=Get SizeBox_Nokey() -> (SizeBox_Nokey:Size Box Object Reference)
    InVisibility=Select(Option 0=Collapsed, Option 1=SelfHitTestInvisible, Index="Equal (Key)"!(A:Key Structure=@n4, B:Key Structure=@n5) -> (ReturnValue:bool))
   )
   n6 = Event OnListItemObjectSet()
   > n7 = Cast To GamepadKeyTextButton!(execute, Object=@n6.ListItemObject) -> (then, CastFailed, AsGamepad Key Text Button:Gamepad Key Text Button Object Reference) {is=title K2Node_DynamicCast}
   > UpdateButton(Button=@n7) {is=title K2Node_CallFunction}
   n8 = Event PreConstruct()
   > Branch(Condition=@n8.IsDesignTime)
   > UpdateContentAlignment()
   > SetStyle!(
    execute
    self:Common Button Base Object Reference
    InStyle:Common Button Style Class Reference=Get ControllerInputStyle() -> (ControllerInputStyle:Common Button Style Class Reference)
   ) -> (then)
   > SetText(self=Get FortRichTextBlock_Desc(), InText=Get ButtonText() -> (ButtonText:Text))
   > SetBoundKey(self=Get BoundKey(), NewBoundAction=Get KeyData!() -> (KeyData))
  ]}
 SetContentAlignment {nodes [
   n1 = SetContentAlignment!() -> (then, ContentAlignment:EHorizontalAlignment Enum=HAlign_Fill) {is=title K2Node_FunctionEntry}
   > Set ContentAlignment(ContentAlignment:EHorizontalAlignment Enum=@n1) -> (Output_Get:EHorizontalAlignment Enum=HAlign_Fill)
   > UpdateContentAlignment()
  ]}
 SetNonInteractableStyle {nodes [
   SetNonInteractableStyle() {is=title K2Node_FunctionEntry}
   > SetStyle(InStyle=/FortUILibrary/Foundation/Buttons/ButtonStyles/ButtonStyle-GamepadBindings_NonInteractable.ButtonStyle-GamepadBindings_NonInteractable_C)
  ]}
 UpdateContentAlignment {nodes [
   UpdateContentAlignment() {is=title K2Node_FunctionEntry}
   > n1 = Cast To BorderSlot!(
    execute
    Object=GetContentSlot!(self:Content Widget Object Reference=Get ContentBorder() -> (ContentBorder:Border Object Reference)) -> (ReturnValue:Panel Slot Object Reference)
   ) -> (then, CastFailed, AsBorder Slot:Border Slot Object Reference) {is=title K2Node_DynamicCast}
   > SetHorizontalAlignment(
    self:Border Slot Object Reference=@n1
    InHorizontalAlignment=Get ContentAlignment() -> (ContentAlignment:EHorizontalAlignment Enum=HAlign_Fill)
   )
  ]}
]
variables [
 ActionNameData {type=string, default=None}
 ButtonText {type=string, default="<Font.Bold>Bold</> <Font.DoubleTap>Double-Tap</> <Font.Hold>Hold</>"}
 ChangeTheBinding_Button {}
 ContentAlignment {type=EHorizontalAlignment, default=HAlign_Left}
 ControllerInputStyle {type=Class, default=null}
 KeyData {type=Key, default=None}
 KeyNone {type=Key, default=None}
 TextData {type=string, default=""}
]
widgets {
 info {inheritedWidgetCount=1, parentClass=/Script/FortniteUI.FortGamepadCustomListItem, rootWidgetClass=&Border, widgetCount=8}
 widgets [
  ContentBorder {
   widgetClassPath=&Border
   bIsVariable=true
   widget=ContentBorder
   properties {
    horizontalAlignment=HAlign_Left
    background {
     bIsDynamicallyLoaded=false
     drawAs=NoDrawType
     imageType=NoImage
     mirroring=NoMirror
     resourceName=None
     resourceObject=null
     tiling=NoTile
     imageSize=<32 32>
     margin=<ltrb 0 0 0 0>
     outlineSettings {
      bUseBrushTransparency=false
      roundingType=HalfHeightRadius
      width=0
      color {colorUseRule=UseColor_Specified, specifiedColor=<rgba 0 0 0 0>}
      cornerRadii=<0 0 0 0>
     }
     tintColor {colorUseRule=UseColor_Specified, specifiedColor=<rgba 1 1 1 1>}
     uVRegion {bIsValid=false, max=<0 0>, min=<0 0>}
    }
    padding=<ltrb 0 0 0 0>
   }
  }
  HorizontalBox {
   widgetClassPath=&HorizontalBox
   parent=ContentBorder
   widget=HorizontalBox
   slot {class=BorderSlot, properties {horizontalAlignment=HAlign_Left, padding=<ltrb 0 0 0 0>}}
  }
  InputActionWidget {
   widgetClassPath=&CommonActionWidget
   bInherited=true
   bIsVariable=true
   parent=HorizontalBox
   widget=InputActionWidget
   properties {
    progressMaterialParam=Percentage
    progressMaterialBrush {
     bIsDynamicallyLoaded=false
     drawAs=Image
     imageType=NoImage
     mirroring=NoMirror
     resourceName=None
     resourceObject=/Game/UI/Materials/HollowCircularProgressBar
     tiling=NoTile
     imageSize=<36 36>
     margin=<ltrb 0 0 0 0>
     outlineSettings {
      bUseBrushTransparency=false
      roundingType=HalfHeightRadius
      width=0
      color {colorUseRule=UseColor_Specified, specifiedColor=<rgba 0 0 0 0>}
      cornerRadii=<0 0 0 0>
     }
     tintColor {colorUseRule=UseColor_Specified, specifiedColor=<rgba 1 1 0 1>}
     uVRegion {bIsValid=false, max=<0 0>, min=<0 0>}
    }
   }
   slot {class=HorizontalBoxSlot, properties {horizontalAlignment=HAlign_Left, verticalAlignment=VAlign_Center, padding=<ltrb 2 2 0 0>}}
  }
  SizeBox_0 {
   widgetClassPath=&SizeBox
   parent=HorizontalBox
   widget=SizeBox_0
   properties {bOverride_HeightOverride=true, bOverride_WidthOverride=true, heightOverride=32, widthOverride=32}
   slot {class=HorizontalBoxSlot, properties {horizontalAlignment=HAlign_Left, verticalAlignment=VAlign_Center, padding=<ltrb 10 0 0 0>}}
  }
  BoundKey {
   widgetClassPath=/FortUILibrary/Input/KeybindWidget
   bIsVariable=true
   parent=SizeBox_0
   widget=BoundKey
   properties {inputTypeOverride=Gamepad}
   slot {class=SizeBoxSlot}
  }
  SizeBox_Nokey {
   widgetClassPath=&SizeBox
   bIsVariable=true
   parent=HorizontalBox
   widget=SizeBox_Nokey
   properties {bOverride_HeightOverride=true, bOverride_WidthOverride=true, heightOverride=32, visibility=Collapsed, widthOverride=32}
   slot {class=HorizontalBoxSlot}
  }
  FortRichTextBlock_NoKey {
   widgetClassPath=/Script/FortniteUI.FortRichTextBlockLegacy
   bIsVariable=true
   parent=SizeBox_Nokey
   widget=FortRichTextBlock_NoKey
   properties {styleSet=/FortUILibrary/Foundation/Buttons/TextStyles/TextStyle_Gamepad_Bindings, text=-, textMargin=<ltrb 10 4 10 4>}
   slot {class=SizeBoxSlot}
  }
  FortRichTextBlock_Desc {
   widgetClassPath=&FortRichTextBlockLegacy
   bIsVariable=true
   parent=HorizontalBox
   widget=FortRichTextBlock_Desc
   properties {
    allowContextMenu=false
    disableTouchInput=true
    styleSet=&TextStyle_Gamepad_Bindings
    text="<Font.Bold>Bold</> <Font.DoubleTap>Double Tap</> <Font.Hold>Hold</>"
    textMargin=<ltrb 0 4 0 4>
   }
   slot {class=HorizontalBoxSlot, properties {horizontalAlignment=HAlign_Left, verticalAlignment=VAlign_Center, padding=<ltrb 10 0 0 0>}}
  }
 ]
}
