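# Gamepad custom-bindings settings panel that lists selectable actions for the player to choose from, with a Clear option.
# UpdatePanel rebuilds the list from input data, creating GamepadKeyTextButton items with the custom preset's key and label.
# An enum switch only creates items for combat, building, edit, LTM-ability, combined and AllModes groups.
# Item clicks, Clear and Back fire delegates, with Back and Clear input handlers registered while active.
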WidgetBlueprint /FortniteCoreUI/Settings/CustomBindings_v3/GamepadCustomActionSelectorPanel_v3
graphs [
 CreateCustomListItem {nodes [
   n1 = CreateCustomListItem!() -> (then, ActionName:Name, InputActionGroup:EFortInputActionGroup Enum=AllModes) {is=title K2Node_FunctionEntry}
   > n2 = Is Valid(InputObject=Get FortGlobalUIContext!() -> (ReturnValue:Fort Global UIContext Object Reference) {is=title K2Node_GetBlueprintContextEx})
   > Set ActionName_Local(ActionName_Local:Name=@n1) -> (Output_Get:Name=None)
   n3 = Get FortGlobalUIContext()
   > Set CustomPresetName_Local(CustomPresetName_Local:Name=String To Name(InString=GetCustomGamepadInputPresetName!(self:Fort Global UIContext Object Reference=@n3) -> (ReturnValue:String))) -> (Output_Get:Name=None)
   > n4 = Switch on EFortInputActionGroup!(execute, Selection:EFortInputActionGroup Enum=@n1.InputActionGroup) -> (
    AllModes:Exec
    Combat:Exec
    Building:Exec
    Movement:Exec
    Edit:Exec
    Death:Exec
    Cinematic:Exec
    Picker:Exec
    Other:Exec
    Interaction:Exec
    AthenaLTMAbilities:Exec
    ShoppingCart:Exec
    ShoppingCartDriver:Exec
    ShoppingCartPassenger:Exec
    Cannon:Exec
    CannonDriver:Exec
    CannonPassenger:Exec
    GolfCart:Exec
    GolfCartDriver:Exec
    GolfCartPassenger:Exec
    QuadCrasher:Exec
    QuadCrasherDriver:Exec
    QuadCrasherPassenger:Exec
    Biplane:Exec
    BiplaneDriver:Exec
    BiplaneGroundDriver:Exec
    BiplanePassenger:Exec
    Hamsterball:Exec
    Jackal:Exec
    Ostrich:Exec
    OstrichDriver:Exec
    OstrichPassenger:Exec
    Meatball:Exec
    MeatballDriver:Exec
    MeatballPassenger:Exec
    HoagieDriver:Exec
    Dagwood:Exec
    DagwoodDriver:Exec
    DagwoodPassenger:Exec
    Nevada:Exec
    NevadaDriver:Exec
    MountedTurret:Exec
    Spectating:Exec
    FullscreenMap:Exec
    CreativeAll:Exec
    CreativeModeratorMode:Exec
    CreativeMoveToolEquipped:Exec
    CreativeMoveObjectsFreely:Exec
    CreativeMoveBuildingsOnGrid:Exec
    CreativeFlying:Exec
    CreativeIslandPanel:Exec
    CreativeChair:Exec
    PropSelectorEquipped:Exec
    AlternativeInteractCarry:Exec
    AlternativeInteractPlace:Exec
    CreativeInputDevice:Exec
    DBNOCarryStart:Exec
    DBNOCarryStop:Exec
    DBNOCarry:Exec
    InteractionStart:Exec
    InteractionStop:Exec
    BattleLab:Exec
    RemoteControl:Exec
    SuperDuper:Exec
    Tether:Exec
    Riding:Exec
    Tank:Exec
    TankDriver:Exec
    TankPassenger:Exec
    RockVehicle:Exec
    RockVehicleB:Exec
    Dirtbike:Exec
    DirtbikeDriver:Exec
    DirtbikePassenger:Exec
    Sportbike:Exec
    SportbikeDriver:Exec
    SportbikePassenger:Exec
    PlayerAugments:Exec
    Targeting:Exec
    SimpleBuild:Exec
    CombatAndBuilding:Exec
    CombatAndAthenaLTMAbilities:Exec
    CombatBuildingAndAthenaLTMAbilities:Exec
   ) {is=title K2Node_SwitchEnum}
   Set InputActionGroup_Local(
    execute=[
     @n4.AllModes
     @n4.AthenaLTMAbilities
     @n4.Building
     @n4.Combat
     @n4.CombatAndAthenaLTMAbilities
     @n4.CombatAndBuilding
     @n4.CombatBuildingAndAthenaLTMAbilities
     @n4.Edit
    ]
    InputActionGroup_Local:EFortInputActionGroup Enum=@n1.InputActionGroup
   ) -> (Output_Get:EFortInputActionGroup Enum=AllModes)
   > n5 = Create Widget!(
    execute:Exec
    Class:User Widget Class Reference=/FortUILibrary/Foundation/Buttons/GamepadKeyTextButton.GamepadKeyTextButton_C
    OwningPlayer:Player Controller Object Reference
    ButtonText:Text="LOCTEXT(\"<Font.Bold>Bold</> <Font.DoubleTap>Double-Tap</> <Font.Hold>Hold</>\")"
    ControllerInputStyle:Common Button Style Class Reference
    Style:Common Button Style Class Reference=/FortUILibrary/Foundation/Buttons/ButtonStyles/ButtonStyle-GamepadBindings.ButtonStyle-GamepadBindings_C
    bApplyAlphaOnDisable:bool=False
    bLocked:bool=False
    bSelectable:bool=True
    bShouldSelectUponReceivingFocus:bool=False
    bInteractableWhenSelected:bool=True
    bToggleable:bool=False
    bTriggerClickedAfterSelection:bool=False
    bDisplayInputActionWhenNotInteractable:bool=True
    bHideInputActionWithKeyboard:bool=False
    bShouldUseFallbackDefaultInputAction:bool=True
    bRequiresHold:bool=False
    HoldData:Common UIHold Data Class Reference
    ClickMethod:EButtonClickMethod Enum=DownAndUp
    InputPriority:int=0
    TriggeringInputAction:Data Table Row Handle Structure
   ) -> (then:Exec, ReturnValue:Gamepad Key Text Button Object Reference)
   > Set ReturnItem_Local(ReturnItem_Local:Gamepad Key Text Button Object Reference=@n5) -> (Output_Get:Gamepad Key Text Button Object Reference)
   n6 = Get List Item Data!(
    PresetName:Name=Get CustomPresetName_Local() -> (CustomPresetName_Local:Name=None)
    ActionName:Name=Get ActionName_Local() -> (ActionName_Local:Name=None)
    InputActionGroup:EFortInputActionGroup Enum=Get InputActionGroup_Local() -> (InputActionGroup_Local:EFortInputActionGroup Enum=AllModes)
   ) -> (ReturnKey:Key Structure, ReturnDescText:Text) {is=title K2Node_MacroInstance}
   > n7 = SetData!(
    execute
    self:Gamepad Key Text Button Object Reference=Get ReturnItem_Local() -> (ReturnItem_Local:Gamepad Key Text Button Object Reference)
    Key:Key Structure=@n6
    DescText:Text=@n6.ReturnDescText
    ActionName:Name=Get ActionName_Local() -> (ActionName_Local=None)
   ) -> (then)
   Return Node!(execute=[@n7.then, @n2.Is Not Valid], ReturnItem:Gamepad Key Text Button Object Reference=Get ReturnItem_Local())
  ]}
 CustomActionSelected {nodes [CustomActionSelected!() -> (then, SelectedIndex:int, ActionName:Name) {is=title K2Node_FunctionEntry}]}
 CustomClearButton {nodes [CustomClearButton() {is=title K2Node_FunctionEntry}]}
 CustomClosedUsingBack {nodes [CustomClosedUsingBack() {is=title K2Node_FunctionEntry}]}
 EventGraph {nodes [
   "On Clicked (Button_Clear)"() {is=title K2Node_ComponentBoundEvent}
   > Call Custom Clear Button!(execute:Exec, self:Gamepad Custom Action Selector Panel V 3 Object Reference) -> (then:Exec) {is=title K2Node_CallDelegate}
   n1 = UpdatePanel!() -> (OutputDelegate, then, InputData:Array of Fort Options Menu Input Data Object References) {is=title K2Node_CustomEvent}
   > n2 = Set CurrentInputData(CurrentInputData:Array of Fort Options Menu Input Data Object References=@n1.InputData) -> (Output_Get:Array of Fort Options Menu Input Data Object References)
   > ClearListItems(self=Get CommonListView_GamepadActions() -> (CommonListView_GamepadActions:Common List View Object Reference))
   > n3 = For Each Loop(Array:Array of Fort Options Menu Input Data Object References=@n2) -> (Array Element:Fort Options Menu Input Data Object Reference)
   > n4 = CreateCustomListItem!(
    execute
    self
    ActionName:Name=Get ActionName!(self:Fort Options Menu Input Data Object Reference=@n3) -> (ActionName:Name=None)
    InputActionGroup:EFortInputActionGroup Enum=GetInputActionGroup!(self:Fort Options Menu Input Data Object Reference=@n3) -> (ReturnValue:EFortInputActionGroup Enum=AllModes)
   ) -> (then, ReturnItem:Gamepad Key Text Button Object Reference) {is=title K2Node_CallFunction}
   > Is Valid(InputObject=@n4)
   > AddItem!(execute, self:List View Object Reference=Get CommonListView_GamepadActions(), Item:Object Reference=@n4) -> (then)
   n5 = BP_OnItemClicked_Bind!() -> (OutputDelegate, then, Item:Object Reference) {is=title K2Node_CustomEvent}
   > n6 = Cast To GamepadKeyTextButton!(execute, Object=@n5.Item) -> (then, CastFailed, AsGamepad Key Text Button:Gamepad Key Text Button Object Reference) {is=title K2Node_DynamicCast}
   n7 = Get CommonListView_GamepadActions()
   > Call Custom Action Selected!(
    execute:Exec
    self:Gamepad Custom Action Selector Panel V 3 Object Reference
    SelectedIndex:int=GetIndexForItem!(
     self:List View Object Reference=@n7
     Item:Object Reference=Get Selected Item!(self:List View Object Reference=@n7) -> (ReturnValue:Object Reference) {is=title K2Node_CallFunction}
    ) -> (ReturnValue:int=0)
    ActionName:Name=Get ActionNameData!(self:Gamepad Key Text Button Object Reference=@n6) -> (ActionNameData:Name=None)
   ) -> (then:Exec) {is=title K2Node_CallDelegate}
   Event On Deactivated()
   > SetVisibility(InVisibility=Collapsed)
   > RemoveInputActionHandler!(
    execute
    self:Common Activatable Panel Legacy Object Reference
    InputActionRow:Data Table Row Handle Structure=Get Back() -> (Back:Data Table Row Handle Structure)
   ) -> (then)
   > RemoveInputActionHandler(InputActionRow=Get Clear() -> (Clear:Data Table Row Handle Structure))
   Event On Activated()
   > SetVisibility(InVisibility=Visible)
   > SetInputActionHandler!(
    execute
    self:Common Activatable Panel Legacy Object Reference
    InputActionRow:Data Table Row Handle Structure=Get Back()
    CommitedEvent:Delegate=Create Event()
   ) -> (then)
   > SetInputActionHandler(InputActionRow=Get Clear(), CommitedEvent=Create Event())
   n8 = Get!() -> (Clear_DataTable:Data Table Object Reference, Clear_RowName:Name)
   > SetActionHandlerState!(
    execute
    self:Common Activatable Panel Legacy Object Reference
    DataTable:Data Table Object Reference=@n8
    RowName:Name=@n8.Clear_RowName
    State:EInputActionState Enum=Hidden
   ) -> (then)
   Event Construct()
   > Bind Event to On Item Clicked(self:List View Object Reference=Get CommonListView_GamepadActions(), Delegate=Create Event()) {is=title K2Node_AddDelegate}
   > "SetText (Text)"(
    self=Get TextHeader!(self:Options Menu Header Object Reference=Get OptionsMenuHeader() -> (OptionsMenuHeader:Options Menu Header Object Reference)) -> (TextHeader:Common Text Object Reference)
    InText="LOCTEXT(\"Actions\")"
   )
  ]}
 GetListItemData {nodes [
   n1 = Inputs!() -> (PresetName:Name, ActionName:Name, InputActionGroup:EFortInputActionGroup Enum=AllModes)
   Outputs!(
    ReturnKey:Key Structure=GetKeyForAction!(
     self:Fort Global UIContext Object Reference=Get FortGlobalUIContext()
     Action:Name=@n1.ActionName
     bForceGamepadKey:bool=true
     PresetNameOverride:Name=@n1
    ) -> (ReturnValue:Key Structure)
    ReturnDescText:Text=GetScreenLabel!(
     self:Fort Input Data Object Reference=GetFortInputDataForBlueprint!() -> (ReturnValue:Fort Input Data Object Reference)
     PresetName:Name=@n1
     InputActionGroup:EFortInputActionGroup Enum=@n1.InputActionGroup
     ActionName:Name=@n1.ActionName
     bRichText:bool=true
    ) -> (ReturnValue:Text)
   )
  ]}
 HandleBack {nodes [
   HandleBack() {is=title K2Node_FunctionEntry}
   > Call Custom Closed Using Back!(execute:Exec, self:Gamepad Custom Action Selector Panel V 3 Object Reference) -> (then:Exec) {is=title K2Node_CallDelegate}
   > Return Node!(execute, PassThrough:bool=false)
  ]}
 HandleClear {nodes [
   HandleClear() {is=title K2Node_FunctionEntry}
   > Call Custom Clear Button()
   > Return Node(PassThrough=false)
  ]}
]
variables [
 Back {type=DataTableRowHandle, default {dataTable=&InputActionDataTable, rowName=Back}}
 Clear {type=DataTableRowHandle, default {dataTable=&InputActionDataTable, rowName=MessageClear}}
 CurrentInputData {type="array<FortOptionsMenuInputData>", default []}
 CustomActionSelected {}
 CustomClearButton {}
 CustomClosedUsingBack {}
 KeyNone {type=Key, default=None}
]
widgets {
 info {
  inheritedWidgetCount=0
  parentClass=/Script/FortniteUI.FortGamePadCustomActionSelectorPanel
  rootWidgetClass=&CommonBorder
  widgetCount=12
 }
 widgets [
  CommonBorder_1 {
   widgetClassPath=&CommonBorder
   widget=CommonBorder_1
   properties {
    horizontalAlignment=HAlign_Center
    navigation=&self:WidgetTree.CommonBorder_1.WidgetNavigation_0
    verticalAlignment=VAlign_Center
    background {
     bIsDynamicallyLoaded=false
     drawAs=Image
     imageType=NoImage
     mirroring=NoMirror
     resourceName=None
     resourceObject=null
     tiling=NoTile
     imageSize=<32 32>
     margin=<ltrb 0 0 0 0>
     outlineSettings {
      bUseBrushTransparency=false
      roundingType=HalfHeightRadius
      width=0
      color {colorUseRule=UseColor_Specified, specifiedColor=<rgba 0 0 0 0>}
      cornerRadii=<0 0 0 0>
     }
     tintColor {colorUseRule=UseColor_Specified, specifiedColor=<rgba 1 1 1 1>}
     uVRegion {bIsValid=false, max=<0 0>, min=<0 0>}
    }
    padding=<ltrb 0 0 0 0>
   }
  }
  SizeBox_0 {
   widgetClassPath=&SizeBox
   parent=CommonBorder_1
   widget=SizeBox_0
   properties {bOverride_WidthOverride=true, widthOverride=600}
   slot {class=BorderSlot, properties {horizontalAlignment=HAlign_Center, verticalAlignment=VAlign_Center, padding=<ltrb 0 0 0 0>}}
  }
  VerticalBox_0 {widgetClassPath=&VerticalBox, parent=SizeBox_0, widget=VerticalBox_0, slot {class=SizeBoxSlot}}
  OptionsMenuHeader {
   widgetClassPath=/FortniteCoreUI/Settings/OptionsMenuHeader
   bIsVariable=true
   parent=VerticalBox_0
   widget=OptionsMenuHeader
   properties {bIsFocusable=false}
   slot {class=VerticalBoxSlot}
  }
  Button_Clear {
   widgetClassPath=&Button
   parent=VerticalBox_0
   widget=Button_Clear
   properties {
    isFocusable=false
    widgetStyle {
     clickedSlateSound {resourceObject=null}
     disabled {
      bIsDynamicallyLoaded=false
      drawAs=NoDrawType
      imageType=NoImage
      mirroring=NoMirror
      resourceName=None
      resourceObject=null
      tiling=NoTile
      imageSize=<0 0>
      margin=<ltrb 0 0 0 0>
      outlineSettings {
       bUseBrushTransparency=false
       roundingType=HalfHeightRadius
       width=0
       color {colorUseRule=UseColor_Specified, specifiedColor=<rgba 0 0 0 0>}
       cornerRadii=<0 0 0 0>
      }
      tintColor {colorUseRule=UseColor_Specified, specifiedColor=<rgba 1 1 1 1>}
      uVRegion {bIsValid=false, max=<0 0>, min=<0 0>}
     }
     disabledForeground {colorUseRule=UseColor_Specified, specifiedColor=<rgba 0.5271 0.5271 0.5271 1>}
     hovered {
      bIsDynamicallyLoaded=false
      drawAs=NoDrawType
      imageType=NoImage
      mirroring=NoMirror
      resourceName=None
      resourceObject=null
      tiling=NoTile
      imageSize=<32 32>
      margin=<ltrb 0 0 0 0>
      outlineSettings {
       bUseBrushTransparency=true
       roundingType=FixedRadius
       width=1
       color {colorUseRule=UseColor_Specified, specifiedColor=<rgba 0.7243 0.7243 0.7243 1>}
       cornerRadii=<4 4 4 4>
      }
      tintColor {colorUseRule=UseColor_Specified, specifiedColor=<rgba 0.7243 0.7243 0.7243 1>}
      uVRegion {bIsValid=false, max=<0 0>, min=<0 0>}
     }
     hoveredForeground {colorUseRule=UseColor_Specified, specifiedColor=<rgba 0 1 1 1>}
     hoveredSlateSound {resourceObject=null}
     normal {
      bIsDynamicallyLoaded=false
      drawAs=NoDrawType
      imageType=NoImage
      mirroring=NoMirror
      resourceName=None
      resourceObject=null
      tiling=NoTile
      imageSize=<32 32>
      margin=<ltrb 0 0 0 0>
      outlineSettings {
       bUseBrushTransparency=true
       roundingType=FixedRadius
       width=1
       color {colorUseRule=UseColor_Specified, specifiedColor=<rgba 0.6951 0.6951 0.6951 1>}
       cornerRadii=<4 4 4 4>
      }
      tintColor {colorUseRule=UseColor_Specified, specifiedColor=<rgba 0.4955 0.4955 0.4955 1>}
      uVRegion {bIsValid=false, max=<0 0>, min=<0 0>}
     }
     normalForeground {colorUseRule=UseColor_Specified, specifiedColor=<rgba 0 1 1 1>}
     normalPadding=<ltrb 0 0 0 0>
     pressed {
      bIsDynamicallyLoaded=false
      drawAs=NoDrawType
      imageType=NoImage
      mirroring=NoMirror
      resourceName=None
      resourceObject=null
      tiling=NoTile
      imageSize=<32 32>
      margin=<ltrb 0 0 0 0>
      outlineSettings {
       bUseBrushTransparency=true
       roundingType=FixedRadius
       width=1
       color {colorUseRule=UseColor_Specified, specifiedColor=<rgba 0.7243 0.7243 0.7243 1>}
       cornerRadii=<4 4 4 4>
      }
      tintColor {colorUseRule=UseColor_Specified, specifiedColor=<rgba 0.3843 0.3843 0.3843 1>}
      uVRegion {bIsValid=false, max=<0 0>, min=<0 0>}
     }
     pressedForeground {colorUseRule=UseColor_Specified, specifiedColor=<rgba 0 1 1 1>}
     pressedPadding=<ltrb 0 0 0 0>
     pressedSlateSound {resourceObject=null}
    }
   }
   slot {class=VerticalBoxSlot}
  }
  CommonBorder_Clear {
   widgetClassPath=&CommonBorder
   bIsVariable=true
   parent=Button_Clear
   widget=CommonBorder_Clear
   properties {
    background {
     bIsDynamicallyLoaded=false
     drawAs=Box
     imageType=NoImage
     mirroring=NoMirror
     resourceName=None
     resourceObject=null
     tiling=NoTile
     imageSize=<32 32>
     margin=<ltrb 0 0 0 0>
     outlineSettings {
      bUseBrushTransparency=false
      roundingType=HalfHeightRadius
      width=0
      color {colorUseRule=UseColor_Specified, specifiedColor=<rgba 0 0 0 0>}
      cornerRadii=<0 0 0 0>
     }
     tintColor {colorUseRule=UseColor_Specified, specifiedColor=<rgba 0 0.0021 0.0365 0.9>}
     uVRegion {bIsValid=false, max=<0 0>, min=<0 0>}
    }
    padding=<ltrb 10 20 10 20>
   }
   slot {class=ButtonSlot, properties {horizontalAlignment=HAlign_Fill, verticalAlignment=VAlign_Fill, padding=<ltrb 0 0 0 0>}}
  }
  HorizontalBox_0 {
   widgetClassPath=&HorizontalBox
   parent=CommonBorder_Clear
   widget=HorizontalBox_0
   properties {visibility=HitTestInvisible}
   slot {class=BorderSlot, properties {padding=<ltrb 10 20 10 20>}}
  }
  FortRichTextBlock_Clear1 {
   widgetClassPath=/Script/FortniteUI.FortRichTextBlockLegacy
   bIsVariable=true
   parent=HorizontalBox_0
   widget=FortRichTextBlock_Clear1
   properties {styleSet=/FortUILibrary/Foundation/Buttons/TextStyles/TextStyle_Gamepad_Bindings, text="<Font.Bold>Press </>"}
   slot {class=HorizontalBoxSlot, properties {verticalAlignment=VAlign_Center}}
  }
  KeybindWidget_Clear {
   widgetClassPath=/FortUILibrary/Input/KeybindWidget
   bIsVariable=true
   parent=HorizontalBox_0
   widget=KeybindWidget_Clear
   properties {boundKeyFallback=Gamepad_FaceButton_Top, inputTypeOverride=Gamepad}
   slot {class=HorizontalBoxSlot}
  }
  FortRichTextBlock_Clear2 {
   widgetClassPath=&FortRichTextBlockLegacy
   bIsVariable=true
   parent=HorizontalBox_0
   widget=FortRichTextBlock_Clear2
   properties {styleSet=&TextStyle_Gamepad_Bindings, text="<Font.Bold> To Clear</>"}
   slot {class=HorizontalBoxSlot, properties {verticalAlignment=VAlign_Center}}
  }
  SizeBox_1 {
   widgetClassPath=&SizeBox
   parent=VerticalBox_0
   widget=SizeBox_1
   properties {bOverride_HeightOverride=true, heightOverride=600}
   slot {class=VerticalBoxSlot}
  }
  CommonListView_GamepadActions {
   widgetClassPath=&CommonListView
   bIsVariable=true
   parent=SizeBox_1
   widget=CommonListView_GamepadActions
   properties {entryWidgetClass=&GamepadKeyTextButton_C}
   slot {class=SizeBoxSlot}
  }
 ]
}
