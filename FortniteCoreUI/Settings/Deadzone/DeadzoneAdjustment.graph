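# Settings widget that previews a controller thumbstick deadzone, using a slider over a configurable SettingType and two overlaid thumbstick images.
# On Construct it passes SettingType to the embedded OptionsMenuSetting, binds its slider's change event, refreshes it and applies the current value.
# Each change converts the slider value to a settings value and uses it as the tinted inner image's render scale.
# Custom navigation centers on the slider when moving Up.
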WidgetBlueprint /FortniteCoreUI/Settings/Deadzone/DeadzoneAdjustment
graphs [
 DoCustomNavigation_0 {nodes [
   n1 = DoCustomNavigation_0!() -> (then, Navigation:EUINavigation Enum) {is=title K2Node_FunctionEntry}
   > n2 = Switch on EUINavigation!(execute, Selection:EUINavigation Enum=@n1) -> (Left:Exec, Right:Exec, Up:Exec, Down:Exec, Next:Exec, Previous:Exec, Invalid:Exec) {is=title K2Node_SwitchEnum}
   n3 = CenterOnWidget(
    execute=@n2.Up
    self:Fort Options Menu Setting Object Reference=Get Deadzone_Adjustment_Slider() -> (Deadzone_Adjustment_Slider:Options Menu Setting Object Reference)
   )
   Return Node!(execute=[@n3.then, @n2.Down, @n2.Left, @n2.Next, @n2.Previous, @n2.Right], ReturnValue:Widget Object Reference)
  ]}
 EventGraph {nodes [
   n1 = OnValueChanged_Bind!() -> (OutputDelegate, then, Value:float=~) {is=title K2Node_CustomEvent}
   n2 = ConvertSliderValueToSettingsValue!(self:Options Menu Setting Object Reference=Get Deadzone_Adjustment_Slider(), SliderValue:double=@n1.Value) -> (SettingsValue:double)
   > SetRenderScale!(
    execute
    self:Widget Object Reference=Get Deadzone_Adjustment() -> (Deadzone_Adjustment:Image Object Reference)
    Scale:Vector 2D Structure=MakeVector2D(X=@n2, Y=@n2)
   ) -> (then)
   Event Construct()
   n3 = Get Deadzone_Adjustment_Slider()
   > Set SettingType!(
    execute
    SettingType:ESettingType Enum=Get SettingType() -> (SettingType:ESettingType Enum=None)
    self:Fort Options Menu Setting Object Reference=@n3
   ) -> (then, Output_Get:ESettingType Enum=None)
   n4 = Get Slider_Menu!(self:Fort Options Menu Setting Object Reference=@n3) -> (Slider_Menu:Analog Slider Object Reference)
   > Bind Event to On Value Changed(self:Slider Object Reference=@n4, Delegate=Create Event()) {is=title K2Node_AddDelegate}
   > Refresh(self:Options Menu Setting Object Reference=@n3)
   > OnValueChanged_Bind!(execute, self, Value:float=GetValue!(self:Slider Object Reference=@n4) -> (ReturnValue:float)) -> (then) {is=title K2Node_CallFunction}
  ]}
]
variables [SettingType {type=ESettingType, default=None}]
widgets {
 info {inheritedWidgetCount=0, parentClass=&UserWidget, rootWidgetClass=&Overlay, widgetCount=10}
 widgets [
  Overlay_0 {widgetClassPath=&Overlay, widget=Overlay_0}
  SizeBox_0 {
   widgetClassPath=&SizeBox
   parent=Overlay_0
   widget=SizeBox_0
   properties {bOverride_WidthOverride=true, widthOverride=400}
   slot {class=OverlaySlot, properties {horizontalAlignment=HAlign_Fill}}
  }
  VerticalBox_0 {widgetClassPath=&VerticalBox, parent=SizeBox_0, widget=VerticalBox_0, slot {class=SizeBoxSlot}}
  Overlay_2 {widgetClassPath=&Overlay, parent=VerticalBox_0, widget=Overlay_2, slot {class=VerticalBoxSlot}}
  Deadzone_Max {
   widgetClassPath=&Image
   bIsVariable=true
   parent=Overlay_2
   widget=Deadzone_Max
   properties {
    visibility=SelfHitTestInvisible
    brush {
     bIsDynamicallyLoaded=false
     drawAs=Image
     imageType=NoImage
     mirroring=NoMirror
     resourceName=None
     resourceObject=/Game/Athena/HUD/Art/Materials/MI_Hud_Thumbstick_BG_Inst
     tiling=NoTile
     imageSize=<300 300>
     margin=<ltrb 0 0 0 0>
     outlineSettings {
      bUseBrushTransparency=false
      roundingType=HalfHeightRadius
      width=0
      color {colorUseRule=UseColor_Specified, specifiedColor=<rgba 0 0 0 0>}
      cornerRadii=<0 0 0 0>
     }
     tintColor {colorUseRule=UseColor_Specified, specifiedColor=<rgba 1 1 1 1>}
     uVRegion {bIsValid=false, max=<0 0>, min=<0 0>}
    }
   }
   slot {class=OverlaySlot, properties {horizontalAlignment=HAlign_Center, verticalAlignment=VAlign_Center}}
  }
  Overlay_3 {
   widgetClassPath=&Overlay
   parent=Overlay_2
   widget=Overlay_3
   slot {class=OverlaySlot, properties {horizontalAlignment=HAlign_Fill, verticalAlignment=VAlign_Fill}}
  }
  Deadzone_Adjustment {
   widgetClassPath=&Image
   bIsVariable=true
   parent=Overlay_3
   widget=Deadzone_Adjustment
   properties {
    visibility=SelfHitTestInvisible
    brush {
     bIsDynamicallyLoaded=false
     drawAs=Image
     imageType=NoImage
     mirroring=NoMirror
     resourceName=None
     resourceObject=&MI_Hud_Thumbstick_BG_Inst
     tiling=NoTile
     imageSize=<300 300>
     margin=<ltrb 0 0 0 0>
     outlineSettings {
      bUseBrushTransparency=false
      roundingType=HalfHeightRadius
      width=0
      color {colorUseRule=UseColor_Specified, specifiedColor=<rgba 0 0 0 0>}
      cornerRadii=<0 0 0 0>
     }
     tintColor {colorUseRule=UseColor_Specified, specifiedColor=<rgba 0 0.1164 1 1>}
     uVRegion {bIsValid=false, max=<0 0>, min=<0 0>}
    }
   }
   slot {class=OverlaySlot, properties {horizontalAlignment=HAlign_Center, verticalAlignment=VAlign_Center}}
  }
  SizeBox_1 {
   widgetClassPath=&SizeBox
   parent=VerticalBox_0
   widget=SizeBox_1
   properties {bOverride_WidthOverride=true, widthOverride=300}
   slot {class=VerticalBoxSlot, properties {padding=<ltrb 0 5 0 0>, size {sizeRule=Fill, value=1}}}
  }
  Deadzone_Adjustment_Slider {
   widgetClassPath=/FortniteCoreUI/Settings/OptionsMenuSetting
   bIsVariable=true
   parent=SizeBox_1
   widget=Deadzone_Adjustment_Slider
   properties {
    hideLabelText=true
    hideSpaceRight=true
    maxSensitivityValue=0.95
    minSensitivityValue=0.05
    navigation=&self:WidgetTree.Deadzone_Adjustment_Slider.WidgetNavigation_1
   }
   slot {class=SizeBoxSlot}
  }
  ResetButton {
   widgetClassPath=&IconTextButton
   bIsVariable=true
   parent=VerticalBox_0
   widget=ResetButton
   properties {
    buttonText=Reset
    navigation=&self:WidgetTree.ResetButton.WidgetNavigation_1
    onButtonBaseClicked=(&self:WidgetTree.ResetButton.HandleOnButtonClicked)
    onButtonBaseDoubleClicked=(&self:WidgetTree.ResetButton.HandleOnButtonDoubleClicked)
    onButtonBaseHovered=(&self:WidgetTree.ResetButton.HandleOnButtonHovered)
    onButtonBaseUnhovered=(&self:WidgetTree.ResetButton.HandleOnButtonUnhovered)
    onSelectedChangedBase=(&self:WidgetTree.ResetButton.HandleOnSelectedChanged)
   }
   slot {class=VerticalBoxSlot, properties {horizontalAlignment=HAlign_Center, verticalAlignment=VAlign_Center}}
  }
 ]
}
