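# Settings widget that shows a controller diagram labeled with gamepad actions for preset configs, plus custom binding buttons and dead zone adjustment.
# Config-changed handlers fill labels for STW and Athena per mode via index Selects, text arrays or input-data labels; the STW path also hides unused extra lines.
# Custom buttons resolve action and label from the custom preset, and clicks broadcast CustomKeySelected.
# The controller image async-loads from a platform image list.
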WidgetBlueprint /FortniteCoreUI/Settings/CustomBindings_v3/GamepadMappingInfo_v3
graphs [
 ClearAllComboTextLines {nodes [
   ClearAllComboTextLines() {is=title K2Node_FunctionEntry}
   > "SetText (Text)"(
    self=[
     Get LB+RBAction() -> (LB+RBAction:Common Text Object Reference)
     Get LB+RTAction() -> (LB+RTAction:Common Text Object Reference)
     Get LT+RBAction() -> (LT+RBAction:Common Text Object Reference)
    ]
    InText=MakeLiteralText()
   )
  ]}
 CustomButtonNavLeft {nodes [CustomButtonNavLeft() {is=title K2Node_FunctionEntry}]}
 CustomKeySelected {nodes [CustomKeySelected!() -> (then, Key:Key Structure=None) {is=title K2Node_FunctionEntry}]}
 DebugButtonInfo {nodes [
   n1 = Inputs!() -> (execute, self:Gamepad Key Text Button Object Reference, PrintToScreen:bool=~, PrintToLog:bool=~)
   n2 = Get BoundKey!(self:Fort Keybind Widget Object Reference=Get BoundKey(self:Gamepad Key Text Button Object Reference=@n1) -> (BoundKey:Keybind Widget Object Reference)) -> (BoundKey:Key Structure)
   > n3 = GetCustomButtonDataForKey!(execute, self, Key:Key Structure=@n2) -> (then, ActionName:Name=None, DescText:Text)
   n4 = Get Key Display Name!(Key:"Key Structure (by ref)"=@n2, bLongDisplayName:bool=true) -> (ReturnValue:Text) {is=title K2Node_CallFunction}
   n5 = Get ContentAlignment!(self:Gamepad Key Text Button Object Reference=@n1) -> (ContentAlignment:EHorizontalAlignment Enum=HAlign_Fill)
   > PrintString!(
    execute
    InString=Append!(
     A="To String (Text)"(InText=@n4)
     B=" - "
     C:String="To String (Name)"(InName=@n3)
     D:String=" - "
     E:String="To String (Text)"(InText=@n3.DescText)
     F:String=" - "
     G:String=Enum to String!(Enumerator:EHorizontalAlignment Enum=@n5) -> (ReturnValue:String)
    ) -> (ReturnValue)
    bPrintToScreen=@n1.PrintToScreen
    bPrintToLog=@n1.PrintToLog
    TextColor
    Duration
    Key
   ) -> (then)
   > Outputs!(Out:Exec)
  ]}
 EventGraph {nodes [
   Update Controller Images() {is=title K2Node_CustomEvent}
   n1 = GetControllerImage!() -> (Texture:Texture 2D Soft Object Reference)
   > n2 = Branch(Condition=IsAssetNull!(Asset:Soft Object Reference=@n1) -> (ReturnValue:bool))
   n3 = Get Image_Gamepad() -> (Image_Gamepad:Image Object Reference)
   SetVisibility(execute=[@n6.Is Not Valid, @n5.CastFailed, @n4.then, @n2.then], self=@n3, InVisibility=Hidden)
   n4 = Async Load Asset(execute=@n2.else, Asset=@n1)
   n5 = Cast To Texture2D(execute=@n4.Completed, Object=@n4)
   > n6 = Is Valid(InputObject=@n5)
   > SetVisibility(self=@n3, InVisibility=HitTestInvisible)
   > SetBrushFromTexture!(execute, self:Image Object Reference=@n3, Texture:Texture 2D Object Reference=@n5, bMatchSize:bool=true) -> (then)
   n7 = OnButtonClicked_Bind!() -> (OutputDelegate, then, Button:Common Button Legacy Object Reference) {is=title K2Node_CustomEvent}
   > n8 = Cast To GamepadKeyTextButton!(execute, Object=@n7.Button) -> (then, CastFailed, AsGamepad Key Text Button:Gamepad Key Text Button Object Reference) {is=title K2Node_DynamicCast}
   > n9 = Set LastClickedButton(LastClickedButton:Gamepad Key Text Button Object Reference=@n8) -> (Output_Get:Gamepad Key Text Button Object Reference)
   > Debug Button Info!(execute:Exec, self:Gamepad Key Text Button Object Reference=@n9, PrintToScreen:bool, PrintToLog:bool) -> (Out:Exec) {is=title K2Node_MacroInstance}
   > Call Custom Key Selected!(
    execute:Exec
    self:Gamepad Mapping Info V 3 Object Reference
    Key:Key Structure=Get BoundKey(self:Fort Keybind Widget Object Reference=Get BoundKey(self:Gamepad Key Text Button Object Reference=@n9) -> (BoundKey:Keybind Widget Object Reference)) -> (BoundKey:Key Structure)
   ) -> (then:Exec) {is=title K2Node_CallDelegate}
   UpdateCustomButtons() {is=title K2Node_CustomEvent}
   > n10 = For Each Loop(Array:Array of Gamepad Key Text Button Object References=Get CustomKeyButtons() -> (CustomKeyButtons:Array of Gamepad Key Text Button Object References)) -> (Array Element:Gamepad Key Text Button Object Reference)
   > Debug Button Info(self=@n10)
   n11 = Get KeyData!(self:Gamepad Key Text Button Object Reference=@n10) -> (KeyData:Key Structure)
   > n12 = GetCustomButtonDataForKey(Key=@n11) -> (ActionName=None)
   > SetDataAndUpdateButton!(execute, self:Gamepad Key Text Button Object Reference=@n10, Key:Key Structure=@n11, DescText:Text=@n12.DescText, ActionName:Name=@n12) -> (then)
   DeadZoneTabSelected() {is=title K2Node_CustomEvent}
   > SetVisibility(self=Get Overlay_DeadzoneAdjustment() -> (Overlay_DeadzoneAdjustment:Overlay Object Reference), InVisibility=Visible)
   > SetVisibility(
    self=[
     Get HorizontalBox_CustomButtons() -> (HorizontalBox_CustomButtons:Horizontal Box Object Reference)
     Get CommonBorder_Legend() -> (CommonBorder_Legend:Common Border Object Reference)
     Get HorizontalBox_AllText() -> (HorizontalBox_AllText:Horizontal Box Object Reference)
    ]
    InVisibility=Collapsed
   )
   > "SetText (Text)"(self=Get Text_Mode() -> (Text_Mode:Common Text Object Reference), InText="LOCTEXT(\"Dead Zone\")")
   > CenterOnWidget(self:Fort Options Menu Setting Object Reference=Get Deadzone_Adjustment_Slider!(self:Deadzone Adjustment Object Reference=Get DeadzoneAdjustment_Move() -> (DeadzoneAdjustment_Move:Deadzone Adjustment Object Reference)) -> (Deadzone_Adjustment_Slider:Options Menu Setting Object Reference))
   DeadZoneTabDeselected() {is=title K2Node_CustomEvent}
   > SetVisibility(self=Get Overlay_DeadzoneAdjustment(), InVisibility=Collapsed)
   DeadzoneMoveResetButtonClicked_Bind!() -> (OutputDelegate, then, Button:Common Button Legacy Object Reference) {is=title K2Node_CustomEvent}
   > ResetGamepadMoveStickDeadZone(self:Fort Settings Context Object Reference=Get FortSettingsContext!() -> (ReturnValue:Fort Settings Context Object Reference) {is=title K2Node_GetBlueprintContextEx})
   > Refresh(self:Options Menu Setting Object Reference=Get Deadzone_Adjustment_Slider(self=Get DeadzoneAdjustment_Move()))
   n13 = Get DeadzoneAdjustment_Move()
   n14 = Get Deadzone_Adjustment_Slider(self=@n13)
   > OnValueChanged_Bind!(
    execute
    self:Deadzone Adjustment Object Reference=@n13
    Value:float=GetValue!(self:Slider Object Reference=Get Slider_Menu!(self:Fort Options Menu Setting Object Reference=@n14) -> (Slider_Menu:Analog Slider Object Reference)) -> (ReturnValue:float)
   ) -> (then)
   DeadzoneLookResetButtonClicked_Bind!() -> (OutputDelegate, then, Button:Common Button Legacy Object Reference) {is=title K2Node_CustomEvent}
   > ResetGamepadLookStickDeadZone(self:Fort Settings Context Object Reference=Get FortSettingsContext())
   > Refresh(self=Get Deadzone_Adjustment_Slider(self=Get DeadzoneAdjustment_Look() -> (DeadzoneAdjustment_Look:Deadzone Adjustment Object Reference)))
   n15 = Get DeadzoneAdjustment_Look()
   n16 = Get Deadzone_Adjustment_Slider(self=@n15)
   > OnValueChanged_Bind(self=@n15, Value=GetValue(self=Get Slider_Menu(self=@n16)))
   Event Construct()
   > n17 = Sequence!(execute) -> (then_0, then_1, then_2:Exec, then_3:Exec, then_4:Exec)
   > InitializeListOfButtonLabelTextWidgets()
   > InitializeListOfButtonTextForSTWCombatProBuild!(
    execute
    self
    ButtonTextLabels:Array of Texts=Get ButtonTextLabels_STW_CombatPro_Build() -> (ButtonTextLabels_STW_CombatPro_Build:Array of Texts)
   ) -> (then)
   > InitializeListOfButtonTextForSTWCombatProCombat!(
    execute
    self
    ButtonTextLabels:Array of Texts=Get ButtonTextLabels_STW_CombatPro_Combat() -> (ButtonTextLabels_STW_CombatPro_Combat:Array of Texts)
   ) -> (then)
   > InitializeListOfButtonTextForSTW_BuilderProBuild!(
    execute
    self
    ButtonTextLabels:Array of Texts=Get ButtonTextLabels_STW_BuilderPro_Build() -> (ButtonTextLabels_STW_BuilderPro_Build:Array of Texts)
   ) -> (then)
   n18 = Set CustomKeyButtons(
    execute=@n17.then_1
    CustomKeyButtons:Array of Gamepad Key Text Button Object References=Make Array!(
     "[0]":Gamepad Key Text Button Object Reference=Get LT_Custom() -> (LT_Custom:Gamepad Key Text Button Object Reference)
     "[1]":Gamepad Key Text Button Object Reference=Get LB_Custom() -> (LB_Custom:Gamepad Key Text Button Object Reference)
     "[2]":Gamepad Key Text Button Object Reference=Get DpadUp_Custom() -> (DpadUp_Custom:Gamepad Key Text Button Object Reference)
     "[3]":Gamepad Key Text Button Object Reference=Get DpadLeft_Custom() -> (DpadLeft_Custom:Gamepad Key Text Button Object Reference)
     "[4]":Gamepad Key Text Button Object Reference=Get DpadRight_Custom() -> (DpadRight_Custom:Gamepad Key Text Button Object Reference)
     "[5]":Gamepad Key Text Button Object Reference=Get DpadDown_Custom() -> (DpadDown_Custom:Gamepad Key Text Button Object Reference)
     "[6]":Gamepad Key Text Button Object Reference=Get LS_Click_Custom() -> (LS_Click_Custom:Gamepad Key Text Button Object Reference)
     "[7]":Gamepad Key Text Button Object Reference=Get RT_Custom() -> (RT_Custom:Gamepad Key Text Button Object Reference)
     "[8]":Gamepad Key Text Button Object Reference=Get RB_Custom() -> (RB_Custom:Gamepad Key Text Button Object Reference)
     "[9]":Gamepad Key Text Button Object Reference=Get FaceTop_Custom() -> (FaceTop_Custom:Gamepad Key Text Button Object Reference)
     "[10]":Gamepad Key Text Button Object Reference=Get FaceLeft_Custom() -> (FaceLeft_Custom:Gamepad Key Text Button Object Reference)
     "[11]":Gamepad Key Text Button Object Reference=Get FaceRight_Custom() -> (FaceRight_Custom:Gamepad Key Text Button Object Reference)
     "[12]":Gamepad Key Text Button Object Reference=Get FaceBottom_Custom() -> (FaceBottom_Custom:Gamepad Key Text Button Object Reference)
     "[13]":Gamepad Key Text Button Object Reference=Get RS_Click_Custom() -> (RS_Click_Custom:Gamepad Key Text Button Object Reference)
     "[14]":Gamepad Key Text Button Object Reference=Get Select_Custom() -> (Select_Custom:Gamepad Key Text Button Object Reference)
     "[15]":Gamepad Key Text Button Object Reference=Get Start_Custom() -> (Start_Custom:Gamepad Key Text Button Object Reference)
    ) -> (Array:Array of Gamepad Key Text Button Object References)
   ) -> (Output_Get:Array of Gamepad Key Text Button Object References)
   > n19 = For Each Loop(Array=@n18)
   > Bind Event to On Button Clicked(self:Common Button Legacy Object Reference=@n19, Delegate=Create Event()) {is=title K2Node_AddDelegate}
   n20 = Get DeadzoneAdjustment_Move()
   Bind Event to On Button Clicked(
    execute=@n17.then_4
    self=Get ResetButton!(self:Deadzone Adjustment Object Reference=@n20) -> (ResetButton:Icon Text Button Object Reference)
    Delegate=Create Event()
   )
   > Set HoverText!(
    execute
    HoverText:Text="LOCTEXT(\"Adjust the dead zone for movement.\")"
    self:Fort Options Menu Setting Object Reference=Get Deadzone_Adjustment_Slider(self=@n20)
   ) -> (then, Output_Get:Text)
   n21 = Get DeadzoneAdjustment_Look()
   > Bind Event to On Button Clicked(self=Get ResetButton(self=@n21), Delegate=Create Event())
   > Set HoverText(HoverText="LOCTEXT(\"Adjust the dead zone for look.\")", self=Get Deadzone_Adjustment_Slider(self=@n21))
   n22 = Get HorizontalTabList_CustomModes() -> (HorizontalTabList_CustomModes:Horizontal Tab List Legacy Object Reference)
   RegisterAllPreregisteredTabInfos(execute=@n17.then_2, self:Fort Tab List Widget Base Legacy Object Reference=@n22)
   > Branch(Condition="integer > integer"(A=GetTabCount!(self:Common Tab List Widget Base Object Reference=@n22) -> (ReturnValue:int=0), B=0) {is=title K2Node_CallFunction})
   > Set CombatModeIdName(CombatModeIdName:Name=GetTabIdAtIndex!(self:Common Tab List Widget Base Object Reference=@n22, Index:int=0) -> (ReturnValue:Name=None)) -> (Output_Get:Name=None)
   > Set BuildModeIdName(BuildModeIdName:Name=GetTabIdAtIndex(self=@n22, Index=1) -> (ReturnValue=None)) -> (Output_Get:Name=None)
   > Set EditModeIdName(EditModeIdName:Name=GetTabIdAtIndex(self=@n22, Index=2) -> (ReturnValue=None)) -> (Output_Get:Name=None)
   > Set LTMAbilitiesModeIdName(LTMAbilitiesModeIdName:Name=GetTabIdAtIndex(self=@n22, Index=3) -> (ReturnValue=None)) -> (Output_Get:Name=None)
   > Set DeadZoneMoveIdName(DeadZoneMoveIdName:Name=GetTabIdAtIndex(self=@n22, Index=4) -> (ReturnValue=None)) -> (Output_Get:Name=None)
   SetNonInteractableStyle(execute=@n17.then_3, self:Gamepad Key Text Button Object Reference=Get Start_Custom())
  ]}
 GamepadMappingInfoNavLeft {nodes [
   GamepadMappingInfoNavLeft() {is=title K2Node_FunctionEntry}
   > Call Custom Button Nav Left!(execute:Exec, self:Gamepad Mapping Info V 3 Object Reference) -> (then:Exec) {is=title K2Node_CallDelegate}
  ]}
 GetControllerImage {nodes [
   GetControllerImage() {is=title K2Node_FunctionEntry}
   n1 = Literal enum GamepadConfigImages!(Enum:GamepadConfigImages Enum=NewEnumerator0) -> (ReturnValue:GamepadConfigImages Enum) {is=title K2Node_EnumLiteral}
   > Return Node!(
    execute
    Texture:Texture 2D Soft Object Reference="Get (a copy)"(
     Array:Array of Texture 2D Soft Object References=Get Appropriate Platform Images() -> (Appropriate Platform Images:Array of Texture 2D Soft Object References)
     Dimension 1="To Integer (Byte)"!(InByte:Byte=@n1) -> (ReturnValue:int=0)
    ) -> (Output:Texture 2D Soft Object Reference)
   )
  ]}
 GetCustomButtonDataForKey {nodes [
   n1 = GetCustomButtonDataForKey!() -> (then, Key:Key Structure=None) {is=title K2Node_FunctionEntry}
   > Set Key_Local(Key_Local:Key Structure=@n1) -> (Output_Get:Key Structure)
   n2 = Get FortGlobalUIContext!() -> (ReturnValue:Fort Global UIContext Object Reference) {is=title K2Node_GetBlueprintContextEx}
   > Set PresetName_Local(PresetName_Local:Name=String To Name(InString=GetCustomGamepadInputPresetName!(self:Fort Global UIContext Object Reference=@n2) -> (ReturnValue:String))) -> (Output_Get:Name=None)
   > n3 = Sequence!(execute) -> (then_0, then_1)
   > n4 = Set ActionName_Local(ActionName_Local:Name=GetActionForKey!(
     self:Fort Global UIContext Object Reference=Get FortGlobalUIContext()
     PresetName:Name=Get PresetName_Local() -> (PresetName_Local:Name=None)
     Key:Key Structure=Get Key_Local() -> (Key_Local:Key Structure)
     InputActionGroup:EFortInputActionGroup Enum=Get CurrentInputActionGroup() -> (CurrentInputActionGroup:EFortInputActionGroup Enum=AllModes)
     bIsUsingGamepad:bool=true
    ) -> (ReturnValue:Name=None)) -> (Output_Get:Name=None)
   n5 = GetInputActionGroup!(self:Fort Global UIContext Object Reference=Get FortGlobalUIContext(), ActionName:"Name (by ref)"=@n4) -> (OutInputActionGroup:EFortInputActionGroup Enum=AllModes, ReturnValue:bool)
   > Set InputActionGroup_Local(InputActionGroup_Local:EFortInputActionGroup Enum=Select(Option 0:EFortInputActionGroup Enum=AllModes, Option 1:EFortInputActionGroup Enum=@n5, Index=@n5.ReturnValue) -> (ReturnValue:EFortInputActionGroup Enum)) -> (Output_Get:EFortInputActionGroup Enum=AllModes)
   Set DescText_Local(
    execute=@n3.then_1
    DescText_Local:Text=GetScreenLabel!(
     self:Fort Input Data Object Reference=GetFortInputDataForBlueprint!() -> (ReturnValue:Fort Input Data Object Reference)
     PresetName:Name=Get PresetName_Local() -> (PresetName_Local=None)
     InputActionGroup:EFortInputActionGroup Enum=Get InputActionGroup_Local() -> (InputActionGroup_Local:EFortInputActionGroup Enum=AllModes)
     ActionName:Name=Get ActionName_Local() -> (ActionName_Local:Name=None)
     bRichText:bool=true
    ) -> (ReturnValue:Text)
   ) -> (Output_Get:Text)
   > Return Node!(execute, ActionName:Name=Get ActionName_Local() -> (ActionName_Local=None), DescText:Text=Get DescText_Local() -> (DescText_Local:Text))
  ]}
 GetDesiredSlateVisibilityForAbilityChordFromTextDisplay {nodes [
   n1 = GetDesiredSlateVisibilityForAbilityChordFromTextDisplay!() -> (then, TextWidget:Text Object Reference) {is=title K2Node_FunctionEntry}
   n2 = "GetText (Text)"!(self:Text Object Reference=@n1) -> (ReturnValue:Text)
   n3 = MakeLiteralText()
   > Return Node!(
    execute
    DesiredSlateVisibility:ESlateVisibility Enum=Select(
     Option 0:ESlateVisibility Enum=HitTestInvisible
     Option 1:ESlateVisibility Enum=Hidden
     Index="Equal Exactly (Text)"!(A:"Text (by ref)"=@n2, B:"Text (by ref)"=@n3) -> (ReturnValue:bool)
    ) -> (ReturnValue:ESlateVisibility Enum)
   )
  ]}
 Handle Config Changed Hide or Show Unused Text Lines {nodes [
   Handle Config Changed Hide or Show Unused Text Lines() {is=title K2Node_FunctionEntry}
   > SetTextFieldVisibilityForExtraLine!(
    execute
    self
    TextWidget:Text Object Reference=Get Text_Gamepad_Face_Button_Top_2() -> (Text_Gamepad_Face_Button_Top_2:Common Text Object Reference)
   ) -> (then)
   > SetTextFieldVisibilityForExtraLine(TextWidget=Get Text_Gamepad_Face_Button_Left_2() -> (Text_Gamepad_Face_Button_Left_2:Common Text Object Reference))
   > SetTextFieldVisibilityForExtraLine(TextWidget=Get Text_Gamepad_Face_Button_Left_3() -> (Text_Gamepad_Face_Button_Left_3:Common Text Object Reference))
   > SetTextFieldVisibilityForExtraLine(TextWidget=Get Text_Gamepad_Face_Button_Right_2() -> (Text_Gamepad_Face_Button_Right_2:Common Text Object Reference))
   > SetTextFieldVisibilityForExtraLine(TextWidget=Get Text_Gamepad_Face_Button_Right_3() -> (Text_Gamepad_Face_Button_Right_3:Common Text Object Reference))
   > SetTextFieldVisibilityForExtraLine(TextWidget=Get Text_Gamepad_RS_Click_2() -> (Text_Gamepad_RS_Click_2:Common Text Object Reference))
   > SetTextFieldVisibilityForExtraLine(TextWidget=Get Text_PS4_LS_Click_2() -> (Text_PS4_LS_Click_2:Common Text Object Reference))
   > SetTextFieldVisibilityForExtraLine(TextWidget=Get Text_Xbox_LS_Click_2() -> (Text_Xbox_LS_Click_2:Common Text Object Reference))
   > SetTextFieldVisibilityForExtraLine(TextWidget=Get Text_Gamepad_Select_2() -> (Text_Gamepad_Select_2:Common Text Object Reference))
   > SetTextFieldVisibilityForExtraLine(TextWidget=Get Text_Gamepad_RS_Click_3() -> (Text_Gamepad_RS_Click_3:Common Text Object Reference))
   n1 = GetDesiredSlateVisibilityForAbilityChordFromTextDisplay!(TextWidget:Text Object Reference=Get LB+RBAction()) -> (DesiredSlateVisibility:ESlateVisibility Enum=Visible) {is=title K2Node_CallFunction}
   > SetVisibility(
    self=[
     Get HorizontalBox-LB+RB() -> (HorizontalBox-LB+RB:Horizontal Box Object Reference)
     Get CommonBorder-LeftBumperSeparator() -> (CommonBorder-LeftBumperSeparator:Common Border Object Reference)
    ]
    InVisibility=@n1
   )
   n2 = GetDesiredSlateVisibilityForAbilityChordFromTextDisplay(TextWidget=Get LB+RTAction()) -> (DesiredSlateVisibility=Visible)
   > SetVisibility(self=Get HorizontalBox-LB+RT() -> (HorizontalBox-LB+RT:Horizontal Box Object Reference), InVisibility=@n2)
   n3 = "Equal (Enum)"(A:ESlateVisibility Enum=@n1, B:ESlateVisibility Enum=Hidden)
   n4 = "Equal (Enum)"(A=@n2, B=Hidden)
   > SetVisibility(
    self=Get Separator-LB+RBfromLB+RT() -> (Separator-LB+RBfromLB+RT:Size Box Object Reference)
    InVisibility=Select(Option 0=SelfHitTestInvisible, Option 1=Hidden, Index=OR Boolean(A=@n4, B=@n3))
   )
   > SetVisibility(
    self=[
     Get HorizontalBox-LT+RB() -> (HorizontalBox-LT+RB:Horizontal Box Object Reference)
     Get CommonBorder-RightBumperSeparator() -> (CommonBorder-RightBumperSeparator:Common Border Object Reference)
    ]
    InVisibility=GetDesiredSlateVisibilityForAbilityChordFromTextDisplay(TextWidget=Get LT+RBAction()) -> (DesiredSlateVisibility=Visible)
   )
  ]}
 Handle Config Changed STW Build Mode {nodes [
   n1 = Handle Config Changed STW Build Mode!() -> (then, Config:int) {is=title K2Node_FunctionEntry}
   > ClearAllComboTextLines() {is=title K2Node_CallFunction}
   > n2 = Switch on Int!(execute:Exec, Selection:int=@n1) -> (Default:Exec, 0:Exec, 1:Exec, 2:Exec, 3:Exec)
   SetButtonTextLabelsFromTextArray!(execute=@n2.2, self, Array:Array of Texts=Get ButtonTextLabels_STW_CombatPro_Build()) -> (then)
   SetButtonTextLabelsFromTextArray(execute=@n2.3, Array=Get ButtonTextLabels_STW_BuilderPro_Build())
   "SetText (Text)"(
    execute=[@n2.0, @n2.1]
    self=Get Text_Gamepad_LT() -> (Text_Gamepad_LT:Common Text Object Reference)
    InText=Select(Option 0:Text="LOCTEXT(\"Equip Targeted Building Piece\")", Option 1:Text="LOCTEXT(\"Equip Targeted Building Piece\")", Index:int=@n1) -> (ReturnValue:Text)
   )
   > "SetText (Text)"(
    self=Get Text_Gamepad_LB() -> (Text_Gamepad_LB:Common Text Object Reference)
    InText=Select(Option 0="LOCTEXT(\"Change Building Material\")", Option 1="LOCTEXT(\"Previous Building Piece\")", Index=@n1)
   )
   > "SetText (Text)"(
    self=[Get Text_Xbox_LS() -> (Text_Xbox_LS:Common Text Object Reference), Get Text_PS4_LS() -> (Text_PS4_LS:Common Text Object Reference)]
    InText=Select(Option 0="LOCTEXT(\"Move\")", Option 1="LOCTEXT(\"Move\")", Index=@n1)
   )
   > "SetText (Text)"(
    self=[
     Get Text_Xbox_LS_Click_1() -> (Text_Xbox_LS_Click_1:Common Text Object Reference)
     Get Text_PS4_LS_Click_1() -> (Text_PS4_LS_Click_1:Common Text Object Reference)
    ]
    InText=Select(Option 0="LOCTEXT(\"Sprint\")", Option 1="LOCTEXT(\"Change Building Material\")", Index=@n1)
   )
   > "SetText (Text)"(
    self=[Get Text_PS4_LS_Click_2(), Get Text_Xbox_LS_Click_2()]
    InText=Select(Option 0="LOCTEXT(\"Auto-Sprint (Double-Click)\")", Option 1="LOCTEXT(\"Auto-Sprint (Double-Click)\")", Index=@n1)
   )
   > "SetText (Text)"(
    self=[
     Get Text_Xbox_Dpad_Up() -> (Text_Xbox_Dpad_Up:Common Text Object Reference)
     Get Text_PS4_Dpad_Up() -> (Text_PS4_Dpad_Up:Common Text Object Reference)
    ]
    InText=Select(Option 0="LOCTEXT(\"Toggle Map\")", Option 1="LOCTEXT(\"Toggle Map\")", Index=@n1)
   )
   > "SetText (Text)"(
    self=[
     Get Text_Xbox_Dpad_Down() -> (Text_Xbox_Dpad_Down:Common Text Object Reference)
     Get Text_PS4_Dpad_Down() -> (Text_PS4_Dpad_Down:Common Text Object Reference)
    ]
    InText=Select(Option 0="LOCTEXT(\"Squad Comms / Emote / Spot Enemy (Tap)\")", Option 1="LOCTEXT(\"Squad Comms / Emote / Spot Enemy (Tap)\")", Index=@n1)
   )
   > "SetText (Text)"(
    self=[
     Get Text_Xbox_Dpad_Left() -> (Text_Xbox_Dpad_Left:Common Text Object Reference)
     Get Text_PS4_Dpad_Left() -> (Text_PS4_Dpad_Left:Common Text Object Reference)
    ]
    InText=Select(Option 0="LOCTEXT(\"-\")", Option 1="LOCTEXT(\"-\")", Index=@n1)
   )
   > "SetText (Text)"(
    self=[
     Get Text_Xbox_Dpad_Right() -> (Text_Xbox_Dpad_Right:Common Text Object Reference)
     Get Text_PS4_Dpad_Right() -> (Text_PS4_Dpad_Right:Common Text Object Reference)
    ]
    InText=Select(Option 0="LOCTEXT(\"-\")", Option 1="LOCTEXT(\"-\")", Index=@n1)
   )
   > "SetText (Text)"(
    self=Get Text_Gamepad_Select_1() -> (Text_Gamepad_Select_1:Common Text Object Reference)
    InText=Select(Option 0="LOCTEXT(\"Inventory (Management, etc.)\")", Option 1="LOCTEXT(\"Inventory (Management, etc.)\")", Index=@n1)
   )
   > "SetText (Text)"(self=Get Text_Gamepad_Select_2(), InText=Select(Option 0="LOCTEXT(\"Chat (Hold)\")", Option 1="LOCTEXT(\"Chat (Hold)\")", Index=@n1))
   > "SetText (Text)"(
    self=Get Text_Gamepad_Start() -> (Text_Gamepad_Start:Common Text Object Reference)
    InText=Select(Option 0="LOCTEXT(\"Game Menu (Friends, etc.)\")", Option 1="LOCTEXT(\"Game Menu (Friends, etc.)\")", Index=@n1)
   )
   > "SetText (Text)"(
    self=Get Text_Gamepad_RT() -> (Text_Gamepad_RT:Common Text Object Reference)
    InText=Select(Option 0="LOCTEXT(\"Place Building Piece / Confirm\")", Option 1="LOCTEXT(\"Place Building Piece / Confirm\")", Index=@n1)
   )
   > "SetText (Text)"(
    self=Get Text_Gamepad_RB() -> (Text_Gamepad_RB:Common Text Object Reference)
    InText=Select(Option 0="LOCTEXT(\"Rotate Building Piece\")", Option 1="LOCTEXT(\"Next Building Piece\")", Index=@n1)
   )
   > "SetText (Text)"(
    self=Get Text_Gamepad_Face_Button_Top_1() -> (Text_Gamepad_Face_Button_Top_1:Common Text Object Reference)
    InText=Select(Option 0="LOCTEXT(\"Cycle Building Piece\")", Option 1="LOCTEXT(\"Pick Trap\")", Index=@n1)
   )
   > "SetText (Text)"(self=Get Text_Gamepad_Face_Button_Top_2(), InText=Select(Index=@n1))
   > "SetText (Text)"(
    self=Get Text_Gamepad_Face_Button_Left_1() -> (Text_Gamepad_Face_Button_Left_1:Common Text Object Reference)
    InText=Select(Option 0="LOCTEXT(\"Interact (Tap / Hold)\")", Option 1="LOCTEXT(\"Interact (Tap / Hold)\")", Index=@n1)
   )
   > "SetText (Text)"(self=Get Text_Gamepad_Face_Button_Left_2(), InText=Select(Option 0="LOCTEXT(\"Pick Trap (Tap)\")", Index=@n1))
   > "SetText (Text)"(self=Get Text_Gamepad_Face_Button_Left_3(), InText=Select(Option 0="LOCTEXT(\"Trap Inventory (Hold)\")", Index=@n1))
   > "SetText (Text)"(
    self=Get Text_Gamepad_Face_Button_Right_1() -> (Text_Gamepad_Face_Button_Right_1:Common Text Object Reference)
    InText=Select(Option 0="LOCTEXT(\"Toggle Combat Mode (Tap)\")", Option 1="LOCTEXT(\"Toggle Combat Mode (Tap)\")", Index=@n1)
   )
   > "SetText (Text)"(
    self=Get Text_Gamepad_Face_Button_Right_2()
    InText=Select(Option 0="LOCTEXT(\"Edit Building Piece (Hold)\")", Option 1="LOCTEXT(\"Edit Building Piece (Hold)\")", Index=@n1)
   )
   > "SetText (Text)"(self=Get Text_Gamepad_Face_Button_Right_3())
   > "SetText (Text)"(
    self=Get Text_Gamepad_Face_Button_Bottom() -> (Text_Gamepad_Face_Button_Bottom:Common Text Object Reference)
    InText=Select(Option 0="LOCTEXT(\"Jump\")", Option 1="LOCTEXT(\"Jump\")", Index=@n1)
   )
   > "SetText (Text)"(
    self=Get Text_Gamepad_RS() -> (Text_Gamepad_RS:Common Text Object Reference)
    InText=Select(Option 0="LOCTEXT(\"Look\")", Option 1="LOCTEXT(\"Look\")", Index=@n1)
   )
   > "SetText (Text)"(
    self=Get Text_Gamepad_RS_Click_1() -> (Text_Gamepad_RS_Click_1:Common Text Object Reference)
    InText=Select(Option 0="LOCTEXT(\"Repair (Hold) / Improve (Hold)\")", Option 1="LOCTEXT(\"Repair (Hold) / Improve (Hold) \")", Index=@n1)
   )
   > "SetText (Text)"(
    self=Get Text_Gamepad_RS_Click_2()
    InText=Select(Option 0="LOCTEXT(\"Reset Building Edit (Edit Mode)\")", Option 1="LOCTEXT(\"Reset Building Edit (Edit Mode)\")", Index=@n1)
   )
   > "SetText (Text)"(self=Get Text_Gamepad_RS_Click_3(), InText=Select(Option 1="LOCTEXT(\"Rotate Building Piece (Tap)\")", Index=@n1))
  ]}
 Handle Config Changed STW Combat Mode {nodes [
   n1 = Handle Config Changed STW Combat Mode!() -> (then, Config:int) {is=title K2Node_FunctionEntry}
   > n2 = Switch on Int(Selection=@n1)
   "SetText (Text)"(
    execute=[@n2.0, @n2.1]
    self=Get Text_Gamepad_LT()
    InText=Select(Option 0="LOCTEXT(\"Aim Down Sights / Heavy Melee\")", Option 1="LOCTEXT(\"Aim Down Sights / Heavy Melee\")", Index=@n1)
   )
   > "SetText (Text)"(
    self=Get Text_Gamepad_LB()
    InText=GetAbility1Label!(self:Fort Input Data Object Reference=GetFortInputDataForBlueprint()) -> (ReturnValue:Text)
   )
   > "SetText (Text)"(self=Get LB+RBAction(), InText=GetAbility3Label!(self:Fort Input Data Object Reference=GetFortInputDataForBlueprint()) -> (ReturnValue:Text))
   > "SetText (Text)"(self=Get LB+RTAction(), InText=MakeLiteralText())
   > "SetText (Text)"(self=[Get Text_Xbox_LS(), Get Text_PS4_LS()], InText=Select(Option 0="LOCTEXT(\"Move\")", Option 1="LOCTEXT(\"Move\")", Index=@n1))
   > "SetText (Text)"(
    self=[Get Text_Xbox_LS_Click_1(), Get Text_PS4_LS_Click_1()]
    InText=Select(Option 0="LOCTEXT(\"Sprint\")", Option 1="LOCTEXT(\"Sprint\")", Index=@n1)
   )
   > "SetText (Text)"(
    self=[Get Text_PS4_LS_Click_2(), Get Text_Xbox_LS_Click_2()]
    InText=Select(Option 0="LOCTEXT(\"Auto-Sprint (Double-Click)\")", Option 1="LOCTEXT(\"Auto-Sprint (Double-Click)\")", Index=0)
   )
   > "SetText (Text)"(
    self=[Get Text_PS4_Dpad_Up(), Get Text_Xbox_Dpad_Up()]
    InText=Select(Option 0="LOCTEXT(\"Toggle Map\")", Option 1="LOCTEXT(\"Toggle Map\")", Index=@n1)
   )
   > "SetText (Text)"(
    self=[Get Text_PS4_Dpad_Down(), Get Text_Xbox_Dpad_Down()]
    InText=Select(Option 0="LOCTEXT(\"Squad Comms / Emote / Spot Enemy (Tap)\")", Option 1="LOCTEXT(\"Squad Comms / Emote / Spot Enemy (Tap)\")", Index=@n1)
   )
   > "SetText (Text)"(
    self=[Get Text_PS4_Dpad_Left(), Get Text_Xbox_Dpad_Left()]
    InText=Select(Option 0="LOCTEXT(\"Gadget 1\")", Option 1="LOCTEXT(\"Gadget 1\")", Index=@n1)
   )
   > "SetText (Text)"(
    self=[Get Text_PS4_Dpad_Right(), Get Text_Xbox_Dpad_Right()]
    InText=Select(Option 0="LOCTEXT(\"Gadget 2\")", Option 1="LOCTEXT(\"Gadget 2\")", Index=@n1)
   )
   > "SetText (Text)"(
    self=Get Text_Gamepad_Select_1()
    InText=Select(Option 0="LOCTEXT(\"Inventory (Management, etc.)\")", Option 1="LOCTEXT(\"Inventory (Management, etc.)\")", Index=@n1)
   )
   > "SetText (Text)"(self=Get Text_Gamepad_Select_2(), InText=Select(Option 0="LOCTEXT(\"Chat (Hold)\")", Option 1="LOCTEXT(\"Chat (Hold)\")", Index=@n1))
   > "SetText (Text)"(
    self=Get Text_Gamepad_Start()
    InText=Select(Option 0="LOCTEXT(\"Game Menu (Friends, etc.)\")", Option 1="LOCTEXT(\"Game Menu (Friends, etc.)\")", Index=@n1)
   )
   > "SetText (Text)"(self=Get Text_Gamepad_RT(), InText=Select(Option 0="LOCTEXT(\"Attack\")", Option 1="LOCTEXT(\"Attack\")", Index=@n1))
   > "SetText (Text)"(
    self=Get Text_Gamepad_RB()
    InText=GetAbility2Label!(self:Fort Input Data Object Reference=GetFortInputDataForBlueprint()) -> (ReturnValue:Text)
   )
   > "SetText (Text)"(self=Get LT+RBAction(), InText=MakeLiteralText())
   > "SetText (Text)"(
    self=Get Text_Gamepad_Face_Button_Top_1()
    InText=Select(Option 0="LOCTEXT(\"Next Weapon (Tap)\")", Option 1="LOCTEXT(\"Next Weapon (Tap)\")", Index=@n1)
   )
   > "SetText (Text)"(
    self=Get Text_Gamepad_Face_Button_Top_2()
    InText=Select(Option 0="LOCTEXT(\"Toggle Pickaxe (Hold)\")", Option 1="LOCTEXT(\"Toggle Pickaxe (Hold)\")", Index=@n1)
   )
   > "SetText (Text)"(
    self=Get Text_Gamepad_Face_Button_Left_1()
    InText=Select(Option 0="LOCTEXT(\"Reload (Tap)\")", Option 1="LOCTEXT(\"Reload (Tap)\")", Index=@n1)
   )
   > "SetText (Text)"(
    self=Get Text_Gamepad_Face_Button_Left_2()
    InText=Select(Option 0="LOCTEXT(\"Craft Ammo (Hold)\")", Option 1="LOCTEXT(\"Craft Ammo (Hold)\")", Index=@n1)
   )
   > "SetText (Text)"(
    self=Get Text_Gamepad_Face_Button_Left_3()
    InText=Select(Option 0="LOCTEXT(\"Interact (Tap / Hold)\")", Option 1="LOCTEXT(\"Interact (Tap / Hold)\")", Index=@n1)
   )
   > "SetText (Text)"(
    self=Get Text_Gamepad_Face_Button_Right_1()
    InText=Select(Option 0="LOCTEXT(\"Toggle Build Mode (Tap)\")", Option 1="LOCTEXT(\"Toggle Build Mode (Tap)\")", Index=@n1)
   )
   > "SetText (Text)"(
    self=Get Text_Gamepad_Face_Button_Right_2()
    InText=Select(Option 0="LOCTEXT(\"Edit Building Piece (Hold)\")", Option 1="LOCTEXT(\"Edit Building Piece (Hold)\")", Index=@n1)
   )
   > "SetText (Text)"(
    self=Get Text_Gamepad_Face_Button_Right_3()
    InText=GetDeployHoverboardLabel!(self:Fort Input Data Object Reference=GetFortInputDataForBlueprint()) -> (ReturnValue:Text)
   )
   > "SetText (Text)"(self=Get Text_Gamepad_Face_Button_Bottom(), InText=Select(Option 0="LOCTEXT(\"Jump\")", Option 1="LOCTEXT(\"Jump\")", Index=@n1))
   > "SetText (Text)"(self=Get Text_Gamepad_RS(), InText=Select(Option 0="LOCTEXT(\"Look\")", Option 1="LOCTEXT(\"Look\")", Index=@n1))
   > "SetText (Text)"(
    self=Get Text_Gamepad_RS_Click_1()
    InText=Select(Option 0="LOCTEXT(\"Repair (Hold) / Improve (Hold)\")", Option 1="LOCTEXT(\"Repair (Hold) / Improve (Hold)\")", Index=@n1)
   )
   > "SetText (Text)"(self=Get Text_Gamepad_RS_Click_2(), InText=Select(Index=@n1))
   > "SetText (Text)"(
    self=Get Text_Gamepad_RS_Click_3()
    InText=Select(Option 0="LOCTEXT(\"Crouch (Tap) / Slide (Hold)\")", Option 1="LOCTEXT(\"Crouch (Tap) / Slide (Hold)\")", Index=@n1)
   )
   SetButtonTextLabelsFromTextArray(execute=[@n2.2, @n2.3], Array=Get ButtonTextLabels_STW_CombatPro_Combat())
  ]}
 HandleConfigChanged {nodes [
   n1 = HandleConfigChanged!() -> (then, Mode:EFortInputActionGroup Enum=AllModes, Config:int) {is=title K2Node_FunctionEntry}
   > n2 = Sequence()
   > n3 = Switch on EFortInputActionGroup!(execute, Selection:EFortInputActionGroup Enum=@n1) -> (
    AllModes:Exec
    Combat:Exec
    Building:Exec
    Movement:Exec
    Edit:Exec
    Death:Exec
    Cinematic:Exec
    Picker:Exec
    Other:Exec
    Interaction:Exec
    AthenaLTMAbilities:Exec
    ShoppingCart:Exec
    ShoppingCartDriver:Exec
    ShoppingCartPassenger:Exec
    Cannon:Exec
    CannonDriver:Exec
    CannonPassenger:Exec
    GolfCart:Exec
    GolfCartDriver:Exec
    GolfCartPassenger:Exec
    QuadCrasher:Exec
    QuadCrasherDriver:Exec
    QuadCrasherPassenger:Exec
    Biplane:Exec
    BiplaneDriver:Exec
    BiplaneGroundDriver:Exec
    BiplanePassenger:Exec
    Hamsterball:Exec
    Jackal:Exec
    Ostrich:Exec
    OstrichDriver:Exec
    OstrichPassenger:Exec
    Meatball:Exec
    MeatballDriver:Exec
    MeatballPassenger:Exec
    HoagieDriver:Exec
    Dagwood:Exec
    DagwoodDriver:Exec
    DagwoodPassenger:Exec
    Nevada:Exec
    NevadaDriver:Exec
    MountedTurret:Exec
    Spectating:Exec
    FullscreenMap:Exec
    CreativeAll:Exec
    CreativeModeratorMode:Exec
    CreativeMoveToolEquipped:Exec
    CreativeMoveObjectsFreely:Exec
    CreativeMoveBuildingsOnGrid:Exec
    CreativeFlying:Exec
    CreativeIslandPanel:Exec
    CreativeChair:Exec
    PropSelectorEquipped:Exec
    AlternativeInteractCarry:Exec
    AlternativeInteractPlace:Exec
    CreativeInputDevice:Exec
    DBNOCarryStart:Exec
    DBNOCarryStop:Exec
    DBNOCarry:Exec
    InteractionStart:Exec
    InteractionStop:Exec
    BattleLab:Exec
    RemoteControl:Exec
    SuperDuper:Exec
    Tether:Exec
    Riding:Exec
    Tank:Exec
    TankDriver:Exec
    TankPassenger:Exec
    RockVehicle:Exec
    RockVehicleB:Exec
    Dirtbike:Exec
    DirtbikeDriver:Exec
    DirtbikePassenger:Exec
    Sportbike:Exec
    SportbikeDriver:Exec
    SportbikePassenger:Exec
    PlayerAugments:Exec
    Targeting:Exec
    SimpleBuild:Exec
    CombatAndBuilding:Exec
    CombatAndAthenaLTMAbilities:Exec
    CombatBuildingAndAthenaLTMAbilities:Exec
   ) {is=title K2Node_SwitchEnum}
   Handle Config Changed STW Build Mode!(execute=@n3.Building, self, Config:int=@n1.Config) -> (then) {is=title K2Node_CallFunction}
   Handle Config Changed STW Combat Mode!(execute=@n3.Combat, self, Config:int=@n1.Config) -> (then) {is=title K2Node_CallFunction}
   Handle Config Changed Hide or Show Unused Text Lines(execute=@n2.then_1) {is=title K2Node_CallFunction}
  ]}
 HandleConfigChanged_Athena {nodes [
   n1 = HandleConfigChanged_Athena!() -> (then, Mode:EFortInputActionGroup Enum=AllModes, Config:int) {is=title K2Node_FunctionEntry}
   > n2 = Sequence!(execute) -> (then_0, then_1, then_2:Exec)
   > ClearAllComboTextLines()
   n3 = Switch on EFortInputActionGroup(execute=@n2.then_1, Selection=@n1)
   "SetText (Text)"(
    execute=@n3.Building
    self=Get Text_Gamepad_LT()
    InText=Select!(
     Option 0="LOCTEXT(\"-\")"
     Option 1="LOCTEXT(\"-\")"
     Option 2:Text="LOCTEXT(\"Change Building Material / Swap Trap\")"
     Option 3:Text="LOCTEXT(\"Stair Piece (Tap Select / Tap Place)\")"
     Option 4:Text="LOCTEXT(\"Stair Piece (Tap Select / Tap Place)\")"
     Index=@n1.Config
    ) -> (ReturnValue)
   )
   > "SetText (Text)"(
    self=Get Text_Gamepad_LB()
    InText=Select(
     Option 0="LOCTEXT(\"Change Building Material / Swap Trap\")"
     Option 1="LOCTEXT(\"Previous Building Piece\")"
     Option 2="LOCTEXT(\"Previous Building Piece\")"
     Option 3="LOCTEXT(\"Roof Piece (Tap Select / Tap Place)\")"
     Option 4="LOCTEXT(\"Roof Piece (Tap Select / Tap Place)\")"
     Index=@n1.Config
    )
   )
   > "SetText (Text)"(
    self=[Get Text_PS4_LS(), Get Text_Xbox_LS()]
    InText=Select(
     Option 0="LOCTEXT(\"Move\")"
     Option 1="LOCTEXT(\"Move\")"
     Option 2="LOCTEXT(\"Move\")"
     Option 3="LOCTEXT(\"Move\")"
     Option 4="LOCTEXT(\"Move\")"
     Index=@n1.Config
    )
   )
   > "SetText (Text)"(
    self=[Get Text_PS4_LS_Click_1(), Get Text_Xbox_LS_Click_1()]
    InText=Select(
     Option 0="LOCTEXT(\"Sprint\")"
     Option 1="LOCTEXT(\"Change Building Material/Swap Trap\")"
     Option 2="LOCTEXT(\"Sprint\")"
     Option 3="LOCTEXT(\"Sprint\")"
     Option 4="LOCTEXT(\"-\")"
     Index=@n1.Config
    )
   )
   > "SetText (Text)"(
    self=[Get Text_PS4_LS_Click_2(), Get Text_Xbox_LS_Click_2()]
    InText=Select(
     Option 0="LOCTEXT(\"Auto-Sprint (Double-Click)\")"
     Option 1="LOCTEXT(\"Auto-Sprint (Double-Click)\")"
     Option 2="LOCTEXT(\"Auto-Sprint (Double-Click)\")"
     Option 3="LOCTEXT(\"Auto-Sprint (Double-Click)\")"
     Option 4="LOCTEXT(\"-\")"
     Index=@n1.Config
    )
   )
   > "SetText (Text)"(
    self=[Get Text_PS4_Dpad_Up(), Get Text_Xbox_Dpad_Up()]
    InText=Select(
     Option 0="LOCTEXT(\"Toggle Map\")"
     Option 1="LOCTEXT(\"Toggle Map\")"
     Option 2="LOCTEXT(\"Inventory (Management, etc.)\")"
     Option 3="LOCTEXT(\"Inventory (Management, etc.)\")"
     Option 4="LOCTEXT(\"Change Building Material / Trap\")"
     Index=@n1.Config
    )
   )
   > "SetText (Text)"(
    self=[Get Text_PS4_Dpad_Down(), Get Text_Xbox_Dpad_Down()]
    InText=Select(
     Option 0="LOCTEXT(\"Emote / Augment Select\")"
     Option 1="LOCTEXT(\"Emote / Augment Select\")"
     Option 2="LOCTEXT(\"Emote / Augment Select\")"
     Option 3="LOCTEXT(\"Emote / Augment Select\")"
     Option 4="LOCTEXT(\"Rotate Building Piece (Tap)/Repair (Hold)\")"
     Index=@n1.Config
    )
   )
   > "SetText (Text)"(
    self=[Get Text_PS4_Dpad_Left(), Get Text_Xbox_Dpad_Left()]
    InText=Select(
     Option 0="LOCTEXT(\"-\")"
     Option 1="LOCTEXT(\"-\")"
     Option 2="LOCTEXT(\"-\")"
     Option 3="LOCTEXT(\"Change Building Material / Trap\")"
     Option 4="LOCTEXT(\"Inventory (Management, etc.)\")"
     Index=@n1.Config
    )
   )
   > "SetText (Text)"(
    self=[Get Text_PS4_Dpad_Right(), Get Text_Xbox_Dpad_Right()]
    InText=Select(
     Option 0="LOCTEXT(\"Squad Requests (Hold) / Augment Select\")"
     Option 1="LOCTEXT(\"Squad Requests (Hold) / Augment Select\")"
     Option 2="LOCTEXT(\"Squad Requests (Hold) / Augment Select\")"
     Option 3="LOCTEXT(\"Squad Requests (Hold) / Augment Select\")"
     Option 4="LOCTEXT(\"Squad Requests (Hold) / Augment Select\")"
     Index=@n1.Config
    )
   )
   > "SetText (Text)"(
    self=Get Text_Gamepad_Select_1()
    InText=Select(
     Option 0="LOCTEXT(\"Inventory (Management, etc.)\")"
     Option 1="LOCTEXT(\"Inventory (Management, etc.)\")"
     Option 2="LOCTEXT(\"Toggle Map\")"
     Option 3="LOCTEXT(\"Toggle Map\")"
     Option 4="LOCTEXT(\"-\")"
     Index=@n1.Config
    )
   )
   > "SetText (Text)"(self=Get Text_Gamepad_Select_2(), InText=Select!(Option 0, Option 1, Option 2, Option 3, Index=@n1.Config) -> (ReturnValue))
   > "SetText (Text)"(
    self=Get Text_Gamepad_Start()
    InText=Select!(
     Option 0="LOCTEXT(\"Game Menu (Friends, etc.)\")"
     Option 1="LOCTEXT(\"Game Menu (Friends, etc.)\")"
     Option 2="LOCTEXT(\"Game Menu (Friends, etc.)\")"
     Option 3="LOCTEXT(\"Game Menu (Friends, etc.)\")"
     Option 4:Text="LOCTEXT(\"Game Menu (Friends, etc.)\")"
     Index=@n1.Config
    ) -> (ReturnValue)
   )
   > "SetText (Text)"(
    self=Get Text_Gamepad_RT()
    InText=Select(
     Option 0="LOCTEXT(\"Confirm\")"
     Option 1="LOCTEXT(\"Confirm\")"
     Option 2="LOCTEXT(\"Confirm\")"
     Option 3="LOCTEXT(\"Wall Piece (Tap Select / Tap Place)\")"
     Option 4="LOCTEXT(\"Wall Piece (Tap Select / Tap Place)\")"
     Index=@n1.Config
    )
   )
   > "SetText (Text)"(
    self=Get Text_Gamepad_RB()
    InText=Select(
     Option 0="LOCTEXT(\"Rotate Building Piece\")"
     Option 1="LOCTEXT(\"Next Building Piece\")"
     Option 2="LOCTEXT(\"Next Building Piece\")"
     Option 3="LOCTEXT(\"Floor Piece (Tap Select / Tap Place)\")"
     Option 4="LOCTEXT(\"Floor Piece (Tap Select / Tap Place)\")"
     Index=@n1.Config
    )
   )
   > "SetText (Text)"(
    self=Get Text_Gamepad_Face_Button_Top_1()
    InText=Select(
     Option 0="LOCTEXT(\"Next Building Piece\")"
     Option 1="LOCTEXT(\"Select Trap\")"
     Option 2="LOCTEXT(\"Toggle Pickaxe\")"
     Option 3="LOCTEXT(\"Toggle Pickaxe\")"
     Option 4="LOCTEXT(\"Toggle Pickaxe\")"
     Index=@n1.Config
    )
   )
   > "SetText (Text)"(self=Get Text_Gamepad_Face_Button_Top_2(), InText=Select(Index=@n1.Config))
   > "SetText (Text)"(
    self=Get Text_Gamepad_Face_Button_Left_1()
    InText=Select(
     Option 0="LOCTEXT(\"Trap (Tap/Hold)\")"
     Option 1="LOCTEXT(\"Interact (Tap/Hold)\")"
     Option 2="LOCTEXT(\"Trap (Tap)\")"
     Option 3="LOCTEXT(\"Trap (Tap Select / Tap Place)\")"
     Option 4="LOCTEXT(\"Trap (Tap Select / Tap Place)\")"
     Index=@n1.Config
    )
   )
   > "SetText (Text)"(
    self=Get Text_Gamepad_Face_Button_Left_2()
    InText=Select(
     Option 0="LOCTEXT(\"Interact (Tap / Hold)\")"
     Option 2="LOCTEXT(\"Interact (Tap / Hold)\")"
     Option 3="LOCTEXT(\"Interact (Tap / Hold)\")"
     Option 4="LOCTEXT(\"Interact (Tap / Hold)\")"
     Index=@n1.Config
    )
   )
   > "SetText (Text)"(self=Get Text_Gamepad_Face_Button_Left_3(), InText=Select(Index=@n1.Config))
   > "SetText (Text)"(
    self=Get Text_Gamepad_Face_Button_Right_1()
    InText=Select(
     Option 0="LOCTEXT(\"Toggle Build Mode\")"
     Option 1="LOCTEXT(\"Toggle Build Mode\")"
     Option 2="LOCTEXT(\"Toggle Build Mode\")"
     Option 3="LOCTEXT(\"Toggle Build Mode\")"
     Option 4="LOCTEXT(\"Toggle Build Mode\")"
     Index=@n1.Config
    )
   )
   > "SetText (Text)"(
    self=Get Text_Gamepad_Face_Button_Right_2()
    InText=Select(
     Option 0="LOCTEXT(\"Edit Building Piece (Hold)\")"
     Option 1="LOCTEXT(\"Edit Building Piece (Hold)\")"
     Option 2="LOCTEXT(\"Edit Building Piece (Hold)\")"
     Option 3="LOCTEXT(\"Edit Building Piece (Hold)\")"
     Option 4="LOCTEXT(\"Edit Building Piece (Hold)\")"
     Index=@n1.Config
    )
   )
   > "SetText (Text)"(self=Get Text_Gamepad_Face_Button_Right_3(), InText=MakeLiteralText())
   > "SetText (Text)"(
    self=Get Text_Gamepad_Face_Button_Bottom()
    InText=Select(
     Option 0="LOCTEXT(\"Jump\")"
     Option 1="LOCTEXT(\"Jump\")"
     Option 2="LOCTEXT(\"Jump\")"
     Option 3="LOCTEXT(\"Jump\")"
     Option 4="LOCTEXT(\"Jump\")"
     Index=@n1.Config
    )
   )
   > "SetText (Text)"(
    self=Get Text_Gamepad_RS()
    InText=Select(
     Option 0="LOCTEXT(\"Look\")"
     Option 1="LOCTEXT(\"Look\")"
     Option 2="LOCTEXT(\"Look\")"
     Option 3="LOCTEXT(\"Look\")"
     Option 4="LOCTEXT(\"Look\")"
     Index=@n1.Config
    )
   )
   > "SetText (Text)"(
    self=Get Text_Gamepad_RS_Click_1()
    InText=Select(
     Option 0="LOCTEXT(\"Repair\")"
     Option 1="LOCTEXT(\"Rotate Building Piece (Tap)/Repair (Hold)\")"
     Option 2="LOCTEXT(\"Rotate Building Piece (Tap)/Repair (Hold)\")"
     Option 3="LOCTEXT(\"Rotate Building Piece (Tap)/Repair (Hold)\")"
     Option 4="LOCTEXT(\"-\")"
     Index=@n1.Config
    )
   )
   > "SetText (Text)"(
    self=Get Text_Gamepad_RS_Click_2()
    InText=Select(
     Option 0="LOCTEXT(\"Reset Building Edit (Edit Mode)\")"
     Option 1="LOCTEXT(\"Reset Building Edit (Edit Mode)\")"
     Option 2="LOCTEXT(\"Change Trap/Reset Building Edit (Edit Mode)\")"
     Option 3="LOCTEXT(\"Change Trap/Reset Building Edit (Edit Mode)\")"
     Option 4="LOCTEXT(\"-\")"
     Index=@n1.Config
    )
   )
   "SetText (Text)"(
    execute=@n3.Combat
    self=Get Text_Gamepad_LT()
    InText=Select(
     Option 0="LOCTEXT(\"Aim Down Sights\")"
     Option 1="LOCTEXT(\"Aim Down Sights\")"
     Option 2="LOCTEXT(\"Aim Down Sights\")"
     Option 3="LOCTEXT(\"Aim Down Sights\")"
     Option 4="LOCTEXT(\"Aim Down Sights\")"
     Index=@n1.Config
    )
   )
   > "SetText (Text)"(
    self=Get Text_Gamepad_LB()
    InText=Select(
     Option 0="LOCTEXT(\"Crouch (Tap) / Slide (Hold)\")"
     Option 1="LOCTEXT(\"Crouch (Tap) / Slide (Hold)\")"
     Option 2="LOCTEXT(\"Previous Weapon\")"
     Option 3="LOCTEXT(\"Previous Weapon\")"
     Option 4="LOCTEXT(\"Previous Weapon\")"
     Index=@n1.Config
    )
   )
   > "SetText (Text)"(
    self=[Get Text_Xbox_LS(), Get Text_PS4_LS()]
    InText=Select(
     Option 0="LOCTEXT(\"Move\")"
     Option 1="LOCTEXT(\"Move\")"
     Option 2="LOCTEXT(\"Move\")"
     Option 3="LOCTEXT(\"Move\")"
     Option 4="LOCTEXT(\"Move\")"
     Index=@n1.Config
    )
   )
   > "SetText (Text)"(
    self=[Get Text_Xbox_LS_Click_1(), Get Text_PS4_LS_Click_1()]
    InText=Select(
     Option 0="LOCTEXT(\"Sprint\")"
     Option 1="LOCTEXT(\"Sprint\")"
     Option 2="LOCTEXT(\"Sprint\")"
     Option 3="LOCTEXT(\"Sprint\")"
     Option 4="LOCTEXT(\"-\")"
     Index=@n1.Config
    )
   )
   > "SetText (Text)"(
    self=[Get Text_PS4_LS_Click_2(), Get Text_Xbox_LS_Click_2()]
    InText=Select(
     Option 0="LOCTEXT(\"Auto-Sprint (Double-Click)\")"
     Option 1="LOCTEXT(\"Auto-Sprint (Double-Click)\")"
     Option 2="LOCTEXT(\"Auto-Sprint (Double-Click)\")"
     Option 3="LOCTEXT(\"Auto-Sprint (Double-Click)\")"
     Option 4="LOCTEXT(\"-\")"
     Index=@n1.Config
    )
   )
   > "SetText (Text)"(
    self=[Get Text_PS4_Dpad_Up(), Get Text_Xbox_Dpad_Up()]
    InText=Select(
     Option 0="LOCTEXT(\"Toggle Map\")"
     Option 1="LOCTEXT(\"Toggle Map\")"
     Option 2="LOCTEXT(\"Inventory (Management, etc.)\")"
     Option 3="LOCTEXT(\"Inventory (Management, etc.)\")"
     Option 4="LOCTEXT(\"Emote / Augment Select\")"
     Index=@n1.Config
    )
   )
   > "SetText (Text)"(
    self=[Get Text_PS4_Dpad_Down(), Get Text_Xbox_Dpad_Down()]
    InText=Select(
     Option 0="LOCTEXT(\"Emote / Augment Select\")"
     Option 1="LOCTEXT(\"Emote / Augment Select\")"
     Option 2="LOCTEXT(\"Emote / Augment Select\")"
     Option 3="LOCTEXT(\"Emote / Augment Select\")"
     Option 4="LOCTEXT(\"Crouch (Tap) / Slide (Hold)\")"
     Index=@n1.Config
    )
   )
   > "SetText (Text)"(
    self=[Get Text_PS4_Dpad_Left(), Get Text_Xbox_Dpad_Left()]
    InText=Select(
     Option 0="LOCTEXT(\"Place Marker/Squad Coms(Hold)\")"
     Option 1="LOCTEXT(\"Place Marker/Squad Coms(Hold)\")"
     Option 2="LOCTEXT(\"Place Marker/Squad Coms(Hold)\")"
     Option 3="LOCTEXT(\"Place Marker/Squad Coms(Hold)\")"
     Option 4="LOCTEXT(\"Inventory (Management, etc.)\")"
     Index=@n1.Config
    )
   )
   > "SetText (Text)"(
    self=[Get Text_PS4_Dpad_Right(), Get Text_Xbox_Dpad_Right()]
    InText=Select(
     Option 0="LOCTEXT(\"Squad Requests (Hold) / Augment Select\")"
     Option 1="LOCTEXT(\"Squad Requests (Hold) / Augment Select\")"
     Option 2="LOCTEXT(\"Squad Requests (Hold) / Augment Select\")"
     Option 3="LOCTEXT(\"Squad Requests (Hold) / Augment Select\")"
     Option 4="LOCTEXT(\"Toggle Map\")"
     Index=@n1.Config
    )
   )
   > "SetText (Text)"(
    self=Get Text_Gamepad_Select_1()
    InText=Select(
     Option 0="LOCTEXT(\"Inventory (Management, etc.)\")"
     Option 1="LOCTEXT(\"Inventory (Management, etc.)\")"
     Option 2="LOCTEXT(\"Toggle Map\")"
     Option 3="LOCTEXT(\"Toggle Map\")"
     Option 4="LOCTEXT(\"-\")"
     Index=@n1.Config
    )
   )
   > "SetText (Text)"(self=Get Text_Gamepad_Select_2(), InText=Select!(Option 0, Option 1, Option 2, Option 3, Index=@n1.Config) -> (ReturnValue))
   > "SetText (Text)"(
    self=Get Text_Gamepad_Start()
    InText=Select!(
     Option 0="LOCTEXT(\"Game Menu (Friends, etc.)\")"
     Option 1="LOCTEXT(\"Game Menu (Friends, etc.)\")"
     Option 2="LOCTEXT(\"Game Menu (Friends, etc.)\")"
     Option 3="LOCTEXT(\"Game Menu (Friends, etc.)\")"
     Option 4:Text="LOCTEXT(\"Game Menu (Friends, etc.)\")"
     Index=@n1.Config
    ) -> (ReturnValue)
   )
   > "SetText (Text)"(
    self=Get Text_Gamepad_RT()
    InText=Select(
     Option 0="LOCTEXT(\"Attack / Confirm\")"
     Option 1="LOCTEXT(\"Attack / Confirm\")"
     Option 2="LOCTEXT(\"Attack / Confirm\")"
     Option 3="LOCTEXT(\"Attack / Confirm\")"
     Option 4="LOCTEXT(\"Attack / Confirm\")"
     Index=@n1.Config
    )
   )
   > "SetText (Text)"(
    self=Get Text_Gamepad_RB()
    InText=Select(
     Option 0="LOCTEXT(\"-\")"
     Option 1="LOCTEXT(\"-\")"
     Option 2="LOCTEXT(\"Next Weapon\")"
     Option 3="LOCTEXT(\"Next Weapon\")"
     Option 4="LOCTEXT(\"Next Weapon\")"
     Index=@n1.Config
    )
   )
   > "SetText (Text)"(
    self=Get Text_Gamepad_Face_Button_Top_1()
    InText=Select(
     Option 0="LOCTEXT(\"Next Weapon\")"
     Option 1="LOCTEXT(\"Next Weapon\")"
     Option 2="LOCTEXT(\"Toggle Pickaxe\")"
     Option 3="LOCTEXT(\"Toggle Pickaxe\")"
     Option 4="LOCTEXT(\"Toggle Pickaxe\")"
     Index=@n1.Config
    )
   )
   > "SetText (Text)"(
    self=Get Text_Gamepad_Face_Button_Top_2()
    InText=Select(
     Option 0="LOCTEXT(\"Pickaxe (Hold)\")"
     Option 1="LOCTEXT(\"Pickaxe (Hold)\")"
     Option 2="LOCTEXT(\"-\")"
     Option 3="LOCTEXT(\"-\")"
     Option 4="LOCTEXT(\"-\")"
     Index=@n1.Config
    )
   )
   > "SetText (Text)"(
    self=Get Text_Gamepad_Face_Button_Left_1()
    InText=Select(
     Option 0="LOCTEXT(\"Reload\")"
     Option 1="LOCTEXT(\"Reload\")"
     Option 2="LOCTEXT(\"Reload\")"
     Option 3="LOCTEXT(\"Reload\")"
     Option 4="LOCTEXT(\"Reload\")"
     Index=@n1.Config
    )
   )
   > "SetText (Text)"(
    self=Get Text_Gamepad_Face_Button_Left_2()
    InText=Select(
     Option 0="LOCTEXT(\"Interact (Tap / Hold)\")"
     Option 1="LOCTEXT(\"Interact (Tap / Hold)\")"
     Option 2="LOCTEXT(\"Interact (Tap / Hold)\")"
     Option 3="LOCTEXT(\"Interact (Tap / Hold)\")"
     Option 4="LOCTEXT(\"Interact (Tap / Hold)\")"
     Index=@n1.Config
    )
   )
   > "SetText (Text)"(
    self=Get Text_Gamepad_Face_Button_Left_3()
    InText=Select(
     Option 0="LOCTEXT(\"-\")"
     Option 1="LOCTEXT(\"-\")"
     Option 2="LOCTEXT(\"-\")"
     Option 3="LOCTEXT(\"-\")"
     Option 4="LOCTEXT(\"-\")"
     Index=@n1.Config
    )
   )
   > "SetText (Text)"(
    self=Get Text_Gamepad_Face_Button_Right_1()
    InText=Select(
     Option 0="LOCTEXT(\"Toggle Build Mode\")"
     Option 1="LOCTEXT(\"Toggle Build Mode\")"
     Option 2="LOCTEXT(\"Toggle Build Mode\")"
     Option 3="LOCTEXT(\"Toggle Build Mode\")"
     Option 4="LOCTEXT(\"Toggle Build Mode\")"
     Index=@n1.Config
    )
   )
   > "SetText (Text)"(
    self=Get Text_Gamepad_Face_Button_Right_2()
    InText=Select(
     Option 0="LOCTEXT(\"Edit Building Piece (Hold)\")"
     Option 1="LOCTEXT(\"Edit Building Piece (Hold)\")"
     Option 2="LOCTEXT(\"Edit Building Piece (Hold)\")"
     Option 3="LOCTEXT(\"Edit Building Piece (Hold)\")"
     Option 4="LOCTEXT(\"Edit Building Piece (Hold)\")"
     Index=@n1.Config
    )
   )
   > "SetText (Text)"(self=Get Text_Gamepad_Face_Button_Right_3(), InText=MakeLiteralText(Value="LOCTEXT(\"-\")"))
   > "SetText (Text)"(
    self=Get Text_Gamepad_Face_Button_Bottom()
    InText=Select(
     Option 0="LOCTEXT(\"Jump\")"
     Option 1="LOCTEXT(\"Jump\")"
     Option 2="LOCTEXT(\"Jump\")"
     Option 3="LOCTEXT(\"Jump\")"
     Option 4="LOCTEXT(\"Jump\")"
     Index=@n1.Config
    )
   )
   > "SetText (Text)"(
    self=Get Text_Gamepad_RS()
    InText=Select(
     Option 0="LOCTEXT(\"Look\")"
     Option 1="LOCTEXT(\"Look\")"
     Option 2="LOCTEXT(\"Look\")"
     Option 3="LOCTEXT(\"Look\")"
     Option 4="LOCTEXT(\"Look\")"
     Index=@n1.Config
    )
   )
   > "SetText (Text)"(
    self=Get Text_Gamepad_RS_Click_1()
    InText=Select(
     Option 0="LOCTEXT(\"Repair\")"
     Option 1="LOCTEXT(\"Repair\")"
     Option 2="LOCTEXT(\"Crouch (Tap) / Slide (Hold)\")"
     Option 3="LOCTEXT(\"Crouch (Tap) / Slide (Hold)\")"
     Option 4="LOCTEXT(\"-\")"
     Index=@n1.Config
    )
   )
   > "SetText (Text)"(
    self=Get Text_Gamepad_RS_Click_2()
    InText=Select(
     Option 0="LOCTEXT(\"-\")"
     Option 1="LOCTEXT(\"-\")"
     Option 2="LOCTEXT(\"Repair (Hold)\")"
     Option 3="LOCTEXT(\"Repair (Hold)\")"
     Option 4="LOCTEXT(\"-\")"
     Index=@n1.Config
    )
   )
   > "SetText (Text)"(
    self=Get Text_Gamepad_RS_Click_3()
    InText=Select(
     Option 0="LOCTEXT(\"-\")"
     Option 1="LOCTEXT(\"-\")"
     Option 2="LOCTEXT(\"-\")"
     Option 3="LOCTEXT(\"-\")"
     Option 4="LOCTEXT(\"-\")"
     Index=@n1.Config
    )
   )
   Handle Config Changed Hide or Show Unused Text Lines(execute=@n2.then_2)
  ]}
 HideAthenaLTMAbilitiesTab {nodes [
   HideAthenaLTMAbilitiesTab() {is=title K2Node_FunctionEntry}
   n1 = Get HorizontalTabList_CustomModes()
   > SetTabVisibility!(
    execute
    self:Common Tab List Widget Base Object Reference=@n1
    TabNameID:Name=Get LTMAbilitiesModeIdName() -> (LTMAbilitiesModeIdName:Name=None)
    NewVisibility:ESlateVisibility Enum=Collapsed
   ) -> (then)
   > Branch(Condition="Equal (Name)"!(
     A:Name=GetActiveTab!(self:Common Tab List Widget Base Object Reference=@n1) -> (ReturnValue:Name=None)
     B:Name=Get LTMAbilitiesModeIdName() -> (LTMAbilitiesModeIdName=None)
    ) -> (ReturnValue:bool) {is=title K2Node_CallFunction})
   > SelectTabByID!(
    execute
    self:Common Tab List Widget Base Object Reference=@n1
    TabNameID:Name=Get CombatModeIdName() -> (CombatModeIdName:Name=None)
    bSuppressClickFeedback:bool=true
   ) -> (then, ReturnValue:bool)
  ]}
 HideDeadZoneTab {nodes [
   HideDeadZoneTab() {is=title K2Node_FunctionEntry}
   n1 = Get HorizontalTabList_CustomModes()
   > SetTabVisibility(self=@n1, TabNameID=Get DeadZoneMoveIdName() -> (DeadZoneMoveIdName:Name=None), NewVisibility=Collapsed)
   > Branch(Condition="Equal (Name)"(A=GetActiveTab(self=@n1) -> (ReturnValue=None), B=Get DeadZoneMoveIdName() -> (DeadZoneMoveIdName=None)))
   > SelectTabByID(self=@n1, TabNameID=Get CombatModeIdName() -> (CombatModeIdName=None), bSuppressClickFeedback=true)
  ]}
 HideEditModeTab {nodes [
   HideEditModeTab() {is=title K2Node_FunctionEntry}
   n1 = Get HorizontalTabList_CustomModes()
   > SetTabVisibility(self=@n1, TabNameID=Get EditModeIdName() -> (EditModeIdName:Name=None), NewVisibility=Collapsed)
   > Branch(Condition="Equal (Name)"(A=GetActiveTab(self=@n1) -> (ReturnValue=None), B=Get EditModeIdName() -> (EditModeIdName=None)))
   > SelectTabByID(self=@n1, TabNameID=Get CombatModeIdName() -> (CombatModeIdName=None), bSuppressClickFeedback=true)
  ]}
 Image_LightboxClicked {nodes [Image_LightboxClicked() {is=title K2Node_FunctionEntry}]}
 InitializeListOfButtonLabelTextWidgets {nodes [
   InitializeListOfButtonLabelTextWidgets() {is=title K2Node_FunctionEntry}
   > Clear(TargetArray:Array of Common Text Object References=Get GamepadConfigTextWidgets() -> (GamepadConfigTextWidgets:Array of Common Text Object References))
   > n1 = Sequence()
   Add(
    execute=@n1.then_1
    TargetArray:Array of Common Text Object References=Get GamepadConfigTextWidgets()
    NewItem:"Common Text Object Reference (by ref)"=Get Text_Gamepad_Select_1()
   ) {is=title K2Node_CallArrayFunction}
   > Add(TargetArray=Get GamepadConfigTextWidgets(), NewItem=Get Text_Gamepad_Select_2())
   > Add(TargetArray=Get GamepadConfigTextWidgets(), NewItem=Get Text_Gamepad_Start())
   Add(execute=@n1.then_2, TargetArray=Get GamepadConfigTextWidgets(), NewItem=Get Text_Gamepad_RT())
   > Add(TargetArray=Get GamepadConfigTextWidgets(), NewItem=Get Text_Gamepad_RB())
   > Add(TargetArray=Get GamepadConfigTextWidgets(), NewItem=Get LT+RBAction())
   > Add(TargetArray=Get GamepadConfigTextWidgets(), NewItem=Get Text_Gamepad_Face_Button_Top_1())
   > Add(TargetArray=Get GamepadConfigTextWidgets(), NewItem=Get Text_Gamepad_Face_Button_Top_2())
   > Add(TargetArray=Get GamepadConfigTextWidgets(), NewItem=Get Text_Gamepad_Face_Button_Left_1())
   > Add(TargetArray=Get GamepadConfigTextWidgets(), NewItem=Get Text_Gamepad_Face_Button_Left_2())
   > Add(TargetArray=Get GamepadConfigTextWidgets(), NewItem=Get Text_Gamepad_Face_Button_Left_3())
   > Add(TargetArray=Get GamepadConfigTextWidgets(), NewItem=Get Text_Gamepad_Face_Button_Right_1())
   > Add(TargetArray=Get GamepadConfigTextWidgets(), NewItem=Get Text_Gamepad_Face_Button_Right_2())
   > Add(TargetArray=Get GamepadConfigTextWidgets(), NewItem=Get Text_Gamepad_Face_Button_Right_3())
   > Add(TargetArray=Get GamepadConfigTextWidgets(), NewItem=Get Text_Gamepad_Face_Button_Bottom())
   > Add(TargetArray=Get GamepadConfigTextWidgets(), NewItem=Get Text_Gamepad_RS())
   > Add(TargetArray=Get GamepadConfigTextWidgets(), NewItem=Get Text_Gamepad_RS_Click_1())
   > Add(TargetArray=Get GamepadConfigTextWidgets(), NewItem=Get Text_Gamepad_RS_Click_2())
   > Add(TargetArray=Get GamepadConfigTextWidgets(), NewItem=Get Text_Gamepad_RS_Click_3())
   Append Array!(
    execute=@n1.then_0
    TargetArray:Array of Common Text Object References=Get GamepadConfigTextWidgets()
    SourceArray:Array of Common Text Object References=Make Array!(
     "[0]":Common Text Object Reference=Get Text_Gamepad_LT()
     "[1]":Common Text Object Reference=Get Text_Gamepad_LB()
     "[2]":Common Text Object Reference=Get LB+RBAction()
     "[3]":Common Text Object Reference=Get LB+RTAction()
     "[4]":Common Text Object Reference=Get Text_PS4_LS()
     "[5]":Common Text Object Reference=Get Text_Xbox_LS()
    ) -> (Array:Array of Common Text Object References)
   ) -> (then)
   > Append Array(
    TargetArray=Get GamepadConfigTextWidgets()
    SourceArray=Make Array!("[0]"=Get Text_PS4_LS_Click_1(), "[1]"=Get Text_Xbox_LS_Click_1(), "[2]"=Get Text_PS4_LS_Click_2(), "[3]"=Get Text_Xbox_LS_Click_2()) -> (Array)
   )
   > Append Array(
    TargetArray=Get GamepadConfigTextWidgets()
    SourceArray=Make Array!(
     "[0]"=Get Text_PS4_Dpad_Up()
     "[1]"=Get Text_Xbox_Dpad_Up()
     "[2]"=Get Text_PS4_Dpad_Down()
     "[3]"=Get Text_Xbox_Dpad_Down()
     "[4]":Common Text Object Reference=Get Text_PS4_Dpad_Left()
     "[5]":Common Text Object Reference=Get Text_Xbox_Dpad_Left()
     "[6]":Common Text Object Reference=Get Text_PS4_Dpad_Right()
     "[7]":Common Text Object Reference=Get Text_Xbox_Dpad_Right()
    ) -> (Array)
   )
  ]}
 InitializeListOfButtonTextForSTWCombatProBuild {nodes [
   n1 = InitializeListOfButtonTextForSTWCombatProBuild!() -> (then, ButtonTextLabels:Array of Texts) {is=title K2Node_FunctionEntry}
   > Clear(TargetArray:Array of Texts=@n1)
   > n2 = Sequence()
   Add(
    execute=@n2.then_1
    TargetArray:Array of Texts=Get ButtonTextLabels() -> (ButtonTextLabels:Array of Texts)
    NewItem:"Text (by ref)"=GetToggleFullScreenMapLabel!(self:Fort Input Data Object Reference=GetFortInputDataForBlueprint()) -> (ReturnValue:Text)
   )
   > Add(TargetArray=Get ButtonTextLabels())
   > Add(
    TargetArray=Get ButtonTextLabels()
    NewItem=GetGameMenuLabel!(self:Fort Input Data Object Reference=GetFortInputDataForBlueprint()) -> (ReturnValue:Text)
   )
   Add(
    execute=@n2.then_2
    TargetArray=Get ButtonTextLabels()
    NewItem=GetPlaceBuildingOrTrapConfirmLabel!(self:Fort Input Data Object Reference=GetFortInputDataForBlueprint()) -> (ReturnValue:Text)
   )
   > Add(
    TargetArray=Get ButtonTextLabels()
    NewItem=GetNextBuildingPieceLabel!(self:Fort Input Data Object Reference=GetFortInputDataForBlueprint()) -> (ReturnValue:Text)
   )
   > Add(TargetArray=Get ButtonTextLabels())
   > Add(
    TargetArray=Get ButtonTextLabels()
    NewItem=GetHarvestWeaponLabel!(self:Fort Input Data Object Reference=GetFortInputDataForBlueprint()) -> (ReturnValue:Text)
   )
   > Add(TargetArray=Get ButtonTextLabels())
   > Add(
    TargetArray=Get ButtonTextLabels()
    NewItem=GetPickTrapTapHoldLabel!(self:Fort Input Data Object Reference=GetFortInputDataForBlueprint()) -> (ReturnValue:Text)
   )
   > Add(
    TargetArray=Get ButtonTextLabels()
    NewItem=GetTrapInventoryHoldLabel!(self:Fort Input Data Object Reference=GetFortInputDataForBlueprint()) -> (ReturnValue:Text)
   )
   > Add(
    TargetArray=Get ButtonTextLabels()
    NewItem=GetInteractTapHoldLabel!(self:Fort Input Data Object Reference=GetFortInputDataForBlueprint()) -> (ReturnValue:Text)
   )
   > Add(
    TargetArray=Get ButtonTextLabels()
    NewItem=GetSwitchQuickbarLabel!(self:Fort Input Data Object Reference=GetFortInputDataForBlueprint()) -> (ReturnValue:Text)
   )
   > Add(
    TargetArray=Get ButtonTextLabels()
    NewItem=GetEditBuildingPieceHoldLabel!(self:Fort Input Data Object Reference=GetFortInputDataForBlueprint()) -> (ReturnValue:Text)
   )
   > Add(TargetArray=Get ButtonTextLabels(), NewItem=MakeLiteralText())
   > Add(
    TargetArray=Get ButtonTextLabels()
    NewItem=GetJumpLabel!(self:Fort Input Data Object Reference=GetFortInputDataForBlueprint()) -> (ReturnValue:Text)
   )
   > Add(
    TargetArray=Get ButtonTextLabels()
    NewItem=GetLookLabel!(self:Fort Input Data Object Reference=GetFortInputDataForBlueprint()) -> (ReturnValue:Text)
   )
   > Add(
    TargetArray=Get ButtonTextLabels()
    NewItem=GetRepairOrImproveLabel!(self:Fort Input Data Object Reference=GetFortInputDataForBlueprint()) -> (ReturnValue:Text)
   )
   > Add(
    TargetArray=Get ButtonTextLabels()
    NewItem=GetBuildingEditResetLabel!(self:Fort Input Data Object Reference=GetFortInputDataForBlueprint()) -> (ReturnValue:Text)
   )
   > Add(
    TargetArray=Get ButtonTextLabels()
    NewItem=GetRotateBuildingLabel!(self:Fort Input Data Object Reference=GetFortInputDataForBlueprint()) -> (ReturnValue:Text)
   )
   Add(
    execute=@n2.then_0
    TargetArray=Get ButtonTextLabels()
    NewItem=GetChangeMaterialLabel!(self:Fort Input Data Object Reference=GetFortInputDataForBlueprint()) -> (ReturnValue:Text)
   )
   > Add(
    TargetArray=Get ButtonTextLabels()
    NewItem=GetPreviousBuildingPieceLabel!(self:Fort Input Data Object Reference=GetFortInputDataForBlueprint()) -> (ReturnValue:Text)
   )
   > Add(TargetArray=Get ButtonTextLabels())
   > Add(TargetArray=Get ButtonTextLabels())
   > Add(
    TargetArray=Get ButtonTextLabels()
    NewItem=GetMoveLabel!(self:Fort Input Data Object Reference=GetFortInputDataForBlueprint()) -> (ReturnValue:Text)
   )
   > Add(TargetArray=Get ButtonTextLabels(), NewItem=GetMoveLabel(self=GetFortInputDataForBlueprint()))
   > Add(
    TargetArray=Get ButtonTextLabels()
    NewItem=GetSprintLabel!(self:Fort Input Data Object Reference=GetFortInputDataForBlueprint()) -> (ReturnValue:Text)
   )
   > Add(TargetArray=Get ButtonTextLabels(), NewItem=GetSprintLabel(self=GetFortInputDataForBlueprint()))
   > Add(
    TargetArray=Get ButtonTextLabels()
    NewItem=GetAutoSprintLabel!(self:Fort Input Data Object Reference=GetFortInputDataForBlueprint()) -> (ReturnValue:Text)
   )
   > Add(TargetArray=Get ButtonTextLabels(), NewItem=GetAutoSprintLabel(self=GetFortInputDataForBlueprint()))
   > Add(
    TargetArray=Get ButtonTextLabels()
    NewItem=GetToggleInventoryLabel!(self:Fort Input Data Object Reference=GetFortInputDataForBlueprint()) -> (ReturnValue:Text)
   )
   > Add(TargetArray=Get ButtonTextLabels(), NewItem=GetToggleInventoryLabel(self=GetFortInputDataForBlueprint()))
   > Add(
    TargetArray=Get ButtonTextLabels()
    NewItem=GetEmoteOrQuickChatOrSpotEnemyLabel!(self:Fort Input Data Object Reference=GetFortInputDataForBlueprint()) -> (ReturnValue:Text)
   )
   > Add(TargetArray=Get ButtonTextLabels(), NewItem=GetEmoteOrQuickChatOrSpotEnemyLabel(self=GetFortInputDataForBlueprint()))
   > Add(TargetArray=Get ButtonTextLabels())
   > Add(TargetArray=Get ButtonTextLabels())
   > Add(TargetArray=Get ButtonTextLabels())
   > Add(TargetArray=Get ButtonTextLabels())
   GetDeployHoverboardLabel(self=GetFortInputDataForBlueprint())
  ]}
 InitializeListOfButtonTextForSTWCombatProCombat {nodes [
   n1 = InitializeListOfButtonTextForSTWCombatProCombat!() -> (then, ButtonTextLabels:Array of Texts) {is=title K2Node_FunctionEntry}
   > Clear(TargetArray=@n1)
   > n2 = Sequence()
   Add(execute=@n2.then_1, TargetArray=Get ButtonTextLabels(), NewItem=GetToggleFullScreenMapLabel(self=GetFortInputDataForBlueprint()))
   > Add(TargetArray=Get ButtonTextLabels())
   > Add(TargetArray=Get ButtonTextLabels(), NewItem=GetGameMenuLabel(self=GetFortInputDataForBlueprint()))
   Add(
    execute=@n2.then_2
    TargetArray=Get ButtonTextLabels()
    NewItem=GetAttackConfirmLabel!(self:Fort Input Data Object Reference=GetFortInputDataForBlueprint()) -> (ReturnValue:Text)
   )
   > Add(
    TargetArray=Get ButtonTextLabels()
    NewItem=GetNextWeaponTapLabel!(self:Fort Input Data Object Reference=GetFortInputDataForBlueprint()) -> (ReturnValue:Text)
   )
   > Add(TargetArray=Get ButtonTextLabels(), NewItem=GetAbility1Label(self=GetFortInputDataForBlueprint()))
   > Add(TargetArray=Get ButtonTextLabels(), NewItem=GetHarvestWeaponLabel(self=GetFortInputDataForBlueprint()))
   > Add(TargetArray=Get ButtonTextLabels())
   > Add(
    TargetArray=Get ButtonTextLabels()
    NewItem=GetReloadTapLabel!(self:Fort Input Data Object Reference=GetFortInputDataForBlueprint()) -> (ReturnValue:Text)
   )
   > Add(TargetArray=Get ButtonTextLabels(), NewItem=GetInteractTapHoldLabel(self=GetFortInputDataForBlueprint()))
   > Add(
    TargetArray=Get ButtonTextLabels()
    NewItem=GetCraftAmmoHoldLabel!(self:Fort Input Data Object Reference=GetFortInputDataForBlueprint()) -> (ReturnValue:Text)
   )
   > Add(TargetArray=Get ButtonTextLabels(), NewItem=GetSwitchQuickbarLabel(self=GetFortInputDataForBlueprint()))
   > Add(TargetArray=Get ButtonTextLabels(), NewItem=GetEditBuildingPieceHoldLabel(self=GetFortInputDataForBlueprint()))
   > Add(TargetArray=Get ButtonTextLabels(), NewItem=GetDeployHoverboardLabel(self=GetFortInputDataForBlueprint()))
   > Add(TargetArray=Get ButtonTextLabels(), NewItem=GetJumpLabel(self=GetFortInputDataForBlueprint()))
   > Add(TargetArray=Get ButtonTextLabels(), NewItem=GetLookLabel(self=GetFortInputDataForBlueprint()))
   > Add(TargetArray=Get ButtonTextLabels(), NewItem=GetRepairOrImproveLabel(self=GetFortInputDataForBlueprint()))
   > Add(TargetArray=Get ButtonTextLabels(), NewItem=MakeLiteralText())
   > Add(
    TargetArray=Get ButtonTextLabels()
    NewItem=GetCrouchLabel!(self:Fort Input Data Object Reference=GetFortInputDataForBlueprint()) -> (ReturnValue:Text)
   )
   Add(
    execute=@n2.then_0
    TargetArray=Get ButtonTextLabels()
    NewItem=GetAimDownSightsHeavyMeleeLabel!(self:Fort Input Data Object Reference=GetFortInputDataForBlueprint()) -> (ReturnValue:Text)
   )
   > Add(
    TargetArray=Get ButtonTextLabels()
    NewItem=GetPreviousWeaponTapLabel!(self:Fort Input Data Object Reference=GetFortInputDataForBlueprint()) -> (ReturnValue:Text)
   )
   > Add(TargetArray=Get ButtonTextLabels(), NewItem=GetAbility3Label(self=GetFortInputDataForBlueprint()))
   > Add(TargetArray=Get ButtonTextLabels(), NewItem=GetAbility2Label(self=GetFortInputDataForBlueprint()))
   > Add(TargetArray=Get ButtonTextLabels(), NewItem=GetMoveLabel(self=GetFortInputDataForBlueprint()))
   > Add(TargetArray=Get ButtonTextLabels(), NewItem=GetMoveLabel(self=GetFortInputDataForBlueprint()))
   > Add(TargetArray=Get ButtonTextLabels(), NewItem=GetSprintLabel(self=GetFortInputDataForBlueprint()))
   > Add(TargetArray=Get ButtonTextLabels(), NewItem=GetSprintLabel(self=GetFortInputDataForBlueprint()))
   > Add(TargetArray=Get ButtonTextLabels(), NewItem=GetAutoSprintLabel(self=GetFortInputDataForBlueprint()))
   > Add(TargetArray=Get ButtonTextLabels(), NewItem=GetAutoSprintLabel(self=GetFortInputDataForBlueprint()))
   > Add(TargetArray=Get ButtonTextLabels(), NewItem=GetToggleInventoryLabel(self=GetFortInputDataForBlueprint()))
   > Add(TargetArray=Get ButtonTextLabels(), NewItem=GetToggleInventoryLabel(self=GetFortInputDataForBlueprint()))
   > Add(TargetArray=Get ButtonTextLabels(), NewItem=GetEmoteOrQuickChatOrSpotEnemyLabel(self=GetFortInputDataForBlueprint()))
   > Add(TargetArray=Get ButtonTextLabels(), NewItem=GetEmoteOrQuickChatOrSpotEnemyLabel(self=GetFortInputDataForBlueprint()))
   > Add(
    TargetArray=Get ButtonTextLabels()
    NewItem=GetGadget1Label!(self:Fort Input Data Object Reference=GetFortInputDataForBlueprint()) -> (ReturnValue:Text)
   )
   > Add(TargetArray=Get ButtonTextLabels(), NewItem=GetGadget1Label(self=GetFortInputDataForBlueprint()))
   > Add(
    TargetArray=Get ButtonTextLabels()
    NewItem=GetGadget2Label!(self:Fort Input Data Object Reference=GetFortInputDataForBlueprint()) -> (ReturnValue:Text)
   )
   > Add(TargetArray=Get ButtonTextLabels(), NewItem=GetGadget2Label(self=GetFortInputDataForBlueprint()))
  ]}
 InitializeListOfButtonTextForSTW_BuilderProBuild {nodes [
   n1 = InitializeListOfButtonTextForSTW_BuilderProBuild!() -> (then, ButtonTextLabels:Array of Texts) {is=title K2Node_FunctionEntry}
   > Clear(TargetArray=@n1)
   > n2 = Sequence()
   Add(execute=@n2.then_1, TargetArray=Get ButtonTextLabels(), NewItem=GetToggleFullScreenMapLabel(self=GetFortInputDataForBlueprint()))
   > Add(TargetArray=Get ButtonTextLabels())
   > Add(TargetArray=Get ButtonTextLabels(), NewItem=GetGameMenuLabel(self=GetFortInputDataForBlueprint()))
   Add(
    execute=@n2.then_2
    TargetArray=Get ButtonTextLabels()
    NewItem=GetGamepadWallSelectAndPlaceLabel!(self:Fort Input Data Object Reference=GetFortInputDataForBlueprint()) -> (ReturnValue:Text)
   )
   > Add(
    TargetArray=Get ButtonTextLabels()
    NewItem=GetGamepadFloorSelectAndPlaceLabel!(self:Fort Input Data Object Reference=GetFortInputDataForBlueprint()) -> (ReturnValue:Text)
   )
   > Add(TargetArray=Get ButtonTextLabels())
   > Add(TargetArray=Get ButtonTextLabels(), NewItem=GetHarvestWeaponLabel(self=GetFortInputDataForBlueprint()))
   > Add(TargetArray=Get ButtonTextLabels())
   > Add(
    TargetArray=Get ButtonTextLabels()
    NewItem=GetGamepadTrapSelectAndPlaceLabel!(self:Fort Input Data Object Reference=GetFortInputDataForBlueprint()) -> (ReturnValue:Text)
   )
   > Add(
    TargetArray=Get ButtonTextLabels()
    NewItem=GetGamepadTrapPickerHoldLabel!(self:Fort Input Data Object Reference=GetFortInputDataForBlueprint()) -> (ReturnValue:Text)
   )
   > Add(TargetArray=Get ButtonTextLabels(), NewItem=GetInteractTapHoldLabel(self=GetFortInputDataForBlueprint()))
   > Add(TargetArray=Get ButtonTextLabels(), NewItem=GetSwitchQuickbarLabel(self=GetFortInputDataForBlueprint()))
   > Add(TargetArray=Get ButtonTextLabels(), NewItem=GetEditBuildingPieceHoldLabel(self=GetFortInputDataForBlueprint()))
   > Add(TargetArray=Get ButtonTextLabels(), NewItem=MakeLiteralText())
   > Add(TargetArray=Get ButtonTextLabels(), NewItem=GetJumpLabel(self=GetFortInputDataForBlueprint()))
   > Add(TargetArray=Get ButtonTextLabels(), NewItem=GetLookLabel(self=GetFortInputDataForBlueprint()))
   > Add(TargetArray=Get ButtonTextLabels(), NewItem=GetRepairOrImproveLabel(self=GetFortInputDataForBlueprint()))
   > Add(TargetArray=Get ButtonTextLabels(), NewItem=GetBuildingEditResetLabel(self=GetFortInputDataForBlueprint()))
   > Add(TargetArray=Get ButtonTextLabels(), NewItem=GetRotateBuildingLabel(self=GetFortInputDataForBlueprint()))
   Add(
    execute=@n2.then_0
    TargetArray=Get ButtonTextLabels()
    NewItem=GetGamepadStairSelectAndPlaceLabel!(self:Fort Input Data Object Reference=GetFortInputDataForBlueprint()) -> (ReturnValue:Text)
   )
   > Add(
    TargetArray=Get ButtonTextLabels()
    NewItem=GetGamepadRoofSelectAndPlaceLabel!(self:Fort Input Data Object Reference=GetFortInputDataForBlueprint()) -> (ReturnValue:Text)
   )
   > Add(TargetArray=Get ButtonTextLabels())
   > Add(TargetArray=Get ButtonTextLabels())
   > Add(TargetArray=Get ButtonTextLabels(), NewItem=GetMoveLabel(self=GetFortInputDataForBlueprint()))
   > Add(TargetArray=Get ButtonTextLabels(), NewItem=GetMoveLabel(self=GetFortInputDataForBlueprint()))
   > Add(TargetArray=Get ButtonTextLabels(), NewItem=GetSprintLabel(self=GetFortInputDataForBlueprint()))
   > Add(TargetArray=Get ButtonTextLabels(), NewItem=GetSprintLabel(self=GetFortInputDataForBlueprint()))
   > Add(TargetArray=Get ButtonTextLabels(), NewItem=GetAutoSprintLabel(self=GetFortInputDataForBlueprint()))
   > Add(TargetArray=Get ButtonTextLabels(), NewItem=GetAutoSprintLabel(self=GetFortInputDataForBlueprint()))
   > Add(TargetArray=Get ButtonTextLabels(), NewItem=GetToggleInventoryLabel(self=GetFortInputDataForBlueprint()))
   > Add(TargetArray=Get ButtonTextLabels(), NewItem=GetToggleInventoryLabel(self=GetFortInputDataForBlueprint()))
   > Add(TargetArray=Get ButtonTextLabels(), NewItem=GetEmoteOrQuickChatOrSpotEnemyLabel(self=GetFortInputDataForBlueprint()))
   > Add(TargetArray=Get ButtonTextLabels(), NewItem=GetEmoteOrQuickChatOrSpotEnemyLabel(self=GetFortInputDataForBlueprint()))
   > Add(
    TargetArray=Get ButtonTextLabels()
    NewItem=GetGamepadChangeMaterialOrTrapPickerLabel!(self:Fort Input Data Object Reference=GetFortInputDataForBlueprint()) -> (ReturnValue:Text)
   )
   > Add(TargetArray=Get ButtonTextLabels(), NewItem=GetGamepadChangeMaterialOrTrapPickerLabel(self=GetFortInputDataForBlueprint()))
   > Add(TargetArray=Get ButtonTextLabels())
   > Add(TargetArray=Get ButtonTextLabels())
  ]}
 MakeGamepadGameplayTagName {nodes [
   n1 = Inputs!() -> (PresetName:Name, InputActionGroup:EFortInputActionGroup Enum=AllModes, ActionName:Name)
   n2 = "To String (Name)"(InName=@n1)
   n3 = "To String (Name)"(InName=@n1.ActionName)
   n4 = Enum to String(Enumerator:EFortInputActionGroup Enum=@n1.InputActionGroup)
   Outputs!(ReturnValue:Name=String To Name(InString=Append(A=Gamepad, B=., C=@n2, D=., E=@n4, F=., G=@n3)))
  ]}
 OnConfigChanged {nodes [
   n1 = OnConfigChanged!() -> (then, Mode:EFortInputActionGroup Enum=AllModes, Config:int) {is=title K2Node_FunctionEntry}
   > n2 = Switch on ESubGame!(
    execute
    Selection:ESubGame Enum=GetSubGame!(self:Fort Global UIContext Object Reference=Get FortGlobalUIContext()) -> (ReturnValue:ESubGame Enum=Campaign)
   ) -> (Campaign:Exec, Athena:Exec, Invalid:Exec, Count:Exec, Creative:Exec) {is=title K2Node_SwitchEnum}
   n3 = Branch(
    execute=@n2.Athena
    Condition=IsCustomGamepadConfig!(self:Fort Global UIContext Object Reference=Get FortGlobalUIContext(), ConfigIndex:int=@n1.Config) -> (ReturnValue:bool)
   )
   HandleConfigChanged_Athena!(execute=@n3.else, self, Mode:EFortInputActionGroup Enum=@n1, Config:int=@n1.Config) -> (then) {is=title K2Node_CallFunction}
   HandleConfigChanged!(execute=[@n2.Campaign, @n2.Invalid], self, Mode:EFortInputActionGroup Enum=@n1, Config:int=@n1.Config) -> (then) {is=title K2Node_CallFunction}
  ]}
 On_Image_LightBox_MouseButtonDown_0 {nodes [
   On_Image_LightBox_MouseButtonDown_0!() -> (then, MyGeometry:Geometry Structure, MouseEvent:"Pointer Event Structure (by ref)") {is=title K2Node_FunctionEntry}
   > Call Image Lightbox Clicked!(execute:Exec, self:Gamepad Mapping Info V 3 Object Reference) -> (then:Exec) {is=title K2Node_CallDelegate}
   > Return Node!(execute, ReturnValue:Event Reply Structure=Handled())
  ]}
 SetButtonTextLabelsFromTextArray {nodes [
   n1 = SetButtonTextLabelsFromTextArray!() -> (then, Array:Array of Texts) {is=title K2Node_FunctionEntry}
   > n2 = For Each Loop(Array:Array of Common Text Object References=Get GamepadConfigTextWidgets()) -> (Array Element:Common Text Object Reference)
   > "SetText (Text)"(self=@n2, InText="Get (a ref)"(Array:Array of Texts=@n1, Dimension 1=@n2.Array Index) -> (Output:"Text (by ref)"))
  ]}
 SetTextFieldVisibilityForExtraLine {nodes [
   n1 = SetTextFieldVisibilityForExtraLine!() -> (then, TextWidget:Text Object Reference) {is=title K2Node_FunctionEntry}
   n2 = "GetText (Text)"(self=@n1)
   n3 = MakeLiteralText()
   > SetVisibility(
    self=@n1
    InVisibility=Select!(
     Option 0:ESlateVisibility Enum=SelfHitTestInvisible
     Option 1:ESlateVisibility Enum=Collapsed
     Index:bool="Equal Exactly (Text)"(A=@n2, B=@n3)
    ) -> (ReturnValue:ESlateVisibility Enum)
   )
  ]}
 SetTextForMode {nodes [
   n1 = SetTextForMode!() -> (then, Selection:EFortInputActionGroup Enum=AllModes) {is=title K2Node_FunctionEntry}
   > n2 = Switch on EFortInputActionGroup(Selection=@n1)
   "SetText (Text)"(execute=@n2.AthenaLTMAbilities, self=Get Text_Mode(), InText="LOCTEXT(\"Ability Controls\")")
   "SetText (Text)"(execute=@n2.Building, self=Get Text_Mode(), InText="LOCTEXT(\"Build Controls\")")
   "SetText (Text)"(execute=@n2.Combat, self=Get Text_Mode(), InText="LOCTEXT(\"Combat Controls\")")
   "SetText (Text)"(execute=@n2.Edit, self=Get Text_Mode(), InText="LOCTEXT(\"Edit Controls\")")
  ]}
 ShowAthenaLTMAbilitiesTab {nodes [
   ShowAthenaLTMAbilitiesTab() {is=title K2Node_FunctionEntry}
   > SetTabVisibility(self=Get HorizontalTabList_CustomModes(), TabNameID=Get LTMAbilitiesModeIdName() -> (LTMAbilitiesModeIdName=None), NewVisibility=Visible)
  ]}
 ShowDeadZoneTab {nodes [
   ShowDeadZoneTab() {is=title K2Node_FunctionEntry}
   > SetTabVisibility(self=Get HorizontalTabList_CustomModes(), TabNameID=Get DeadZoneMoveIdName() -> (DeadZoneMoveIdName=None), NewVisibility=Visible)
  ]}
 ShowEditModeTab {nodes [
   ShowEditModeTab() {is=title K2Node_FunctionEntry}
   > SetTabVisibility(self=Get HorizontalTabList_CustomModes(), TabNameID=Get EditModeIdName() -> (EditModeIdName=None), NewVisibility=Visible)
  ]}
 Update Displayed Mapping {nodes [
   n1 = Update Displayed Mapping!() -> (then, Platform:int, Mode:EFortInputActionGroup Enum=AllModes, Configuration:int) {is=title K2Node_FunctionEntry}
   > n2 = Sequence!(execute) -> (then_0, then_1, then_2, then_3:Exec)
   SetVisibility(
    execute=@n2.then_2
    self=Get VerticalBoxMenu+Map() -> (VerticalBoxMenu+Map:Vertical Box Object Reference)
    InVisibility=Select!(
     Option 0=HitTestInvisible
     Option 1=HitTestInvisible
     Option 2:ESlateVisibility Enum=HitTestInvisible
     Option 3:ESlateVisibility Enum=HitTestInvisible
     Option 4:ESlateVisibility Enum=HitTestInvisible
     Option 5:ESlateVisibility Enum=HitTestInvisible
     Option 6:ESlateVisibility Enum=HitTestInvisible
     Index:int=Get Configuration() -> (Configuration:int=0)
    ) -> (ReturnValue)
   )
   n3 = Get Platform() -> (Platform:int=0)
   SetVisibility(
    execute=@n2.then_1
    self=Get VerticalBox_Xbox_LS_And_Dpad() -> (VerticalBox_Xbox_LS_And_Dpad:Vertical Box Object Reference)
    InVisibility=Select(
     Option 0=HitTestInvisible
     Option 1=Collapsed
     Option 2=Collapsed
     Option 3=HitTestInvisible
     Option 4=Collapsed
     Option 5=Collapsed
     Option 6=Collapsed
     Index=@n3
    )
   )
   > SetVisibility(
    self=Get VerticalBox_PS4_LS_And_Dpad() -> (VerticalBox_PS4_LS_And_Dpad:Vertical Box Object Reference)
    InVisibility=Select(
     Option 0=Collapsed
     Option 1=HitTestInvisible
     Option 2=HitTestInvisible
     Option 3=Collapsed
     Option 4=HitTestInvisible
     Option 5=HitTestInvisible
     Option 6=HitTestInvisible
     Index=@n3
    )
   )
   SetTextForMode!(execute=@n2.then_3, self, Selection:EFortInputActionGroup Enum=@n1.Mode) -> (then) {is=title K2Node_CallFunction}
   Set Appropriate Platform Images(
    execute=@n2.then_0
    Appropriate Platform Images:Array of Texture 2D Soft Object References=Select(
     Option 0:Array of Texture 2D Soft Object References=Get XB1 Images() -> (XB1 Images:Array of Texture 2D Soft Object References)
     Option 1:Array of Texture 2D Soft Object References=Get PS4 Images() -> (PS4 Images:Array of Texture 2D Soft Object References)
     Option 2:Array of Texture 2D Soft Object References=Get Generic Images() -> (Generic Images:Array of Texture 2D Soft Object References)
     Option 3:Array of Texture 2D Soft Object References=Get XSX Images() -> (XSX Images:Array of Texture 2D Soft Object References)
     Option 4:Array of Texture 2D Soft Object References=Get PS5 Images() -> (PS5 Images:Array of Texture 2D Soft Object References)
     Option 5:Array of Texture 2D Soft Object References=Get Erebus Images() -> (Erebus Images:Array of Texture 2D Soft Object References)
     Option 6:Array of Texture 2D Soft Object References=Get Switch2 Images() -> (Switch2 Images:Array of Texture 2D Soft Object References)
     Index=Get Platform() -> (Platform=0)
    ) -> (ReturnValue:Array of Texture 2D Soft Object References)
   ) -> (Output_Get:Array of Texture 2D Soft Object References)
   > Update Controller Images() {is=title K2Node_CallFunction}
  ]}
]
variables [
 Appropriate Platform Images {type="array<Texture2D>", default [/FortUILibrary/Input/GamepadGeneric/Generic_ControllerConfig]}
 BuildModeIdName {type=string, default=None}
 ButtonTextLabels_STW_BuilderPro_Build {type="array<string>", default []}
 ButtonTextLabels_STW_CombatPro_Build {type="array<string>", default []}
 ButtonTextLabels_STW_CombatPro_Combat {type="array<string>", default []}
 CombatModeIdName {type=string, default=None}
 CurrentInputActionGroup {type=EFortInputActionGroup, default=AllModes}
 CustomButtonNavLeft {}
 CustomKeyButtons {type="array<GamepadKeyTextButton_C>", default [null,null,null,null,null,null,null,null,null,null,null,null,null,null,null,null]}
 CustomKeySelected {}
 DeadZoneMoveIdName {type=string, default=None}
 EditModeIdName {type=string, default=None}
 Erebus Images {type="array<Texture2D>", default [/FortUILibrary/Input/GamepadSwitch/T_UI_Erebus_ControllerConfig]}
 GamepadConfigTextWidgets {type="array<CommonTextBlock>", default []}
 Generic Images {type="array<Texture2D>", default [&Generic_ControllerConfig]}
 Image_LightboxClicked {}
 LastClickedButton {type=GamepadKeyTextButton_C, default=null}
 LTMAbilitiesModeIdName {type=string, default=None}
 PS4 Images {type="array<Texture2D>", default [/FortUILibrary/Input/GamepadPS4/PS4_ControllerConfig]}
 PS5 Images {type="array<Texture2D>", default [/FortUILibrary/Input/GamepadPS5/PS5_ControllerConfig]}
 Switch2 Images {type="array<Texture2D>", default [/FortUILibrary/Input/GamepadSwitch2/T_UI_Ounce_ControllerConfig]}
 XB1 Images {type="array<Texture2D>", default [/FortUILibrary/Input/GamepadXboxOne/XB1_ControllerConfig]}
 XSX Images {type="array<Texture2D>", default [/FortUILibrary/Input/GamepadXboxSeriesX/XboxSeriesX_ControllerConfig]}
]
widgets {
 info {
  inheritedWidgetCount=0
  parentClass=/Script/FortniteUI.FortGamepadMappingInfo
  rootWidgetClass=&CommonBorder
  widgetCount=275
 }
 widgets [
  CommonBorder_0 {
   widgetClassPath=&CommonBorder
   widget=CommonBorder_0
   properties {
    horizontalAlignment=HAlign_Center
    verticalAlignment=VAlign_Center
    background {
     bIsDynamicallyLoaded=false
     drawAs=NoDrawType
     imageType=NoImage
     mirroring=NoMirror
     resourceName=None
     resourceObject=null
     tiling=NoTile
     imageSize=<208 128>
     margin=<ltrb 0.7522 0.01 0.0442 0.75>
     outlineSettings {
      bUseBrushTransparency=false
      roundingType=HalfHeightRadius
      width=0
      color {colorUseRule=UseColor_Specified, specifiedColor=<rgba 0 0 0 0>}
      cornerRadii=<0 0 0 0>
     }
     tintColor {colorUseRule=UseColor_Specified, specifiedColor=<rgba 1 1 1 1>}
     uVRegion {bIsValid=false, max=<0 0>, min=<0 0>}
    }
    padding=<ltrb 0 0 0 0>
   }
  }
  Overlay_0 {
   widgetClassPath=&Overlay
   parent=CommonBorder_0
   widget=Overlay_0
   slot {class=BorderSlot, properties {horizontalAlignment=HAlign_Center, verticalAlignment=VAlign_Center, padding=<ltrb 0 0 0 0>}}
  }
  Config {
   widgetClassPath=&Overlay
   parent=Overlay_0
   widget=Config
   slot {class=OverlaySlot, properties {horizontalAlignment=HAlign_Fill, verticalAlignment=VAlign_Fill}}
  }
  SizeBox_0 {
   widgetClassPath=&SizeBox
   parent=Config
   widget=SizeBox_0
   properties {bOverride_MinDesiredWidth=true, minDesiredWidth=1664}
   slot {class=OverlaySlot, properties {horizontalAlignment=HAlign_Fill, verticalAlignment=VAlign_Fill}}
  }
  Overlay_4 {widgetClassPath=&Overlay, parent=SizeBox_0, widget=Overlay_4, slot {class=SizeBoxSlot}}
  Overlay_DeadzoneAdjustment {
   widgetClassPath=&Overlay
   bIsVariable=true
   parent=Overlay_4
   widget=Overlay_DeadzoneAdjustment
   properties {
    navigation=&self:WidgetTree.Overlay_DeadzoneAdjustment.WidgetNavigation_0
    visibility=Collapsed
   }
   slot {class=OverlaySlot, properties {horizontalAlignment=HAlign_Center, verticalAlignment=VAlign_Center, padding=<ltrb 0 80 0 0>}}
  }
  DeadzoneAdjustment_Move {
   widgetClassPath=/FortniteCoreUI/Settings/Deadzone/DeadzoneAdjustment
   bIsVariable=true
   parent=Overlay_DeadzoneAdjustment
   widget=DeadzoneAdjustment_Move
   properties {bIsFocusable=false, settingType=MoveStickDeadZone, visibility=Visible}
   slot {class=OverlaySlot, properties {verticalAlignment=VAlign_Center, padding=<ltrb 0 25 1000 0>}}
  }
  DeadzoneAdjustment_Look {
   widgetClassPath=&DeadzoneAdjustment
   bIsVariable=true
   parent=Overlay_DeadzoneAdjustment
   widget=DeadzoneAdjustment_Look
   properties {bIsFocusable=false, settingType=LookStickDeadZone, visibility=Visible}
   slot {class=OverlaySlot, properties {horizontalAlignment=HAlign_Right, verticalAlignment=VAlign_Center, padding=<ltrb 0 25 0 0>}}
  }
  FortRichTextBlock_DeadzoneWarning {
   widgetClassPath=/Script/FortniteUI.FortRichTextBlockLegacy
   bIsVariable=true
   parent=Overlay_DeadzoneAdjustment
   widget=FortRichTextBlock_DeadzoneWarning
   properties {
    justification=Center
    styleSet=/FortUILibrary/Foundation/Buttons/TextStyles/TextStyle_Gamepad_Bindings
    text=WARNING: Very low deadzones can cause drift. Use at your own risk.
    visibility=SelfHitTestInvisible
   }
   slot {class=OverlaySlot, properties {horizontalAlignment=HAlign_Center, verticalAlignment=VAlign_Bottom, padding=<ltrb 0 0 0 -50>}}
  }
  ModeBox {
   widgetClassPath=&SizeBox
   parent=Overlay_4
   widget=ModeBox
   properties {bOverride_WidthOverride=true, heightOverride=500, widthOverride=650}
   slot {class=OverlaySlot, properties {horizontalAlignment=HAlign_Center, verticalAlignment=VAlign_Fill}}
  }
  VerticalBox_48 {widgetClassPath=&VerticalBox, parent=ModeBox, widget=VerticalBox_48, slot {class=SizeBoxSlot}}
  Text_Mode {
   widgetClassPath=&CommonTextBlock
   bIsVariable=true
   parent=VerticalBox_48
   widget=Text_Mode
   properties {
    autoWrapText=true
    justification=Center
    style=/FortUILibrary/Styles/Typography/Foundation/TextStyle-Header-M-NavyBlue.TextStyle-Header-M-NavyBlue_C
    text=Mode
    colorAndOpacity {colorUseRule=UseColor_Specified, specifiedColor=<rgba 0.0176 0.0222 0.0497 1>}
    font {
     bForceMonospaced=false
     bMaterialIsStencil=false
     fontMaterial=null
     fontObject=&Burbank
     letterSpacing=0
     monospacedWidth=1
     size=32
     skewAmount=0
     typefaceFontName=Burbank-Black
     outlineSettings {
      bApplyOutlineToDropShadows=false
      bMiteredCorners=false
      bSeparateFillAlpha=false
      outlineMaterial=null
      outlineSize=0
      outlineColor=<rgba 0 0 0 1>
     }
    }
    shadowOffset=<0 0>
   }
   slot {class=VerticalBoxSlot, properties {verticalAlignment=VAlign_Top}}
  }
  HorizontalTabList_CustomModes {
   widgetClassPath=/Game/UI/Foundation/TabbedView/HorizontalTabList_Legacy
   bIsVariable=true
   parent=VerticalBox_48
   widget=HorizontalTabList_CustomModes
   properties {
    bAutoListenForInput=true
    buttonStyle=/FortUILibrary/Foundation/Buttons/ButtonStyles/Tab/ButtonStyle-Tab-Manage.ButtonStyle-Tab-Manage_C
    onTabButtonCreation=(&self:WidgetTree.HorizontalTabList_CustomModes.HandleOnTabButtonCreated)
    onTabButtonRemoval=(&self:WidgetTree.HorizontalTabList_CustomModes.HandleOnTabButtonRemoved)
    useButtonStyle=true
    nextTabInputActionData {dataTable=&InputActionDataTable, rowName=SecondaryTabRight}
    preregisteredTabInfoArray [
     {
      bAllowedInZone=true
      bHidden=false
      tabButtonType=/FortUILibrary/Foundation/Buttons/IconTabButton_Legacy.IconTabButton_Legacy_C
      tabContentType=null
      tabNameId=CombatModeId
      tabScene=null
      tabLabelInfo {
       displayName=Combat
       iconResource=null
       iconBrush {
        bIsDynamicallyLoaded=false
        drawAs=Image
        imageType=NoImage
        mirroring=NoMirror
        resourceName=None
        resourceObject=/Game/UI/Foundation/Textures/Icons/GamepadOptions/CombatMode
        tiling=NoTile
        imageSize=<64 64>
        margin=<ltrb 0 0 0 0>
        outlineSettings {
         bUseBrushTransparency=false
         roundingType=HalfHeightRadius
         width=0
         color {colorUseRule=UseColor_Specified, specifiedColor=<rgba 0 0 0 0>}
         cornerRadii=<0 0 0 1>
        }
        tintColor {colorUseRule=UseColor_Specified, specifiedColor=<rgba 1 1 1 1>}
        uVRegion {bIsValid=false, max=<0 0>, min=<0 0>}
       }
       priorityData {priority=Priority_3, whenContesting=QueueBehind}
      }
     }
     {
      bAllowedInZone=true
      bHidden=false
      tabButtonType=&IconTabButton_Legacy_C
      tabContentType=null
      tabNameId=BuildModeId
      tabScene=null
      tabLabelInfo {
       displayName=Build
       iconResource=null
       iconBrush {
        bIsDynamicallyLoaded=false
        drawAs=Image
        imageType=NoImage
        mirroring=NoMirror
        resourceName=None
        resourceObject=/Game/UI/Foundation/Textures/Icons/GamepadOptions/BuildMode
        tiling=NoTile
        imageSize=<64 64>
        margin=<ltrb 0 0 0 0>
        outlineSettings {
         bUseBrushTransparency=false
         roundingType=HalfHeightRadius
         width=0
         color {colorUseRule=UseColor_Specified, specifiedColor=<rgba 0 0 0 0>}
         cornerRadii=<0 0 0 1>
        }
        tintColor {colorUseRule=UseColor_Specified, specifiedColor=<rgba 1 1 1 1>}
        uVRegion {bIsValid=false, max=<0 0>, min=<0 0>}
       }
       priorityData {priority=Priority_3, whenContesting=QueueBehind}
      }
     }
     {
      bAllowedInZone=true
      bHidden=false
      tabButtonType=&IconTabButton_Legacy_C
      tabContentType=null
      tabNameId=EditModeId
      tabScene=null
      tabLabelInfo {
       displayName=Edit
       iconResource=null
       iconBrush {
        bIsDynamicallyLoaded=false
        drawAs=Image
        imageType=NoImage
        mirroring=NoMirror
        resourceName=None
        resourceObject=/FortniteCoreUI/UI/Foundation/Textures/Icons/GamepadOptions/EditMode
        tiling=NoTile
        imageSize=<128 128>
        margin=<ltrb 0 0 0 0>
        outlineSettings {
         bUseBrushTransparency=false
         roundingType=HalfHeightRadius
         width=0
         color {colorUseRule=UseColor_Specified, specifiedColor=<rgba 0 0 0 0>}
         cornerRadii=<0 0 0 1>
        }
        tintColor {colorUseRule=UseColor_Specified, specifiedColor=<rgba 1 1 1 1>}
        uVRegion {bIsValid=false, max=<0 0>, min=<0 0>}
       }
       priorityData {priority=Priority_3, whenContesting=QueueBehind}
      }
     }
     {
      bAllowedInZone=true
      bHidden=false
      tabButtonType=&IconTabButton_Legacy_C
      tabContentType=null
      tabNameId=LTMAbilitiesId
      tabScene=null
      tabLabelInfo {
       displayName=LTM Abilities
       iconResource=null
       iconBrush {
        bIsDynamicallyLoaded=false
        drawAs=Image
        imageType=NoImage
        mirroring=NoMirror
        resourceName=None
        resourceObject=&EditMode
        tiling=NoTile
        imageSize=<128 128>
        margin=<ltrb 0 0 0 0>
        outlineSettings {
         bUseBrushTransparency=false
         roundingType=HalfHeightRadius
         width=0
         color {colorUseRule=UseColor_Specified, specifiedColor=<rgba 0 0 0 0>}
         cornerRadii=<0 0 0 1>
        }
        tintColor {colorUseRule=UseColor_Specified, specifiedColor=<rgba 1 1 1 1>}
        uVRegion {bIsValid=false, max=<0 0>, min=<0 0>}
       }
       priorityData {priority=Priority_3, whenContesting=QueueBehind}
      }
     }
     {
      bAllowedInZone=true
      bHidden=false
      tabButtonType=&IconTabButton_Legacy_C
      tabContentType=null
      tabNameId=DeadZoneMoveId
      tabScene=null
      tabLabelInfo {
       displayName=Move Dead Zone
       iconResource=null
       iconBrush {
        bIsDynamicallyLoaded=false
        drawAs=Image
        imageType=NoImage
        mirroring=NoMirror
        resourceName=None
        resourceObject=/Game/UI/Foundation/Textures/Icons/GamepadOptions/DeadZone
        tiling=NoTile
        imageSize=<64 64>
        margin=<ltrb 0 0 0 0>
        outlineSettings {
         bUseBrushTransparency=false
         roundingType=HalfHeightRadius
         width=0
         color {colorUseRule=UseColor_Specified, specifiedColor=<rgba 0 0 0 0>}
         cornerRadii=<0 0 0 1>
        }
        tintColor {colorUseRule=UseColor_Specified, specifiedColor=<rgba 1 1 1 1>}
        uVRegion {bIsValid=false, max=<0 0>, min=<0 0>}
       }
       priorityData {priority=Priority_3, whenContesting=QueueBehind}
      }
     }
    ]
    previousTabInputActionData {dataTable=&InputActionDataTable, rowName=SecondaryTabLeft}
   }
   slot {class=VerticalBoxSlot, properties {horizontalAlignment=HAlign_Center, verticalAlignment=VAlign_Top}}
  }
  ScaleBox_4 {
   widgetClassPath=&ScaleBox
   parent=VerticalBox_48
   widget=ScaleBox_4
   slot {class=VerticalBoxSlot, properties {horizontalAlignment=HAlign_Center, verticalAlignment=VAlign_Top}}
  }
  Image_Gamepad {
   widgetClassPath=&Image
   bIsVariable=true
   parent=ScaleBox_4
   widget=Image_Gamepad
   properties {
    visibility=Hidden
    brush {
     bIsDynamicallyLoaded=false
     drawAs=Image
     imageType=NoImage
     mirroring=NoMirror
     resourceName=None
     resourceObject=&Generic_ControllerConfig
     tiling=NoTile
     imageSize=<1024 512>
     margin=<ltrb 0 0 0 0>
     outlineSettings {
      bUseBrushTransparency=false
      roundingType=HalfHeightRadius
      width=0
      color {colorUseRule=UseColor_Specified, specifiedColor=<rgba 0 0 0 0>}
      cornerRadii=<0 0 0 0>
     }
     tintColor {colorUseRule=UseColor_Specified, specifiedColor=<rgba 1 1 1 1>}
     uVRegion {bIsValid=false, max=<0 0>, min=<0 0>}
    }
   }
   slot {class=ScaleBoxSlot}
  }
  VerticalBoxMenu+Map {
   widgetClassPath=&VerticalBox
   bIsVariable=true
   parent=VerticalBox_48
   widget=VerticalBoxMenu+Map
   slot {class=VerticalBoxSlot, properties {horizontalAlignment=HAlign_Center, verticalAlignment=VAlign_Top}}
  }
  SizeBox_543 {
   widgetClassPath=&SizeBox
   parent=VerticalBoxMenu+Map
   widget=SizeBox_543
   properties {bOverride_HeightOverride=true, bOverride_WidthOverride=true, heightOverride=44, widthOverride=512}
   slot {class=VerticalBoxSlot, properties {verticalAlignment=VAlign_Top}}
  }
  CommonBorder_509 {
   widgetClassPath=&CommonBorder
   parent=SizeBox_543
   widget=CommonBorder_509
   properties {style=/FortUILibrary/Styles/Border/Border-SolidBG-StatItem.Border-SolidBG-StatItem_C, padding=<ltrb 15 0 0 0>}
   slot {class=SizeBoxSlot}
  }
  HorizontalBox_134 {
   widgetClassPath=&HorizontalBox
   parent=CommonBorder_509
   widget=HorizontalBox_134
   slot {class=BorderSlot, properties {padding=<ltrb 15 0 0 0>}}
  }
  SpecialRight {
   widgetClassPath=/FortUILibrary/Input/KeybindWidget
   bIsVariable=true
   parent=HorizontalBox_134
   widget=SpecialRight
   properties {boundKeyFallback=Gamepad_Special_Right, inputTypeOverride=Gamepad}
   slot {class=HorizontalBoxSlot}
  }
  Text_Gamepad_Start {
   widgetClassPath=&CommonTextBlock
   bIsVariable=true
   parent=HorizontalBox_134
   widget=Text_Gamepad_Start
   properties {
    scrollStyle=&TextScrollStyle-Base_C
    style=/FortUILibrary/Styles/Typography/Foundation/TextStyle-Base-XS-B-DarkBlueFade.TextStyle-Base-XS-B-DarkBlueFade_C
    text=Start
    colorAndOpacity {colorUseRule=UseColor_Specified, specifiedColor=<rgba 0.008 0.011 0.0262 0.8>}
    font {
     bForceMonospaced=false
     bMaterialIsStencil=false
     fontMaterial=null
     fontObject=&NotoSans
     letterSpacing=0
     monospacedWidth=1
     size=16
     skewAmount=0
     typefaceFontName=NotoSans-Bold
     outlineSettings {
      bApplyOutlineToDropShadows=false
      bMiteredCorners=false
      bSeparateFillAlpha=false
      outlineMaterial=null
      outlineSize=0
      outlineColor=<rgba 0 0 0 1>
     }
    }
    shadowOffset=<0 0>
   }
   slot {class=HorizontalBoxSlot, properties {verticalAlignment=VAlign_Center, padding=<ltrb 10 0 12 0>, size {sizeRule=Fill, value=1}}}
  }
  SizeBox_544 {
   widgetClassPath=&SizeBox
   parent=VerticalBoxMenu+Map
   widget=SizeBox_544
   properties {bOverride_WidthOverride=true, heightOverride=4, widthOverride=512}
   slot {class=VerticalBoxSlot, properties {verticalAlignment=VAlign_Top}}
  }
  CommonBorder_510 {
   widgetClassPath=&CommonBorder
   parent=SizeBox_544
   widget=CommonBorder_510
   properties {style=/FortUILibrary/Styles/Border/Border-ItemInfo-Locked.Border-ItemInfo-Locked_C, verticalAlignment=VAlign_Center, padding=<ltrb 0 0 0 0>}
   slot {class=SizeBoxSlot}
  }
  SizeBox_545 {
   widgetClassPath=&SizeBox
   parent=VerticalBoxMenu+Map
   widget=SizeBox_545
   properties {bOverride_HeightOverride=true, bOverride_WidthOverride=true, heightOverride=76, widthOverride=512}
   slot {class=VerticalBoxSlot, properties {verticalAlignment=VAlign_Top}}
  }
  CommonBorder_511 {
   widgetClassPath=&CommonBorder
   parent=SizeBox_545
   widget=CommonBorder_511
   properties {
    style=&Border-SolidBG-StatItem_C
    verticalAlignment=VAlign_Center
    padding=<ltrb 15 0 0 0>
   }
   slot {class=SizeBoxSlot}
  }
  HorizontalBox_14 {
   widgetClassPath=&HorizontalBox
   parent=CommonBorder_511
   widget=HorizontalBox_14
   slot {class=BorderSlot, properties {verticalAlignment=VAlign_Center, padding=<ltrb 15 0 0 0>}}
  }
  SpecialLeft {
   widgetClassPath=&KeybindWidget
   bIsVariable=true
   parent=HorizontalBox_14
   widget=SpecialLeft
   properties {boundKeyFallback=Gamepad_Special_Left, inputTypeOverride=Gamepad}
   slot {class=HorizontalBoxSlot}
  }
  VerticalBox_0 {
   widgetClassPath=&VerticalBox
   parent=HorizontalBox_14
   widget=VerticalBox_0
   slot {class=HorizontalBoxSlot, properties {padding=<ltrb 10 0 0 0>, size {sizeRule=Fill, value=1}}}
  }
  HorizontalBox_1 {widgetClassPath=&HorizontalBox, parent=VerticalBox_0, widget=HorizontalBox_1, slot {class=VerticalBoxSlot}}
  Text_Gamepad_Select_1 {
   widgetClassPath=&CommonTextBlock
   bIsVariable=true
   parent=HorizontalBox_1
   widget=Text_Gamepad_Select_1
   properties {
    scrollStyle=&TextScrollStyle-Base_C
    style=&TextStyle-Base-XS-B-DarkBlueFade_C
    text=Select 1
    colorAndOpacity {colorUseRule=UseColor_Specified, specifiedColor=<rgba 0.008 0.011 0.0262 0.8>}
    font {
     bForceMonospaced=false
     bMaterialIsStencil=false
     fontMaterial=null
     fontObject=&NotoSans
     letterSpacing=0
     monospacedWidth=1
     size=16
     skewAmount=0
     typefaceFontName=NotoSans-Bold
     outlineSettings {
      bApplyOutlineToDropShadows=false
      bMiteredCorners=false
      bSeparateFillAlpha=false
      outlineMaterial=null
      outlineSize=0
      outlineColor=<rgba 0 0 0 1>
     }
    }
    shadowOffset=<0 0>
   }
   slot {class=HorizontalBoxSlot, properties {padding=<ltrb 0 0 12 0>, size {sizeRule=Fill, value=1}}}
  }
  HorizontalBox_2 {widgetClassPath=&HorizontalBox, parent=VerticalBox_0, widget=HorizontalBox_2, slot {class=VerticalBoxSlot}}
  Text_Gamepad_Select_2 {
   widgetClassPath=&CommonTextBlock
   bIsVariable=true
   parent=HorizontalBox_2
   widget=Text_Gamepad_Select_2
   properties {
    scrollStyle=&TextScrollStyle-Base_C
    style=&TextStyle-Base-XS-B-DarkBlueFade_C
    text=Select 2
    colorAndOpacity {colorUseRule=UseColor_Specified, specifiedColor=<rgba 0.008 0.011 0.0262 0.8>}
    font {
     bForceMonospaced=false
     bMaterialIsStencil=false
     fontMaterial=null
     fontObject=&NotoSans
     letterSpacing=0
     monospacedWidth=1
     size=16
     skewAmount=0
     typefaceFontName=NotoSans-Bold
     outlineSettings {
      bApplyOutlineToDropShadows=false
      bMiteredCorners=false
      bSeparateFillAlpha=false
      outlineMaterial=null
      outlineSize=0
      outlineColor=<rgba 0 0 0 1>
     }
    }
    shadowOffset=<0 0>
   }
   slot {class=HorizontalBoxSlot, properties {padding=<ltrb 0 0 12 0>, size {sizeRule=Fill, value=1}}}
  }
  HorizontalBox_AllText {
   widgetClassPath=&HorizontalBox
   bIsVariable=true
   parent=Overlay_4
   widget=HorizontalBox_AllText
   slot {class=OverlaySlot, properties {horizontalAlignment=HAlign_Fill, verticalAlignment=VAlign_Fill}}
  }
  VerticalBox_19 {
   widgetClassPath=&VerticalBox
   parent=HorizontalBox_AllText
   widget=VerticalBox_19
   slot {class=HorizontalBoxSlot, properties {verticalAlignment=VAlign_Top, padding=<ltrb 12 22 0 0>, size {sizeRule=Fill, value=1}}}
  }
  SizeBox_509 {
   widgetClassPath=&SizeBox
   parent=VerticalBox_19
   widget=SizeBox_509
   properties {bOverride_HeightOverride=true, heightOverride=44}
   slot {class=VerticalBoxSlot}
  }
  CommonBorder_496 {
   widgetClassPath=&CommonBorder
   parent=SizeBox_509
   widget=CommonBorder_496
   properties {
    style=&Border-SolidBG-StatItem_C
    verticalAlignment=VAlign_Center
    padding=<ltrb 32 0 15 0>
   }
   slot {class=SizeBoxSlot}
  }
  HorizontalBox_48 {
   widgetClassPath=&HorizontalBox
   parent=CommonBorder_496
   widget=HorizontalBox_48
   slot {class=BorderSlot, properties {verticalAlignment=VAlign_Center, padding=<ltrb 32 0 15 0>}}
  }
  Text_Gamepad_LT {
   widgetClassPath=&CommonTextBlock
   bIsVariable=true
   parent=HorizontalBox_48
   widget=Text_Gamepad_LT
   properties {
    justification=Right
    scrollStyle=&TextScrollStyle-Base_C
    style=&TextStyle-Base-XS-B-DarkBlueFade_C
    text=Left Trigger
    colorAndOpacity {colorUseRule=UseColor_Specified, specifiedColor=<rgba 0.008 0.011 0.0262 0.8>}
    font {
     bForceMonospaced=false
     bMaterialIsStencil=false
     fontMaterial=null
     fontObject=&NotoSans
     letterSpacing=0
     monospacedWidth=1
     size=16
     skewAmount=0
     typefaceFontName=NotoSans-Bold
     outlineSettings {
      bApplyOutlineToDropShadows=false
      bMiteredCorners=false
      bSeparateFillAlpha=false
      outlineMaterial=null
      outlineSize=0
      outlineColor=<rgba 0 0 0 1>
     }
    }
    shadowOffset=<0 0>
   }
   slot {
    class=HorizontalBoxSlot
    properties {horizontalAlignment=HAlign_Right, verticalAlignment=VAlign_Center, padding=<ltrb 0 0 10 0>, size {sizeRule=Fill, value=1}}
   }
  }
  LT {
   widgetClassPath=&KeybindWidget
   bIsVariable=true
   parent=HorizontalBox_48
   widget=LT
   properties {boundKeyFallback=Gamepad_LeftTrigger, inputTypeOverride=Gamepad}
   slot {class=HorizontalBoxSlot}
  }
  SizeBox_514 {
   widgetClassPath=&SizeBox
   parent=VerticalBox_19
   widget=SizeBox_514
   properties {bOverride_HeightOverride=true, heightOverride=4}
   slot {class=VerticalBoxSlot}
  }
  CommonBorder_497 {
   widgetClassPath=&CommonBorder
   parent=SizeBox_514
   widget=CommonBorder_497
   properties {style=&Border-ItemInfo-Locked_C, verticalAlignment=VAlign_Center, padding=<ltrb 0 0 56 0>}
   slot {class=SizeBoxSlot}
  }
  SizeBox_530 {
   widgetClassPath=&SizeBox
   parent=VerticalBox_19
   widget=SizeBox_530
   properties {bOverride_HeightOverride=true, heightOverride=44}
   slot {class=VerticalBoxSlot}
  }
  CommonBorder_498 {
   widgetClassPath=&CommonBorder
   parent=SizeBox_530
   widget=CommonBorder_498
   properties {
    style=&Border-SolidBG-StatItem_C
    verticalAlignment=VAlign_Center
    padding=<ltrb 32 0 15 0>
   }
   slot {class=SizeBoxSlot}
  }
  HorizontalBox_49 {
   widgetClassPath=&HorizontalBox
   parent=CommonBorder_498
   widget=HorizontalBox_49
   slot {class=BorderSlot, properties {verticalAlignment=VAlign_Center, padding=<ltrb 32 0 15 0>}}
  }
  Text_Gamepad_LB {
   widgetClassPath=&CommonTextBlock
   bIsVariable=true
   parent=HorizontalBox_49
   widget=Text_Gamepad_LB
   properties {
    justification=Right
    scrollStyle=&TextScrollStyle-Base_C
    style=&TextStyle-Base-XS-B-DarkBlueFade_C
    text=Left Bumper
    colorAndOpacity {colorUseRule=UseColor_Specified, specifiedColor=<rgba 0.008 0.011 0.0262 0.8>}
    font {
     bForceMonospaced=false
     bMaterialIsStencil=false
     fontMaterial=null
     fontObject=&NotoSans
     letterSpacing=0
     monospacedWidth=1
     size=16
     skewAmount=0
     typefaceFontName=NotoSans-Bold
     outlineSettings {
      bApplyOutlineToDropShadows=false
      bMiteredCorners=false
      bSeparateFillAlpha=false
      outlineMaterial=null
      outlineSize=0
      outlineColor=<rgba 0 0 0 1>
     }
    }
    shadowOffset=<0 0>
   }
   slot {
    class=HorizontalBoxSlot
    properties {horizontalAlignment=HAlign_Right, verticalAlignment=VAlign_Center, padding=<ltrb 0 0 10 0>, size {sizeRule=Fill, value=1}}
   }
  }
  LS {
   widgetClassPath=&KeybindWidget
   bIsVariable=true
   parent=HorizontalBox_49
   widget=LS
   properties {boundKeyFallback=Gamepad_LeftShoulder, inputTypeOverride=Gamepad}
   slot {class=HorizontalBoxSlot}
  }
  SizeBox {
   widgetClassPath=&SizeBox
   parent=VerticalBox_19
   widget=SizeBox
   properties {bOverride_HeightOverride=true, heightOverride=4}
   slot {class=VerticalBoxSlot}
  }
  CommonBorder-LeftBumperSeparator {
   widgetClassPath=&CommonBorder
   bIsVariable=true
   parent=SizeBox
   widget=CommonBorder-LeftBumperSeparator
   properties {style=&Border-ItemInfo-Locked_C, verticalAlignment=VAlign_Center, padding=<ltrb 0 0 56 0>}
   slot {class=SizeBoxSlot}
  }
  HorizontalBox-LB+RB {widgetClassPath=&HorizontalBox, bIsVariable=true, parent=VerticalBox_19, widget=HorizontalBox-LB+RB, slot {class=VerticalBoxSlot}}
  SizeBox_1 {
   widgetClassPath=&SizeBox
   parent=HorizontalBox-LB+RB
   widget=SizeBox_1
   properties {bOverride_HeightOverride=true, heightOverride=44}
   slot {class=HorizontalBoxSlot, properties {size {sizeRule=Fill, value=1}}}
  }
  CommonBorder_2 {
   widgetClassPath=&CommonBorder
   parent=SizeBox_1
   widget=CommonBorder_2
   properties {
    style=&Border-SolidBG-StatItem_C
    verticalAlignment=VAlign_Center
    padding=<ltrb 32 0 12 0>
   }
   slot {class=SizeBoxSlot}
  }
  HorizontalBox_9 {
   widgetClassPath=&HorizontalBox
   parent=CommonBorder_2
   widget=HorizontalBox_9
   slot {class=BorderSlot, properties {verticalAlignment=VAlign_Center, padding=<ltrb 32 0 12 0>}}
  }
  LB+RBAction {
   widgetClassPath=&CommonTextBlock
   bIsVariable=true
   parent=HorizontalBox_9
   widget=LB+RBAction
   properties {
    justification=Right
    scrollStyle=&TextScrollStyle-Base_C
    style=&TextStyle-Base-XS-B-DarkBlueFade_C
    text=L1 + R1 Ability
    colorAndOpacity {colorUseRule=UseColor_Specified, specifiedColor=<rgba 0.008 0.011 0.0262 0.8>}
    font {
     bForceMonospaced=false
     bMaterialIsStencil=false
     fontMaterial=null
     fontObject=&NotoSans
     letterSpacing=0
     monospacedWidth=1
     size=16
     skewAmount=0
     typefaceFontName=NotoSans-Bold
     outlineSettings {
      bApplyOutlineToDropShadows=false
      bMiteredCorners=false
      bSeparateFillAlpha=false
      outlineMaterial=null
      outlineSize=0
      outlineColor=<rgba 0 0 0 1>
     }
    }
    shadowOffset=<0 0>
   }
   slot {
    class=HorizontalBoxSlot
    properties {horizontalAlignment=HAlign_Right, verticalAlignment=VAlign_Center, padding=<ltrb 0 0 12 0>, size {sizeRule=Fill, value=1}}
   }
  }
  HorizontalBox_5 {widgetClassPath=&HorizontalBox, parent=HorizontalBox_9, widget=HorizontalBox_5, slot {class=HorizontalBoxSlot}}
  LSCombo1 {
   widgetClassPath=&KeybindWidget
   bIsVariable=true
   parent=HorizontalBox_5
   widget=LSCombo1
   properties {boundKeyFallback=Gamepad_LeftShoulder, inputTypeOverride=Gamepad}
   slot {class=HorizontalBoxSlot}
  }
  CommonTextBlock_1 {
   widgetClassPath=&CommonTextBlock
   parent=HorizontalBox_5
   widget=CommonTextBlock_1
   properties {
    justification=Center
    style=&TextStyle-Base-XS-B-DarkBlueFade_C
    text=+
    colorAndOpacity {colorUseRule=UseColor_Specified, specifiedColor=<rgba 0.008 0.011 0.0262 0.8>}
    font {
     bForceMonospaced=false
     bMaterialIsStencil=false
     fontMaterial=null
     fontObject=&NotoSans
     letterSpacing=0
     monospacedWidth=1
     size=16
     skewAmount=0
     typefaceFontName=NotoSans-Bold
     outlineSettings {
      bApplyOutlineToDropShadows=false
      bMiteredCorners=false
      bSeparateFillAlpha=false
      outlineMaterial=null
      outlineSize=0
      outlineColor=<rgba 0 0 0 1>
     }
    }
    shadowOffset=<0 0>
   }
   slot {class=HorizontalBoxSlot, properties {verticalAlignment=VAlign_Center, padding=<ltrb 8 0 8 0>}}
  }
  RSCombo {
   widgetClassPath=&KeybindWidget
   bIsVariable=true
   parent=HorizontalBox_5
   widget=RSCombo
   properties {boundKeyFallback=Gamepad_RightShoulder, inputTypeOverride=Gamepad}
   slot {class=HorizontalBoxSlot}
  }
  Separator-LB+RBfromLB+RT {
   widgetClassPath=&SizeBox
   bIsVariable=true
   parent=VerticalBox_19
   widget=Separator-LB+RBfromLB+RT
   properties {bOverride_HeightOverride=true, heightOverride=4}
   slot {class=VerticalBoxSlot}
  }
  CommonBorder_4 {
   widgetClassPath=&CommonBorder
   parent=Separator-LB+RBfromLB+RT
   widget=CommonBorder_4
   properties {style=&Border-ItemInfo-Locked_C, verticalAlignment=VAlign_Center, padding=<ltrb 0 0 56 0>}
   slot {class=SizeBoxSlot}
  }
  HorizontalBox-LB+RT {widgetClassPath=&HorizontalBox, bIsVariable=true, parent=VerticalBox_19, widget=HorizontalBox-LB+RT, slot {class=VerticalBoxSlot}}
  SizeBox_4 {
   widgetClassPath=&SizeBox
   parent=HorizontalBox-LB+RT
   widget=SizeBox_4
   properties {bOverride_HeightOverride=true, heightOverride=44}
   slot {class=HorizontalBoxSlot, properties {size {sizeRule=Fill, value=1}}}
  }
  CommonBorder_5 {
   widgetClassPath=&CommonBorder
   parent=SizeBox_4
   widget=CommonBorder_5
   properties {
    style=&Border-SolidBG-StatItem_C
    verticalAlignment=VAlign_Center
    padding=<ltrb 32 0 12 0>
   }
   slot {class=SizeBoxSlot}
  }
  HorizontalBox_10 {
   widgetClassPath=&HorizontalBox
   parent=CommonBorder_5
   widget=HorizontalBox_10
   slot {class=BorderSlot, properties {verticalAlignment=VAlign_Center, padding=<ltrb 32 0 12 0>}}
  }
  LB+RTAction {
   widgetClassPath=&CommonTextBlock
   bIsVariable=true
   parent=HorizontalBox_10
   widget=LB+RTAction
   properties {
    justification=Right
    scrollStyle=&TextScrollStyle-Base_C
    style=&TextStyle-Base-XS-B-DarkBlueFade_C
    text=L1 + R2 Ability
    colorAndOpacity {colorUseRule=UseColor_Specified, specifiedColor=<rgba 0.008 0.011 0.0262 0.8>}
    font {
     bForceMonospaced=false
     bMaterialIsStencil=false
     fontMaterial=null
     fontObject=&NotoSans
     letterSpacing=0
     monospacedWidth=1
     size=16
     skewAmount=0
     typefaceFontName=NotoSans-Bold
     outlineSettings {
      bApplyOutlineToDropShadows=false
      bMiteredCorners=false
      bSeparateFillAlpha=false
      outlineMaterial=null
      outlineSize=0
      outlineColor=<rgba 0 0 0 1>
     }
    }
    shadowOffset=<0 0>
   }
   slot {
    class=HorizontalBoxSlot
    properties {horizontalAlignment=HAlign_Right, verticalAlignment=VAlign_Center, padding=<ltrb 0 0 12 0>, size {sizeRule=Fill, value=1}}
   }
  }
  HorizontalBox_6 {widgetClassPath=&HorizontalBox, parent=HorizontalBox_10, widget=HorizontalBox_6, slot {class=HorizontalBoxSlot}}
  LSCombo2 {
   widgetClassPath=&KeybindWidget
   bIsVariable=true
   parent=HorizontalBox_6
   widget=LSCombo2
   properties {boundKeyFallback=Gamepad_LeftShoulder, inputTypeOverride=Gamepad}
   slot {class=HorizontalBoxSlot}
  }
  CommonTextBlock_0 {
   widgetClassPath=&CommonTextBlock
   parent=HorizontalBox_6
   widget=CommonTextBlock_0
   properties {
    justification=Center
    style=&TextStyle-Base-XS-B-DarkBlueFade_C
    text=+
    colorAndOpacity {colorUseRule=UseColor_Specified, specifiedColor=<rgba 0.008 0.011 0.0262 0.8>}
    font {
     bForceMonospaced=false
     bMaterialIsStencil=false
     fontMaterial=null
     fontObject=&NotoSans
     letterSpacing=0
     monospacedWidth=1
     size=16
     skewAmount=0
     typefaceFontName=NotoSans-Bold
     outlineSettings {
      bApplyOutlineToDropShadows=false
      bMiteredCorners=false
      bSeparateFillAlpha=false
      outlineMaterial=null
      outlineSize=0
      outlineColor=<rgba 0 0 0 1>
     }
    }
    shadowOffset=<0 0>
   }
   slot {class=HorizontalBoxSlot, properties {verticalAlignment=VAlign_Center, padding=<ltrb 8 0 8 0>}}
  }
  RTCombo {
   widgetClassPath=&KeybindWidget
   bIsVariable=true
   parent=HorizontalBox_6
   widget=RTCombo
   properties {boundKeyFallback=Gamepad_RightTrigger, inputTypeOverride=Gamepad}
   slot {class=HorizontalBoxSlot}
  }
  Spacer_45 {widgetClassPath=&Spacer, parent=VerticalBox_19, widget=Spacer_45, properties {size=<1 40>}, slot {class=VerticalBoxSlot}}
  VerticalBox_Xbox_LS_And_Dpad {
   widgetClassPath=&VerticalBox
   bIsVariable=true
   parent=VerticalBox_19
   widget=VerticalBox_Xbox_LS_And_Dpad
   slot {class=VerticalBoxSlot}
  }
  SizeBox_Xbox_LS {
   widgetClassPath=&SizeBox
   parent=VerticalBox_Xbox_LS_And_Dpad
   widget=SizeBox_Xbox_LS
   properties {bOverride_HeightOverride=true, heightOverride=44}
   slot {class=VerticalBoxSlot}
  }
  CommonBorder_499 {
   widgetClassPath=&CommonBorder
   parent=SizeBox_Xbox_LS
   widget=CommonBorder_499
   properties {
    horizontalAlignment=HAlign_Right
    style=&Border-SolidBG-StatItem_C
    verticalAlignment=VAlign_Center
    padding=<ltrb 32 0 15 0>
   }
   slot {class=SizeBoxSlot}
  }
  HorizontalBox_75 {
   widgetClassPath=&HorizontalBox
   parent=CommonBorder_499
   widget=HorizontalBox_75
   slot {class=BorderSlot, properties {horizontalAlignment=HAlign_Right, verticalAlignment=VAlign_Center, padding=<ltrb 32 0 15 0>}}
  }
  Text_Xbox_LS {
   widgetClassPath=&CommonTextBlock
   bIsVariable=true
   parent=HorizontalBox_75
   widget=Text_Xbox_LS
   properties {
    justification=Right
    scrollStyle=&TextScrollStyle-Base_C
    style=&TextStyle-Base-XS-B-DarkBlueFade_C
    text=Left Stick
    colorAndOpacity {colorUseRule=UseColor_Specified, specifiedColor=<rgba 0.008 0.011 0.0262 0.8>}
    font {
     bForceMonospaced=false
     bMaterialIsStencil=false
     fontMaterial=null
     fontObject=&NotoSans
     letterSpacing=0
     monospacedWidth=1
     size=16
     skewAmount=0
     typefaceFontName=NotoSans-Bold
     outlineSettings {
      bApplyOutlineToDropShadows=false
      bMiteredCorners=false
      bSeparateFillAlpha=false
      outlineMaterial=null
      outlineSize=0
      outlineColor=<rgba 0 0 0 1>
     }
    }
    shadowOffset=<0 0>
   }
   slot {
    class=HorizontalBoxSlot
    properties {horizontalAlignment=HAlign_Right, verticalAlignment=VAlign_Center, padding=<ltrb 0 0 10 0>, size {sizeRule=Fill, value=1}}
   }
  }
  LStick {
   widgetClassPath=&KeybindWidget
   bIsVariable=true
   parent=HorizontalBox_75
   widget=LStick
   properties {boundKeyFallback=Gamepad_LeftX, inputTypeOverride=Gamepad}
   slot {class=HorizontalBoxSlot}
  }
  SizeBox_532 {
   widgetClassPath=&SizeBox
   parent=VerticalBox_Xbox_LS_And_Dpad
   widget=SizeBox_532
   properties {bOverride_HeightOverride=true, heightOverride=4}
   slot {class=VerticalBoxSlot}
  }
  CommonBorder_500 {
   widgetClassPath=&CommonBorder
   parent=SizeBox_532
   widget=CommonBorder_500
   properties {style=&Border-ItemInfo-Locked_C, verticalAlignment=VAlign_Center, padding=<ltrb 0 0 56 0>}
   slot {class=SizeBoxSlot}
  }
  SizeBox_Xbox_LS_Click {
   widgetClassPath=&SizeBox
   parent=VerticalBox_Xbox_LS_And_Dpad
   widget=SizeBox_Xbox_LS_Click
   properties {bOverride_HeightOverride=true, heightOverride=76}
   slot {class=VerticalBoxSlot}
  }
  CommonBorder_501 {
   widgetClassPath=&CommonBorder
   parent=SizeBox_Xbox_LS_Click
   widget=CommonBorder_501
   properties {
    horizontalAlignment=HAlign_Right
    style=&Border-SolidBG-StatItem_C
    verticalAlignment=VAlign_Center
    padding=<ltrb 32 0 15 0>
   }
   slot {class=SizeBoxSlot}
  }
  HorizontalBox_12 {
   widgetClassPath=&HorizontalBox
   parent=CommonBorder_501
   widget=HorizontalBox_12
   slot {class=BorderSlot, properties {horizontalAlignment=HAlign_Right, verticalAlignment=VAlign_Center, padding=<ltrb 32 0 15 0>}}
  }
  VerticalBox_2 {
   widgetClassPath=&VerticalBox
   parent=HorizontalBox_12
   widget=VerticalBox_2
   slot {class=HorizontalBoxSlot, properties {padding=<ltrb 0 0 10 0>}}
  }
  HorizontalBox_107 {widgetClassPath=&HorizontalBox, parent=VerticalBox_2, widget=HorizontalBox_107, slot {class=VerticalBoxSlot}}
  Text_Xbox_LS_Click_1 {
   widgetClassPath=&CommonTextBlock
   bIsVariable=true
   parent=HorizontalBox_107
   widget=Text_Xbox_LS_Click_1
   properties {
    justification=Right
    scrollStyle=&TextScrollStyle-Base_C
    style=&TextStyle-Base-XS-B-DarkBlueFade_C
    text=Left Stick Click 1
    colorAndOpacity {colorUseRule=UseColor_Specified, specifiedColor=<rgba 0.008 0.011 0.0262 0.8>}
    font {
     bForceMonospaced=false
     bMaterialIsStencil=false
     fontMaterial=null
     fontObject=&NotoSans
     letterSpacing=0
     monospacedWidth=1
     size=16
     skewAmount=0
     typefaceFontName=NotoSans-Bold
     outlineSettings {
      bApplyOutlineToDropShadows=false
      bMiteredCorners=false
      bSeparateFillAlpha=false
      outlineMaterial=null
      outlineSize=0
      outlineColor=<rgba 0 0 0 1>
     }
    }
    shadowOffset=<0 0>
   }
   slot {class=HorizontalBoxSlot, properties {horizontalAlignment=HAlign_Right, verticalAlignment=VAlign_Center, size {sizeRule=Fill, value=1}}}
  }
  HorizontalBox_4 {widgetClassPath=&HorizontalBox, parent=VerticalBox_2, widget=HorizontalBox_4, slot {class=VerticalBoxSlot}}
  Text_Xbox_LS_Click_2 {
   widgetClassPath=&CommonTextBlock
   bIsVariable=true
   parent=HorizontalBox_4
   widget=Text_Xbox_LS_Click_2
   properties {
    justification=Right
    scrollStyle=&TextScrollStyle-Base_C
    style=&TextStyle-Base-XS-B-DarkBlueFade_C
    text=Left Stick Click 2
    colorAndOpacity {colorUseRule=UseColor_Specified, specifiedColor=<rgba 0.008 0.011 0.0262 0.8>}
    font {
     bForceMonospaced=false
     bMaterialIsStencil=false
     fontMaterial=null
     fontObject=&NotoSans
     letterSpacing=0
     monospacedWidth=1
     size=16
     skewAmount=0
     typefaceFontName=NotoSans-Bold
     outlineSettings {
      bApplyOutlineToDropShadows=false
      bMiteredCorners=false
      bSeparateFillAlpha=false
      outlineMaterial=null
      outlineSize=0
      outlineColor=<rgba 0 0 0 1>
     }
    }
    shadowOffset=<0 0>
   }
   slot {class=HorizontalBoxSlot, properties {horizontalAlignment=HAlign_Right, verticalAlignment=VAlign_Center, size {sizeRule=Fill, value=1}}}
  }
  LStickClick {
   widgetClassPath=&KeybindWidget
   bIsVariable=true
   parent=HorizontalBox_12
   widget=LStickClick
   properties {boundKeyFallback=Gamepad_LeftThumbstick, inputTypeOverride=Gamepad}
   slot {class=HorizontalBoxSlot}
  }
  Spacer_46 {
   widgetClassPath=&Spacer
   parent=VerticalBox_Xbox_LS_And_Dpad
   widget=Spacer_46
   properties {size=<1 40>}
   slot {class=VerticalBoxSlot}
  }
  SizeBox_Xbox_Dpad_Up {
   widgetClassPath=&SizeBox
   parent=VerticalBox_Xbox_LS_And_Dpad
   widget=SizeBox_Xbox_Dpad_Up
   properties {bOverride_HeightOverride=true, heightOverride=44}
   slot {class=VerticalBoxSlot}
  }
  CommonBorder_502 {
   widgetClassPath=&CommonBorder
   parent=SizeBox_Xbox_Dpad_Up
   widget=CommonBorder_502
   properties {
    horizontalAlignment=HAlign_Right
    style=&Border-SolidBG-StatItem_C
    verticalAlignment=VAlign_Center
    padding=<ltrb 32 0 15 0>
   }
   slot {class=SizeBoxSlot}
  }
  HorizontalBox_108 {
   widgetClassPath=&HorizontalBox
   parent=CommonBorder_502
   widget=HorizontalBox_108
   slot {class=BorderSlot, properties {horizontalAlignment=HAlign_Right, verticalAlignment=VAlign_Center, padding=<ltrb 32 0 15 0>}}
  }
  Text_Xbox_Dpad_Up {
   widgetClassPath=&CommonTextBlock
   bIsVariable=true
   parent=HorizontalBox_108
   widget=Text_Xbox_Dpad_Up
   properties {
    justification=Right
    scrollStyle=&TextScrollStyle-Base_C
    style=&TextStyle-Base-XS-B-DarkBlueFade_C
    text=Dpad Up
    colorAndOpacity {colorUseRule=UseColor_Specified, specifiedColor=<rgba 0.008 0.011 0.0262 0.8>}
    font {
     bForceMonospaced=false
     bMaterialIsStencil=false
     fontMaterial=null
     fontObject=&NotoSans
     letterSpacing=0
     monospacedWidth=1
     size=16
     skewAmount=0
     typefaceFontName=NotoSans-Bold
     outlineSettings {
      bApplyOutlineToDropShadows=false
      bMiteredCorners=false
      bSeparateFillAlpha=false
      outlineMaterial=null
      outlineSize=0
      outlineColor=<rgba 0 0 0 1>
     }
    }
    shadowOffset=<0 0>
   }
   slot {
    class=HorizontalBoxSlot
    properties {horizontalAlignment=HAlign_Right, verticalAlignment=VAlign_Center, padding=<ltrb 0 0 10 0>, size {sizeRule=Fill, value=1}}
   }
  }
  DPadUp {
   widgetClassPath=&KeybindWidget
   bIsVariable=true
   parent=HorizontalBox_108
   widget=DPadUp
   properties {boundKeyFallback=Gamepad_DPad_Up, inputTypeOverride=Gamepad}
   slot {class=HorizontalBoxSlot}
  }
  SizeBox_535 {
   widgetClassPath=&SizeBox
   parent=VerticalBox_Xbox_LS_And_Dpad
   widget=SizeBox_535
   properties {bOverride_HeightOverride=true, heightOverride=4}
   slot {class=VerticalBoxSlot}
  }
  CommonBorder_503 {
   widgetClassPath=&CommonBorder
   parent=SizeBox_535
   widget=CommonBorder_503
   properties {style=&Border-ItemInfo-Locked_C, verticalAlignment=VAlign_Center, padding=<ltrb 0 0 56 0>}
   slot {class=SizeBoxSlot}
  }
  SizeBox_Xbox_Dpad_Down {
   widgetClassPath=&SizeBox
   parent=VerticalBox_Xbox_LS_And_Dpad
   widget=SizeBox_Xbox_Dpad_Down
   properties {bOverride_HeightOverride=true, heightOverride=44}
   slot {class=VerticalBoxSlot}
  }
  CommonBorder_504 {
   widgetClassPath=&CommonBorder
   parent=SizeBox_Xbox_Dpad_Down
   widget=CommonBorder_504
   properties {
    horizontalAlignment=HAlign_Right
    style=&Border-SolidBG-StatItem_C
    verticalAlignment=VAlign_Center
    padding=<ltrb 32 0 15 0>
   }
   slot {class=SizeBoxSlot}
  }
  HorizontalBox_109 {
   widgetClassPath=&HorizontalBox
   parent=CommonBorder_504
   widget=HorizontalBox_109
   slot {class=BorderSlot, properties {horizontalAlignment=HAlign_Right, verticalAlignment=VAlign_Center, padding=<ltrb 32 0 15 0>}}
  }
  Text_Xbox_Dpad_Down {
   widgetClassPath=&CommonTextBlock
   bIsVariable=true
   parent=HorizontalBox_109
   widget=Text_Xbox_Dpad_Down
   properties {
    justification=Right
    scrollStyle=&TextScrollStyle-Base_C
    style=&TextStyle-Base-XS-B-DarkBlueFade_C
    text=Dpad Down
    colorAndOpacity {colorUseRule=UseColor_Specified, specifiedColor=<rgba 0.008 0.011 0.0262 0.8>}
    font {
     bForceMonospaced=false
     bMaterialIsStencil=false
     fontMaterial=null
     fontObject=&NotoSans
     letterSpacing=0
     monospacedWidth=1
     size=16
     skewAmount=0
     typefaceFontName=NotoSans-Bold
     outlineSettings {
      bApplyOutlineToDropShadows=false
      bMiteredCorners=false
      bSeparateFillAlpha=false
      outlineMaterial=null
      outlineSize=0
      outlineColor=<rgba 0 0 0 1>
     }
    }
    shadowOffset=<0 0>
   }
   slot {
    class=HorizontalBoxSlot
    properties {horizontalAlignment=HAlign_Right, verticalAlignment=VAlign_Center, padding=<ltrb 0 0 10 0>, size {sizeRule=Fill, value=1}}
   }
  }
  DPadDown {
   widgetClassPath=&KeybindWidget
   bIsVariable=true
   parent=HorizontalBox_109
   widget=DPadDown
   properties {boundKeyFallback=Gamepad_DPad_Down, inputTypeOverride=Gamepad}
   slot {class=HorizontalBoxSlot}
  }
  SizeBox_537 {
   widgetClassPath=&SizeBox
   parent=VerticalBox_Xbox_LS_And_Dpad
   widget=SizeBox_537
   properties {bOverride_HeightOverride=true, heightOverride=4}
   slot {class=VerticalBoxSlot}
  }
  CommonBorder_505 {
   widgetClassPath=&CommonBorder
   parent=SizeBox_537
   widget=CommonBorder_505
   properties {style=&Border-ItemInfo-Locked_C, verticalAlignment=VAlign_Center, padding=<ltrb 0 0 56 0>}
   slot {class=SizeBoxSlot}
  }
  SizeBox_Xbox_Dpad_Left {
   widgetClassPath=&SizeBox
   parent=VerticalBox_Xbox_LS_And_Dpad
   widget=SizeBox_Xbox_Dpad_Left
   properties {bOverride_HeightOverride=true, heightOverride=44}
   slot {class=VerticalBoxSlot}
  }
  CommonBorder_506 {
   widgetClassPath=&CommonBorder
   parent=SizeBox_Xbox_Dpad_Left
   widget=CommonBorder_506
   properties {
    horizontalAlignment=HAlign_Right
    style=&Border-SolidBG-StatItem_C
    verticalAlignment=VAlign_Center
    padding=<ltrb 32 0 15 0>
   }
   slot {class=SizeBoxSlot}
  }
  HorizontalBox_116 {
   widgetClassPath=&HorizontalBox
   parent=CommonBorder_506
   widget=HorizontalBox_116
   slot {class=BorderSlot, properties {horizontalAlignment=HAlign_Right, verticalAlignment=VAlign_Center, padding=<ltrb 32 0 15 0>}}
  }
  Text_Xbox_Dpad_Left {
   widgetClassPath=&CommonTextBlock
   bIsVariable=true
   parent=HorizontalBox_116
   widget=Text_Xbox_Dpad_Left
   properties {
    justification=Right
    scrollStyle=&TextScrollStyle-Base_C
    style=&TextStyle-Base-XS-B-DarkBlueFade_C
    text=Dpad Left
    colorAndOpacity {colorUseRule=UseColor_Specified, specifiedColor=<rgba 0.008 0.011 0.0262 0.8>}
    font {
     bForceMonospaced=false
     bMaterialIsStencil=false
     fontMaterial=null
     fontObject=&NotoSans
     letterSpacing=0
     monospacedWidth=1
     size=16
     skewAmount=0
     typefaceFontName=NotoSans-Bold
     outlineSettings {
      bApplyOutlineToDropShadows=false
      bMiteredCorners=false
      bSeparateFillAlpha=false
      outlineMaterial=null
      outlineSize=0
      outlineColor=<rgba 0 0 0 1>
     }
    }
    shadowOffset=<0 0>
   }
   slot {
    class=HorizontalBoxSlot
    properties {horizontalAlignment=HAlign_Right, verticalAlignment=VAlign_Center, padding=<ltrb 0 0 10 0>, size {sizeRule=Fill, value=1}}
   }
  }
  DPadLeft {
   widgetClassPath=&KeybindWidget
   bIsVariable=true
   parent=HorizontalBox_116
   widget=DPadLeft
   properties {boundKeyFallback=Gamepad_DPad_Left, inputTypeOverride=Gamepad}
   slot {class=HorizontalBoxSlot}
  }
  SizeBox_539 {
   widgetClassPath=&SizeBox
   parent=VerticalBox_Xbox_LS_And_Dpad
   widget=SizeBox_539
   properties {bOverride_HeightOverride=true, heightOverride=4}
   slot {class=VerticalBoxSlot}
  }
  CommonBorder_507 {
   widgetClassPath=&CommonBorder
   parent=SizeBox_539
   widget=CommonBorder_507
   properties {style=&Border-ItemInfo-Locked_C, verticalAlignment=VAlign_Center, padding=<ltrb 0 0 56 0>}
   slot {class=SizeBoxSlot}
  }
  SizeBox_Xbox_Dpad_Right {
   widgetClassPath=&SizeBox
   parent=VerticalBox_Xbox_LS_And_Dpad
   widget=SizeBox_Xbox_Dpad_Right
   properties {bOverride_HeightOverride=true, heightOverride=44}
   slot {class=VerticalBoxSlot}
  }
  CommonBorder_508 {
   widgetClassPath=&CommonBorder
   parent=SizeBox_Xbox_Dpad_Right
   widget=CommonBorder_508
   properties {
    horizontalAlignment=HAlign_Right
    style=&Border-SolidBG-StatItem_C
    verticalAlignment=VAlign_Center
    padding=<ltrb 32 0 15 0>
   }
   slot {class=SizeBoxSlot}
  }
  HorizontalBox_133 {
   widgetClassPath=&HorizontalBox
   parent=CommonBorder_508
   widget=HorizontalBox_133
   slot {class=BorderSlot, properties {horizontalAlignment=HAlign_Right, verticalAlignment=VAlign_Center, padding=<ltrb 32 0 15 0>}}
  }
  Text_Xbox_Dpad_Right {
   widgetClassPath=&CommonTextBlock
   bIsVariable=true
   parent=HorizontalBox_133
   widget=Text_Xbox_Dpad_Right
   properties {
    justification=Right
    scrollStyle=&TextScrollStyle-Base_C
    style=&TextStyle-Base-XS-B-DarkBlueFade_C
    text=Dpad Right
    colorAndOpacity {colorUseRule=UseColor_Specified, specifiedColor=<rgba 0.008 0.011 0.0262 0.8>}
    font {
     bForceMonospaced=false
     bMaterialIsStencil=false
     fontMaterial=null
     fontObject=&NotoSans
     letterSpacing=0
     monospacedWidth=1
     size=16
     skewAmount=0
     typefaceFontName=NotoSans-Bold
     outlineSettings {
      bApplyOutlineToDropShadows=false
      bMiteredCorners=false
      bSeparateFillAlpha=false
      outlineMaterial=null
      outlineSize=0
      outlineColor=<rgba 0 0 0 1>
     }
    }
    shadowOffset=<0 0>
   }
   slot {
    class=HorizontalBoxSlot
    properties {horizontalAlignment=HAlign_Right, verticalAlignment=VAlign_Center, padding=<ltrb 0 0 10 0>, size {sizeRule=Fill, value=1}}
   }
  }
  DPadRight {
   widgetClassPath=&KeybindWidget
   bIsVariable=true
   parent=HorizontalBox_133
   widget=DPadRight
   properties {boundKeyFallback=Gamepad_DPad_Right, inputTypeOverride=Gamepad}
   slot {class=HorizontalBoxSlot}
  }
  VerticalBox_PS4_LS_And_Dpad {
   widgetClassPath=&VerticalBox
   bIsVariable=true
   parent=VerticalBox_19
   widget=VerticalBox_PS4_LS_And_Dpad
   slot {class=VerticalBoxSlot}
  }
  SizeBox_PS4_Dpad_Up {
   widgetClassPath=&SizeBox
   parent=VerticalBox_PS4_LS_And_Dpad
   widget=SizeBox_PS4_Dpad_Up
   properties {bOverride_HeightOverride=true, heightOverride=44}
   slot {class=VerticalBoxSlot}
  }
  CommonBorder_1026 {
   widgetClassPath=&CommonBorder
   parent=SizeBox_PS4_Dpad_Up
   widget=CommonBorder_1026
   properties {
    style=&Border-SolidBG-StatItem_C
    verticalAlignment=VAlign_Center
    padding=<ltrb 32 0 15 0>
   }
   slot {class=SizeBoxSlot}
  }
  HorizontalBox_722 {
   widgetClassPath=&HorizontalBox
   parent=CommonBorder_1026
   widget=HorizontalBox_722
   slot {class=BorderSlot, properties {verticalAlignment=VAlign_Center, padding=<ltrb 32 0 15 0>}}
  }
  Text_PS4_Dpad_Up {
   widgetClassPath=&CommonTextBlock
   bIsVariable=true
   parent=HorizontalBox_722
   widget=Text_PS4_Dpad_Up
   properties {
    justification=Right
    scrollStyle=&TextScrollStyle-Base_C
    style=&TextStyle-Base-XS-B-DarkBlueFade_C
    text=Dpad Up
    colorAndOpacity {colorUseRule=UseColor_Specified, specifiedColor=<rgba 0.008 0.011 0.0262 0.8>}
    font {
     bForceMonospaced=false
     bMaterialIsStencil=false
     fontMaterial=null
     fontObject=&NotoSans
     letterSpacing=0
     monospacedWidth=1
     size=16
     skewAmount=0
     typefaceFontName=NotoSans-Bold
     outlineSettings {
      bApplyOutlineToDropShadows=false
      bMiteredCorners=false
      bSeparateFillAlpha=false
      outlineMaterial=null
      outlineSize=0
      outlineColor=<rgba 0 0 0 1>
     }
    }
    shadowOffset=<0 0>
   }
   slot {
    class=HorizontalBoxSlot
    properties {horizontalAlignment=HAlign_Right, verticalAlignment=VAlign_Center, padding=<ltrb 0 0 10 0>, size {sizeRule=Fill, value=1}}
   }
  }
  DpadUp2 {
   widgetClassPath=&KeybindWidget
   bIsVariable=true
   parent=HorizontalBox_722
   widget=DpadUp2
   properties {boundKeyFallback=Gamepad_DPad_Up, inputTypeOverride=Gamepad}
   slot {class=HorizontalBoxSlot}
  }
  SizeBox_1093 {
   widgetClassPath=&SizeBox
   parent=VerticalBox_PS4_LS_And_Dpad
   widget=SizeBox_1093
   properties {bOverride_HeightOverride=true, heightOverride=4}
   slot {class=VerticalBoxSlot}
  }
  CommonBorder_1027 {
   widgetClassPath=&CommonBorder
   parent=SizeBox_1093
   widget=CommonBorder_1027
   properties {style=&Border-ItemInfo-Locked_C, verticalAlignment=VAlign_Center, padding=<ltrb 0 0 56 0>}
   slot {class=SizeBoxSlot}
  }
  SizeBox_PS4_Dpad_Down {
   widgetClassPath=&SizeBox
   parent=VerticalBox_PS4_LS_And_Dpad
   widget=SizeBox_PS4_Dpad_Down
   properties {bOverride_HeightOverride=true, heightOverride=44}
   slot {class=VerticalBoxSlot}
  }
  CommonBorder_1028 {
   widgetClassPath=&CommonBorder
   parent=SizeBox_PS4_Dpad_Down
   widget=CommonBorder_1028
   properties {
    style=&Border-SolidBG-StatItem_C
    verticalAlignment=VAlign_Center
    padding=<ltrb 32 0 15 0>
   }
   slot {class=SizeBoxSlot}
  }
  HorizontalBox_723 {
   widgetClassPath=&HorizontalBox
   parent=CommonBorder_1028
   widget=HorizontalBox_723
   slot {class=BorderSlot, properties {verticalAlignment=VAlign_Center, padding=<ltrb 32 0 15 0>}}
  }
  Text_PS4_Dpad_Down {
   widgetClassPath=&CommonTextBlock
   bIsVariable=true
   parent=HorizontalBox_723
   widget=Text_PS4_Dpad_Down
   properties {
    justification=Right
    scrollStyle=&TextScrollStyle-Base_C
    style=&TextStyle-Base-XS-B-DarkBlueFade_C
    text=Dpad Down
    colorAndOpacity {colorUseRule=UseColor_Specified, specifiedColor=<rgba 0.008 0.011 0.0262 0.8>}
    font {
     bForceMonospaced=false
     bMaterialIsStencil=false
     fontMaterial=null
     fontObject=&NotoSans
     letterSpacing=0
     monospacedWidth=1
     size=16
     skewAmount=0
     typefaceFontName=NotoSans-Bold
     outlineSettings {
      bApplyOutlineToDropShadows=false
      bMiteredCorners=false
      bSeparateFillAlpha=false
      outlineMaterial=null
      outlineSize=0
      outlineColor=<rgba 0 0 0 1>
     }
    }
    shadowOffset=<0 0>
   }
   slot {
    class=HorizontalBoxSlot
    properties {horizontalAlignment=HAlign_Right, verticalAlignment=VAlign_Center, padding=<ltrb 0 0 10 0>, size {sizeRule=Fill, value=1}}
   }
  }
  DpadDown2 {
   widgetClassPath=&KeybindWidget
   bIsVariable=true
   parent=HorizontalBox_723
   widget=DpadDown2
   properties {boundKeyFallback=Gamepad_DPad_Down, inputTypeOverride=Gamepad}
   slot {class=HorizontalBoxSlot}
  }
  SizeBox_1095 {
   widgetClassPath=&SizeBox
   parent=VerticalBox_PS4_LS_And_Dpad
   widget=SizeBox_1095
   properties {bOverride_HeightOverride=true, heightOverride=4}
   slot {class=VerticalBoxSlot}
  }
  CommonBorder_1029 {
   widgetClassPath=&CommonBorder
   parent=SizeBox_1095
   widget=CommonBorder_1029
   properties {style=&Border-ItemInfo-Locked_C, verticalAlignment=VAlign_Center, padding=<ltrb 0 0 56 0>}
   slot {class=SizeBoxSlot}
  }
  SizeBox_PS4_Dpad_Left {
   widgetClassPath=&SizeBox
   parent=VerticalBox_PS4_LS_And_Dpad
   widget=SizeBox_PS4_Dpad_Left
   properties {bOverride_HeightOverride=true, heightOverride=44}
   slot {class=VerticalBoxSlot}
  }
  CommonBorder_1030 {
   widgetClassPath=&CommonBorder
   parent=SizeBox_PS4_Dpad_Left
   widget=CommonBorder_1030
   properties {
    style=&Border-SolidBG-StatItem_C
    verticalAlignment=VAlign_Center
    padding=<ltrb 32 0 15 0>
   }
   slot {class=SizeBoxSlot}
  }
  HorizontalBox_724 {
   widgetClassPath=&HorizontalBox
   parent=CommonBorder_1030
   widget=HorizontalBox_724
   slot {class=BorderSlot, properties {verticalAlignment=VAlign_Center, padding=<ltrb 32 0 15 0>}}
  }
  Text_PS4_Dpad_Left {
   widgetClassPath=&CommonTextBlock
   bIsVariable=true
   parent=HorizontalBox_724
   widget=Text_PS4_Dpad_Left
   properties {
    justification=Right
    scrollStyle=&TextScrollStyle-Base_C
    style=&TextStyle-Base-XS-B-DarkBlueFade_C
    text=Dpad Left
    colorAndOpacity {colorUseRule=UseColor_Specified, specifiedColor=<rgba 0.008 0.011 0.0262 0.8>}
    font {
     bForceMonospaced=false
     bMaterialIsStencil=false
     fontMaterial=null
     fontObject=&NotoSans
     letterSpacing=0
     monospacedWidth=1
     size=16
     skewAmount=0
     typefaceFontName=NotoSans-Bold
     outlineSettings {
      bApplyOutlineToDropShadows=false
      bMiteredCorners=false
      bSeparateFillAlpha=false
      outlineMaterial=null
      outlineSize=0
      outlineColor=<rgba 0 0 0 1>
     }
    }
    shadowOffset=<0 0>
   }
   slot {
    class=HorizontalBoxSlot
    properties {horizontalAlignment=HAlign_Right, verticalAlignment=VAlign_Center, padding=<ltrb 0 0 10 0>, size {sizeRule=Fill, value=1}}
   }
  }
  DpadLeft2 {
   widgetClassPath=&KeybindWidget
   bIsVariable=true
   parent=HorizontalBox_724
   widget=DpadLeft2
   properties {boundKeyFallback=Gamepad_DPad_Left, inputTypeOverride=Gamepad}
   slot {class=HorizontalBoxSlot}
  }
  SizeBox_1097 {
   widgetClassPath=&SizeBox
   parent=VerticalBox_PS4_LS_And_Dpad
   widget=SizeBox_1097
   properties {bOverride_HeightOverride=true, heightOverride=4}
   slot {class=VerticalBoxSlot}
  }
  CommonBorder_1031 {
   widgetClassPath=&CommonBorder
   parent=SizeBox_1097
   widget=CommonBorder_1031
   properties {style=&Border-ItemInfo-Locked_C, verticalAlignment=VAlign_Center, padding=<ltrb 0 0 56 0>}
   slot {class=SizeBoxSlot}
  }
  SizeBox_PS4_Dpad_Right {
   widgetClassPath=&SizeBox
   parent=VerticalBox_PS4_LS_And_Dpad
   widget=SizeBox_PS4_Dpad_Right
   properties {bOverride_HeightOverride=true, heightOverride=44}
   slot {class=VerticalBoxSlot}
  }
  CommonBorder_1032 {
   widgetClassPath=&CommonBorder
   parent=SizeBox_PS4_Dpad_Right
   widget=CommonBorder_1032
   properties {
    style=&Border-SolidBG-StatItem_C
    verticalAlignment=VAlign_Center
    padding=<ltrb 32 0 15 0>
   }
   slot {class=SizeBoxSlot}
  }
  HorizontalBox_725 {
   widgetClassPath=&HorizontalBox
   parent=CommonBorder_1032
   widget=HorizontalBox_725
   slot {class=BorderSlot, properties {verticalAlignment=VAlign_Center, padding=<ltrb 32 0 15 0>}}
  }
  Text_PS4_Dpad_Right {
   widgetClassPath=&CommonTextBlock
   bIsVariable=true
   parent=HorizontalBox_725
   widget=Text_PS4_Dpad_Right
   properties {
    justification=Right
    scrollStyle=&TextScrollStyle-Base_C
    style=&TextStyle-Base-XS-B-DarkBlueFade_C
    text=Dpad Right
    colorAndOpacity {colorUseRule=UseColor_Specified, specifiedColor=<rgba 0.008 0.011 0.0262 0.8>}
    font {
     bForceMonospaced=false
     bMaterialIsStencil=false
     fontMaterial=null
     fontObject=&NotoSans
     letterSpacing=0
     monospacedWidth=1
     size=16
     skewAmount=0
     typefaceFontName=NotoSans-Bold
     outlineSettings {
      bApplyOutlineToDropShadows=false
      bMiteredCorners=false
      bSeparateFillAlpha=false
      outlineMaterial=null
      outlineSize=0
      outlineColor=<rgba 0 0 0 1>
     }
    }
    shadowOffset=<0 0>
   }
   slot {
    class=HorizontalBoxSlot
    properties {horizontalAlignment=HAlign_Right, verticalAlignment=VAlign_Center, padding=<ltrb 0 0 10 0>, size {sizeRule=Fill, value=1}}
   }
  }
  DpadRight2 {
   widgetClassPath=&KeybindWidget
   bIsVariable=true
   parent=HorizontalBox_725
   widget=DpadRight2
   properties {boundKeyFallback=Gamepad_DPad_Right, inputTypeOverride=Gamepad}
   slot {class=HorizontalBoxSlot}
  }
  Spacer_130 {
   widgetClassPath=&Spacer
   parent=VerticalBox_PS4_LS_And_Dpad
   widget=Spacer_130
   properties {size=<1 40>}
   slot {class=VerticalBoxSlot}
  }
  SizeBox_PS4_LS {
   widgetClassPath=&SizeBox
   parent=VerticalBox_PS4_LS_And_Dpad
   widget=SizeBox_PS4_LS
   properties {bOverride_HeightOverride=true, heightOverride=44}
   slot {class=VerticalBoxSlot}
  }
  CommonBorder_1023 {
   widgetClassPath=&CommonBorder
   parent=SizeBox_PS4_LS
   widget=CommonBorder_1023
   properties {
    style=&Border-SolidBG-StatItem_C
    verticalAlignment=VAlign_Center
    padding=<ltrb 32 0 15 0>
   }
   slot {class=SizeBoxSlot}
  }
  HorizontalBox_720 {
   widgetClassPath=&HorizontalBox
   parent=CommonBorder_1023
   widget=HorizontalBox_720
   slot {class=BorderSlot, properties {verticalAlignment=VAlign_Center, padding=<ltrb 32 0 15 0>}}
  }
  Text_PS4_LS {
   widgetClassPath=&CommonTextBlock
   bIsVariable=true
   parent=HorizontalBox_720
   widget=Text_PS4_LS
   properties {
    justification=Right
    scrollStyle=&TextScrollStyle-Base_C
    style=&TextStyle-Base-XS-B-DarkBlueFade_C
    text=Left Stick
    colorAndOpacity {colorUseRule=UseColor_Specified, specifiedColor=<rgba 0.008 0.011 0.0262 0.8>}
    font {
     bForceMonospaced=false
     bMaterialIsStencil=false
     fontMaterial=null
     fontObject=&NotoSans
     letterSpacing=0
     monospacedWidth=1
     size=16
     skewAmount=0
     typefaceFontName=NotoSans-Bold
     outlineSettings {
      bApplyOutlineToDropShadows=false
      bMiteredCorners=false
      bSeparateFillAlpha=false
      outlineMaterial=null
      outlineSize=0
      outlineColor=<rgba 0 0 0 1>
     }
    }
    shadowOffset=<0 0>
   }
   slot {
    class=HorizontalBoxSlot
    properties {horizontalAlignment=HAlign_Right, verticalAlignment=VAlign_Center, padding=<ltrb 0 0 10 0>, size {sizeRule=Fill, value=1}}
   }
  }
  LStick2 {
   widgetClassPath=&KeybindWidget
   bIsVariable=true
   parent=HorizontalBox_720
   widget=LStick2
   properties {boundKeyFallback=Gamepad_LeftX, inputTypeOverride=Gamepad}
   slot {class=HorizontalBoxSlot}
  }
  SizeBox_1090 {
   widgetClassPath=&SizeBox
   parent=VerticalBox_PS4_LS_And_Dpad
   widget=SizeBox_1090
   properties {bOverride_HeightOverride=true, heightOverride=4}
   slot {class=VerticalBoxSlot}
  }
  CommonBorder_1024 {
   widgetClassPath=&CommonBorder
   parent=SizeBox_1090
   widget=CommonBorder_1024
   properties {style=&Border-ItemInfo-Locked_C, verticalAlignment=VAlign_Center, padding=<ltrb 0 0 56 0>}
   slot {class=SizeBoxSlot}
  }
  SizeBox_PS4_LS_Click {
   widgetClassPath=&SizeBox
   parent=VerticalBox_PS4_LS_And_Dpad
   widget=SizeBox_PS4_LS_Click
   properties {bOverride_HeightOverride=true, heightOverride=76}
   slot {class=VerticalBoxSlot}
  }
  CommonBorder_1025 {
   widgetClassPath=&CommonBorder
   parent=SizeBox_PS4_LS_Click
   widget=CommonBorder_1025
   properties {
    style=&Border-SolidBG-StatItem_C
    verticalAlignment=VAlign_Center
    padding=<ltrb 32 0 15 0>
   }
   slot {class=SizeBoxSlot}
  }
  HorizontalBox_13 {
   widgetClassPath=&HorizontalBox
   parent=CommonBorder_1025
   widget=HorizontalBox_13
   slot {class=BorderSlot, properties {verticalAlignment=VAlign_Center, padding=<ltrb 32 0 15 0>}}
  }
  VerticalBox_1 {
   widgetClassPath=&VerticalBox
   parent=HorizontalBox_13
   widget=VerticalBox_1
   slot {class=HorizontalBoxSlot, properties {padding=<ltrb 0 0 10 0>, size {sizeRule=Fill, value=1}}}
  }
  HorizontalBox_721 {
   widgetClassPath=&HorizontalBox
   parent=VerticalBox_1
   widget=HorizontalBox_721
   slot {class=VerticalBoxSlot, properties {size {sizeRule=Fill, value=1}}}
  }
  Text_PS4_LS_Click_1 {
   widgetClassPath=&CommonTextBlock
   bIsVariable=true
   parent=HorizontalBox_721
   widget=Text_PS4_LS_Click_1
   properties {
    justification=Right
    scrollStyle=&TextScrollStyle-Base_C
    style=&TextStyle-Base-XS-B-DarkBlueFade_C
    text=Left Stick Click 1
    colorAndOpacity {colorUseRule=UseColor_Specified, specifiedColor=<rgba 0.008 0.011 0.0262 0.8>}
    font {
     bForceMonospaced=false
     bMaterialIsStencil=false
     fontMaterial=null
     fontObject=&NotoSans
     letterSpacing=0
     monospacedWidth=1
     size=16
     skewAmount=0
     typefaceFontName=NotoSans-Bold
     outlineSettings {
      bApplyOutlineToDropShadows=false
      bMiteredCorners=false
      bSeparateFillAlpha=false
      outlineMaterial=null
      outlineSize=0
      outlineColor=<rgba 0 0 0 1>
     }
    }
    shadowOffset=<0 0>
   }
   slot {class=HorizontalBoxSlot, properties {horizontalAlignment=HAlign_Right, verticalAlignment=VAlign_Center, size {sizeRule=Fill, value=1}}}
  }
  HorizontalBox_3 {
   widgetClassPath=&HorizontalBox
   parent=VerticalBox_1
   widget=HorizontalBox_3
   slot {class=VerticalBoxSlot, properties {size {sizeRule=Fill, value=1}}}
  }
  Text_PS4_LS_Click_2 {
   widgetClassPath=&CommonTextBlock
   bIsVariable=true
   parent=HorizontalBox_3
   widget=Text_PS4_LS_Click_2
   properties {
    justification=Right
    scrollStyle=&TextScrollStyle-Base_C
    style=&TextStyle-Base-XS-B-DarkBlueFade_C
    text=Left Stick Click 2
    colorAndOpacity {colorUseRule=UseColor_Specified, specifiedColor=<rgba 0.008 0.011 0.0262 0.8>}
    font {
     bForceMonospaced=false
     bMaterialIsStencil=false
     fontMaterial=null
     fontObject=&NotoSans
     letterSpacing=0
     monospacedWidth=1
     size=16
     skewAmount=0
     typefaceFontName=NotoSans-Bold
     outlineSettings {
      bApplyOutlineToDropShadows=false
      bMiteredCorners=false
      bSeparateFillAlpha=false
      outlineMaterial=null
      outlineSize=0
      outlineColor=<rgba 0 0 0 1>
     }
    }
    shadowOffset=<0 0>
   }
   slot {class=HorizontalBoxSlot, properties {horizontalAlignment=HAlign_Right, verticalAlignment=VAlign_Center, size {sizeRule=Fill, value=1}}}
  }
  LStickClick2 {
   widgetClassPath=&KeybindWidget
   bIsVariable=true
   parent=HorizontalBox_13
   widget=LStickClick2
   properties {boundKeyFallback=Gamepad_LeftThumbstick, inputTypeOverride=Gamepad}
   slot {class=HorizontalBoxSlot}
  }
  SizeBox_193 {
   widgetClassPath=&SizeBox
   parent=HorizontalBox_AllText
   widget=SizeBox_193
   properties {bOverride_MinDesiredWidth=true, minDesiredWidth=650}
   slot {class=HorizontalBoxSlot}
  }
  VerticalBox_66 {
   widgetClassPath=&VerticalBox
   parent=HorizontalBox_AllText
   widget=VerticalBox_66
   slot {class=HorizontalBoxSlot, properties {verticalAlignment=VAlign_Top, padding=<ltrb 0 22 12 0>, size {sizeRule=Fill, value=1}}}
  }
  SizeBox_546 {
   widgetClassPath=&SizeBox
   parent=VerticalBox_66
   widget=SizeBox_546
   properties {bOverride_HeightOverride=true, heightOverride=44}
   slot {class=VerticalBoxSlot}
  }
  CommonBorder_512 {
   widgetClassPath=&CommonBorder
   parent=SizeBox_546
   widget=CommonBorder_512
   properties {
    style=&Border-SolidBG-StatItem_C
    verticalAlignment=VAlign_Center
    padding=<ltrb 15 0 32 0>
   }
   slot {class=SizeBoxSlot}
  }
  HorizontalBox_135 {
   widgetClassPath=&HorizontalBox
   parent=CommonBorder_512
   widget=HorizontalBox_135
   slot {class=BorderSlot, properties {verticalAlignment=VAlign_Center, padding=<ltrb 15 0 32 0>}}
  }
  RT {
   widgetClassPath=&KeybindWidget
   bIsVariable=true
   parent=HorizontalBox_135
   widget=RT
   properties {boundKeyFallback=Gamepad_RightTrigger, inputTypeOverride=Gamepad}
   slot {class=HorizontalBoxSlot}
  }
  Text_Gamepad_RT {
   widgetClassPath=&CommonTextBlock
   bIsVariable=true
   parent=HorizontalBox_135
   widget=Text_Gamepad_RT
   properties {
    scrollStyle=&TextScrollStyle-Base_C
    style=&TextStyle-Base-XS-B-DarkBlueFade_C
    text=Right Trigger
    colorAndOpacity {colorUseRule=UseColor_Specified, specifiedColor=<rgba 0.008 0.011 0.0262 0.8>}
    font {
     bForceMonospaced=false
     bMaterialIsStencil=false
     fontMaterial=null
     fontObject=&NotoSans
     letterSpacing=0
     monospacedWidth=1
     size=16
     skewAmount=0
     typefaceFontName=NotoSans-Bold
     outlineSettings {
      bApplyOutlineToDropShadows=false
      bMiteredCorners=false
      bSeparateFillAlpha=false
      outlineMaterial=null
      outlineSize=0
      outlineColor=<rgba 0 0 0 1>
     }
    }
    shadowOffset=<0 0>
   }
   slot {class=HorizontalBoxSlot, properties {verticalAlignment=VAlign_Center, padding=<ltrb 10 0 0 0>, size {sizeRule=Fill, value=1}}}
  }
  SizeBox_547 {
   widgetClassPath=&SizeBox
   parent=VerticalBox_66
   widget=SizeBox_547
   properties {bOverride_HeightOverride=true, heightOverride=4}
   slot {class=VerticalBoxSlot}
  }
  CommonBorder_513 {
   widgetClassPath=&CommonBorder
   parent=SizeBox_547
   widget=CommonBorder_513
   properties {style=&Border-ItemInfo-Locked_C, verticalAlignment=VAlign_Center, padding=<ltrb 0 0 56 0>}
   slot {class=SizeBoxSlot}
  }
  SizeBox_548 {
   widgetClassPath=&SizeBox
   parent=VerticalBox_66
   widget=SizeBox_548
   properties {bOverride_HeightOverride=true, heightOverride=44}
   slot {class=VerticalBoxSlot}
  }
  CommonBorder_514 {
   widgetClassPath=&CommonBorder
   parent=SizeBox_548
   widget=CommonBorder_514
   properties {
    style=&Border-SolidBG-StatItem_C
    verticalAlignment=VAlign_Center
    padding=<ltrb 15 0 32 0>
   }
   slot {class=SizeBoxSlot}
  }
  HorizontalBox_142 {
   widgetClassPath=&HorizontalBox
   parent=CommonBorder_514
   widget=HorizontalBox_142
   slot {class=BorderSlot, properties {verticalAlignment=VAlign_Center, padding=<ltrb 15 0 32 0>}}
  }
  RS {
   widgetClassPath=&KeybindWidget
   bIsVariable=true
   parent=HorizontalBox_142
   widget=RS
   properties {boundKeyFallback=Gamepad_RightShoulder, inputTypeOverride=Gamepad}
   slot {class=HorizontalBoxSlot}
  }
  Text_Gamepad_RB {
   widgetClassPath=&CommonTextBlock
   bIsVariable=true
   parent=HorizontalBox_142
   widget=Text_Gamepad_RB
   properties {
    scrollStyle=&TextScrollStyle-Base_C
    style=&TextStyle-Base-XS-B-DarkBlueFade_C
    text=Right Bumper
    colorAndOpacity {colorUseRule=UseColor_Specified, specifiedColor=<rgba 0.008 0.011 0.0262 0.8>}
    font {
     bForceMonospaced=false
     bMaterialIsStencil=false
     fontMaterial=null
     fontObject=&NotoSans
     letterSpacing=0
     monospacedWidth=1
     size=16
     skewAmount=0
     typefaceFontName=NotoSans-Bold
     outlineSettings {
      bApplyOutlineToDropShadows=false
      bMiteredCorners=false
      bSeparateFillAlpha=false
      outlineMaterial=null
      outlineSize=0
      outlineColor=<rgba 0 0 0 1>
     }
    }
    shadowOffset=<0 0>
   }
   slot {class=HorizontalBoxSlot, properties {verticalAlignment=VAlign_Center, padding=<ltrb 10 0 0 0>, size {sizeRule=Fill, value=1}}}
  }
  SizeBox_549 {
   widgetClassPath=&SizeBox
   parent=VerticalBox_66
   widget=SizeBox_549
   properties {bOverride_HeightOverride=true, heightOverride=4}
   slot {class=VerticalBoxSlot}
  }
  CommonBorder-RightBumperSeparator {
   widgetClassPath=&CommonBorder
   bIsVariable=true
   parent=SizeBox_549
   widget=CommonBorder-RightBumperSeparator
   properties {style=&Border-ItemInfo-Locked_C, verticalAlignment=VAlign_Center, padding=<ltrb 0 0 56 0>}
   slot {class=SizeBoxSlot}
  }
  SizeBox_550 {
   widgetClassPath=&SizeBox
   parent=VerticalBox_66
   widget=SizeBox_550
   properties {bOverride_HeightOverride=true, heightOverride=44}
   slot {class=VerticalBoxSlot}
  }
  HorizontalBox-LT+RB {widgetClassPath=&HorizontalBox, bIsVariable=true, parent=SizeBox_550, widget=HorizontalBox-LT+RB, slot {class=SizeBoxSlot}}
  CommonBorder_7 {
   widgetClassPath=&CommonBorder
   parent=HorizontalBox-LT+RB
   widget=CommonBorder_7
   properties {
    style=&Border-SolidBG-StatItem_C
    verticalAlignment=VAlign_Center
    padding=<ltrb 12 0 32 0>
   }
   slot {class=HorizontalBoxSlot, properties {size {sizeRule=Fill, value=1}}}
  }
  HorizontalBox_8 {
   widgetClassPath=&HorizontalBox
   parent=CommonBorder_7
   widget=HorizontalBox_8
   slot {class=BorderSlot, properties {verticalAlignment=VAlign_Center, padding=<ltrb 12 0 32 0>}}
  }
  HorizontalBox_11 {
   widgetClassPath=&HorizontalBox
   parent=HorizontalBox_8
   widget=HorizontalBox_11
   slot {class=HorizontalBoxSlot, properties {padding=<ltrb 0 0 12 0>}}
  }
  LTCombo {
   widgetClassPath=&KeybindWidget
   bIsVariable=true
   parent=HorizontalBox_11
   widget=LTCombo
   properties {boundKeyFallback=Gamepad_LeftTrigger, inputTypeOverride=Gamepad}
   slot {class=HorizontalBoxSlot}
  }
  CommonTextBlock_270 {
   widgetClassPath=&CommonTextBlock
   parent=HorizontalBox_11
   widget=CommonTextBlock_270
   properties {
    justification=Center
    style=&TextStyle-Base-XS-B-DarkBlueFade_C
    text=+
    colorAndOpacity {colorUseRule=UseColor_Specified, specifiedColor=<rgba 0.008 0.011 0.0262 0.8>}
    font {
     bForceMonospaced=false
     bMaterialIsStencil=false
     fontMaterial=null
     fontObject=&NotoSans
     letterSpacing=0
     monospacedWidth=1
     size=16
     skewAmount=0
     typefaceFontName=NotoSans-Bold
     outlineSettings {
      bApplyOutlineToDropShadows=false
      bMiteredCorners=false
      bSeparateFillAlpha=false
      outlineMaterial=null
      outlineSize=0
      outlineColor=<rgba 0 0 0 1>
     }
    }
    shadowOffset=<0 0>
   }
   slot {class=HorizontalBoxSlot, properties {verticalAlignment=VAlign_Center, padding=<ltrb 8 0 8 0>}}
  }
  RSCombo2 {
   widgetClassPath=&KeybindWidget
   bIsVariable=true
   parent=HorizontalBox_11
   widget=RSCombo2
   properties {boundKeyFallback=Gamepad_RightShoulder, inputTypeOverride=Gamepad}
   slot {class=HorizontalBoxSlot}
  }
  LT+RBAction {
   widgetClassPath=&CommonTextBlock
   bIsVariable=true
   parent=HorizontalBox_8
   widget=LT+RBAction
   properties {
    scrollStyle=&TextScrollStyle-Base_C
    style=&TextStyle-Base-XS-B-DarkBlueFade_C
    text=L2 + R1 Ability
    colorAndOpacity {colorUseRule=UseColor_Specified, specifiedColor=<rgba 0.008 0.011 0.0262 0.8>}
    font {
     bForceMonospaced=false
     bMaterialIsStencil=false
     fontMaterial=null
     fontObject=&NotoSans
     letterSpacing=0
     monospacedWidth=1
     size=16
     skewAmount=0
     typefaceFontName=NotoSans-Bold
     outlineSettings {
      bApplyOutlineToDropShadows=false
      bMiteredCorners=false
      bSeparateFillAlpha=false
      outlineMaterial=null
      outlineSize=0
      outlineColor=<rgba 0 0 0 1>
     }
    }
    shadowOffset=<0 0>
   }
   slot {class=HorizontalBoxSlot, properties {verticalAlignment=VAlign_Center, size {sizeRule=Fill, value=1}}}
  }
  Spacer_48 {widgetClassPath=&Spacer, parent=VerticalBox_66, widget=Spacer_48, properties {size=<1 12>}, slot {class=VerticalBoxSlot}}
  SizeBox_551 {
   widgetClassPath=&SizeBox
   parent=VerticalBox_66
   widget=SizeBox_551
   properties {bOverride_HeightOverride=true, heightOverride=76}
   slot {class=VerticalBoxSlot}
  }
  CommonBorder_517 {
   widgetClassPath=&CommonBorder
   parent=SizeBox_551
   widget=CommonBorder_517
   properties {
    style=&Border-SolidBG-StatItem_C
    verticalAlignment=VAlign_Center
    padding=<ltrb 15 0 32 0>
   }
   slot {class=SizeBoxSlot}
  }
  HorizontalBox_15 {
   widgetClassPath=&HorizontalBox
   parent=CommonBorder_517
   widget=HorizontalBox_15
   slot {class=BorderSlot, properties {verticalAlignment=VAlign_Center, padding=<ltrb 15 0 32 0>}}
  }
  FaceButtonTop {
   widgetClassPath=&KeybindWidget
   bIsVariable=true
   parent=HorizontalBox_15
   widget=FaceButtonTop
   properties {boundKeyFallback=Gamepad_FaceButton_Top, inputTypeOverride=Gamepad}
   slot {class=HorizontalBoxSlot}
  }
  VerticalBox_91 {
   widgetClassPath=&VerticalBox
   parent=HorizontalBox_15
   widget=VerticalBox_91
   slot {class=HorizontalBoxSlot, properties {padding=<ltrb 10 0 0 0>, size {sizeRule=Fill, value=1}}}
  }
  HorizontalBox_160 {widgetClassPath=&HorizontalBox, parent=VerticalBox_91, widget=HorizontalBox_160, slot {class=VerticalBoxSlot}}
  Text_Gamepad_Face_Button_Top_1 {
   widgetClassPath=&CommonTextBlock
   bIsVariable=true
   parent=HorizontalBox_160
   widget=Text_Gamepad_Face_Button_Top_1
   properties {
    scrollStyle=&TextScrollStyle-Base_C
    style=&TextStyle-Base-XS-B-DarkBlueFade_C
    text=Face Button Top 1
    colorAndOpacity {colorUseRule=UseColor_Specified, specifiedColor=<rgba 0.008 0.011 0.0262 0.8>}
    font {
     bForceMonospaced=false
     bMaterialIsStencil=false
     fontMaterial=null
     fontObject=&NotoSans
     letterSpacing=0
     monospacedWidth=1
     size=16
     skewAmount=0
     typefaceFontName=NotoSans-Bold
     outlineSettings {
      bApplyOutlineToDropShadows=false
      bMiteredCorners=false
      bSeparateFillAlpha=false
      outlineMaterial=null
      outlineSize=0
      outlineColor=<rgba 0 0 0 1>
     }
    }
    shadowOffset=<0 0>
   }
   slot {class=HorizontalBoxSlot, properties {verticalAlignment=VAlign_Center, size {sizeRule=Fill, value=1}}}
  }
  HorizontalBox_236 {widgetClassPath=&HorizontalBox, parent=VerticalBox_91, widget=HorizontalBox_236, slot {class=VerticalBoxSlot}}
  Text_Gamepad_Face_Button_Top_2 {
   widgetClassPath=&CommonTextBlock
   bIsVariable=true
   parent=HorizontalBox_236
   widget=Text_Gamepad_Face_Button_Top_2
   properties {
    scrollStyle=&TextScrollStyle-Base_C
    style=&TextStyle-Base-XS-B-DarkBlueFade_C
    text=Face Button Top 2
    colorAndOpacity {colorUseRule=UseColor_Specified, specifiedColor=<rgba 0.008 0.011 0.0262 0.8>}
    font {
     bForceMonospaced=false
     bMaterialIsStencil=false
     fontMaterial=null
     fontObject=&NotoSans
     letterSpacing=0
     monospacedWidth=1
     size=16
     skewAmount=0
     typefaceFontName=NotoSans-Bold
     outlineSettings {
      bApplyOutlineToDropShadows=false
      bMiteredCorners=false
      bSeparateFillAlpha=false
      outlineMaterial=null
      outlineSize=0
      outlineColor=<rgba 0 0 0 1>
     }
    }
    shadowOffset=<0 0>
   }
   slot {class=HorizontalBoxSlot, properties {verticalAlignment=VAlign_Center, size {sizeRule=Fill, value=1}}}
  }
  SizeBox_552 {
   widgetClassPath=&SizeBox
   parent=VerticalBox_66
   widget=SizeBox_552
   properties {bOverride_HeightOverride=true, heightOverride=4}
   slot {class=VerticalBoxSlot}
  }
  CommonBorder_518 {
   widgetClassPath=&CommonBorder
   parent=SizeBox_552
   widget=CommonBorder_518
   properties {style=&Border-ItemInfo-Locked_C, verticalAlignment=VAlign_Center, padding=<ltrb 0 0 56 0>}
   slot {class=SizeBoxSlot}
  }
  SizeBox_553 {
   widgetClassPath=&SizeBox
   parent=VerticalBox_66
   widget=SizeBox_553
   properties {bOverride_HeightOverride=true, heightOverride=108}
   slot {class=VerticalBoxSlot}
  }
  CommonBorder_519 {
   widgetClassPath=&CommonBorder
   parent=SizeBox_553
   widget=CommonBorder_519
   properties {
    style=&Border-SolidBG-StatItem_C
    verticalAlignment=VAlign_Center
    padding=<ltrb 15 0 32 0>
   }
   slot {class=SizeBoxSlot}
  }
  HorizontalBox_16 {
   widgetClassPath=&HorizontalBox
   parent=CommonBorder_519
   widget=HorizontalBox_16
   slot {class=BorderSlot, properties {verticalAlignment=VAlign_Center, padding=<ltrb 15 0 32 0>}}
  }
  FaceButtonLeft {
   widgetClassPath=&KeybindWidget
   bIsVariable=true
   parent=HorizontalBox_16
   widget=FaceButtonLeft
   properties {boundKeyFallback=Gamepad_FaceButton_Left, inputTypeOverride=Gamepad}
   slot {class=HorizontalBoxSlot}
  }
  VerticalBox_74 {
   widgetClassPath=&VerticalBox
   parent=HorizontalBox_16
   widget=VerticalBox_74
   slot {class=HorizontalBoxSlot, properties {padding=<ltrb 10 0 0 0>, size {sizeRule=Fill, value=1}}}
  }
  HorizontalBox_161 {widgetClassPath=&HorizontalBox, parent=VerticalBox_74, widget=HorizontalBox_161, slot {class=VerticalBoxSlot}}
  Text_Gamepad_Face_Button_Left_1 {
   widgetClassPath=&CommonTextBlock
   bIsVariable=true
   parent=HorizontalBox_161
   widget=Text_Gamepad_Face_Button_Left_1
   properties {
    scrollStyle=&TextScrollStyle-Base_C
    style=&TextStyle-Base-XS-B-DarkBlueFade_C
    text=Face Button Left 1
    wrappingPolicy=AllowPerCharacterWrapping
    colorAndOpacity {colorUseRule=UseColor_Specified, specifiedColor=<rgba 0.008 0.011 0.0262 0.8>}
    font {
     bForceMonospaced=false
     bMaterialIsStencil=false
     fontMaterial=null
     fontObject=&NotoSans
     letterSpacing=0
     monospacedWidth=1
     size=16
     skewAmount=0
     typefaceFontName=NotoSans-Bold
     outlineSettings {
      bApplyOutlineToDropShadows=false
      bMiteredCorners=false
      bSeparateFillAlpha=false
      outlineMaterial=null
      outlineSize=0
      outlineColor=<rgba 0 0 0 1>
     }
    }
    shadowOffset=<0 0>
   }
   slot {class=HorizontalBoxSlot, properties {verticalAlignment=VAlign_Center, size {sizeRule=Fill, value=1}}}
  }
  Text_Gamepad_Face_Button_Left_2 {
   widgetClassPath=&CommonTextBlock
   bIsVariable=true
   parent=VerticalBox_74
   widget=Text_Gamepad_Face_Button_Left_2
   properties {
    scrollStyle=&TextScrollStyle-Base_C
    style=&TextStyle-Base-XS-B-DarkBlueFade_C
    text=Face Button Left 2
    wrappingPolicy=AllowPerCharacterWrapping
    colorAndOpacity {colorUseRule=UseColor_Specified, specifiedColor=<rgba 0.008 0.011 0.0262 0.8>}
    font {
     bForceMonospaced=false
     bMaterialIsStencil=false
     fontMaterial=null
     fontObject=&NotoSans
     letterSpacing=0
     monospacedWidth=1
     size=16
     skewAmount=0
     typefaceFontName=NotoSans-Bold
     outlineSettings {
      bApplyOutlineToDropShadows=false
      bMiteredCorners=false
      bSeparateFillAlpha=false
      outlineMaterial=null
      outlineSize=0
      outlineColor=<rgba 0 0 0 1>
     }
    }
    shadowOffset=<0 0>
   }
   slot {class=VerticalBoxSlot}
  }
  Text_Gamepad_Face_Button_Left_3 {
   widgetClassPath=&CommonTextBlock
   bIsVariable=true
   parent=VerticalBox_74
   widget=Text_Gamepad_Face_Button_Left_3
   properties {
    scrollStyle=&TextScrollStyle-Base_C
    style=&TextStyle-Base-XS-B-DarkBlueFade_C
    text=Face Button Left 3
    wrappingPolicy=AllowPerCharacterWrapping
    colorAndOpacity {colorUseRule=UseColor_Specified, specifiedColor=<rgba 0.008 0.011 0.0262 0.8>}
    font {
     bForceMonospaced=false
     bMaterialIsStencil=false
     fontMaterial=null
     fontObject=&NotoSans
     letterSpacing=0
     monospacedWidth=1
     size=16
     skewAmount=0
     typefaceFontName=NotoSans-Bold
     outlineSettings {
      bApplyOutlineToDropShadows=false
      bMiteredCorners=false
      bSeparateFillAlpha=false
      outlineMaterial=null
      outlineSize=0
      outlineColor=<rgba 0 0 0 1>
     }
    }
    shadowOffset=<0 0>
   }
   slot {class=VerticalBoxSlot}
  }
  SizeBox_554 {
   widgetClassPath=&SizeBox
   parent=VerticalBox_66
   widget=SizeBox_554
   properties {bOverride_HeightOverride=true, heightOverride=4}
   slot {class=VerticalBoxSlot}
  }
  CommonBorder_520 {
   widgetClassPath=&CommonBorder
   parent=SizeBox_554
   widget=CommonBorder_520
   properties {style=&Border-ItemInfo-Locked_C, verticalAlignment=VAlign_Center, padding=<ltrb 0 0 56 0>}
   slot {class=SizeBoxSlot}
  }
  SizeBox_555 {
   widgetClassPath=&SizeBox
   parent=VerticalBox_66
   widget=SizeBox_555
   properties {bOverride_HeightOverride=true, heightOverride=108}
   slot {class=VerticalBoxSlot}
  }
  CommonBorder_521 {
   widgetClassPath=&CommonBorder
   parent=SizeBox_555
   widget=CommonBorder_521
   properties {
    style=&Border-SolidBG-StatItem_C
    verticalAlignment=VAlign_Center
    padding=<ltrb 15 0 32 0>
   }
   slot {class=SizeBoxSlot}
  }
  HorizontalBox_17 {
   widgetClassPath=&HorizontalBox
   parent=CommonBorder_521
   widget=HorizontalBox_17
   slot {class=BorderSlot, properties {verticalAlignment=VAlign_Center, padding=<ltrb 15 0 32 0>}}
  }
  FaceButtonRight {
   widgetClassPath=&KeybindWidget
   bIsVariable=true
   parent=HorizontalBox_17
   widget=FaceButtonRight
   properties {boundKeyFallback=Gamepad_FaceButton_Right, inputTypeOverride=Gamepad}
   slot {class=HorizontalBoxSlot}
  }
  VerticalBox_83 {
   widgetClassPath=&VerticalBox
   parent=HorizontalBox_17
   widget=VerticalBox_83
   slot {class=HorizontalBoxSlot, properties {padding=<ltrb 10 0 0 0>, size {sizeRule=Fill, value=1}}}
  }
  HorizontalBox_187 {widgetClassPath=&HorizontalBox, parent=VerticalBox_83, widget=HorizontalBox_187, slot {class=VerticalBoxSlot}}
  Text_Gamepad_Face_Button_Right_1 {
   widgetClassPath=&CommonTextBlock
   bIsVariable=true
   parent=HorizontalBox_187
   widget=Text_Gamepad_Face_Button_Right_1
   properties {
    scrollStyle=&TextScrollStyle-Base_C
    style=&TextStyle-Base-XS-B-DarkBlueFade_C
    text=Face Button Right 1
    colorAndOpacity {colorUseRule=UseColor_Specified, specifiedColor=<rgba 0.008 0.011 0.0262 0.8>}
    font {
     bForceMonospaced=false
     bMaterialIsStencil=false
     fontMaterial=null
     fontObject=&NotoSans
     letterSpacing=0
     monospacedWidth=1
     size=16
     skewAmount=0
     typefaceFontName=NotoSans-Bold
     outlineSettings {
      bApplyOutlineToDropShadows=false
      bMiteredCorners=false
      bSeparateFillAlpha=false
      outlineMaterial=null
      outlineSize=0
      outlineColor=<rgba 0 0 0 1>
     }
    }
    shadowOffset=<0 0>
   }
   slot {class=HorizontalBoxSlot, properties {verticalAlignment=VAlign_Center, size {sizeRule=Fill, value=1}}}
  }
  HorizontalBox_209 {widgetClassPath=&HorizontalBox, parent=VerticalBox_83, widget=HorizontalBox_209, slot {class=VerticalBoxSlot}}
  Text_Gamepad_Face_Button_Right_2 {
   widgetClassPath=&CommonTextBlock
   bIsVariable=true
   parent=HorizontalBox_209
   widget=Text_Gamepad_Face_Button_Right_2
   properties {
    scrollStyle=&TextScrollStyle-Base_C
    style=&TextStyle-Base-XS-B-DarkBlueFade_C
    text=Face Button Right 2
    colorAndOpacity {colorUseRule=UseColor_Specified, specifiedColor=<rgba 0.008 0.011 0.0262 0.8>}
    font {
     bForceMonospaced=false
     bMaterialIsStencil=false
     fontMaterial=null
     fontObject=&NotoSans
     letterSpacing=0
     monospacedWidth=1
     size=16
     skewAmount=0
     typefaceFontName=NotoSans-Bold
     outlineSettings {
      bApplyOutlineToDropShadows=false
      bMiteredCorners=false
      bSeparateFillAlpha=false
      outlineMaterial=null
      outlineSize=0
      outlineColor=<rgba 0 0 0 1>
     }
    }
    shadowOffset=<0 0>
   }
   slot {class=HorizontalBoxSlot, properties {verticalAlignment=VAlign_Center, size {sizeRule=Fill, value=1}}}
  }
  HorizontalBox_7 {widgetClassPath=&HorizontalBox, parent=VerticalBox_83, widget=HorizontalBox_7, slot {class=VerticalBoxSlot}}
  Text_Gamepad_Face_Button_Right_3 {
   widgetClassPath=&CommonTextBlock
   bIsVariable=true
   parent=HorizontalBox_7
   widget=Text_Gamepad_Face_Button_Right_3
   properties {
    scrollStyle=&TextScrollStyle-Base_C
    style=&TextStyle-Base-XS-B-DarkBlueFade_C
    text=Face Button Right 3
    colorAndOpacity {colorUseRule=UseColor_Specified, specifiedColor=<rgba 0.008 0.011 0.0262 0.8>}
    font {
     bForceMonospaced=false
     bMaterialIsStencil=false
     fontMaterial=null
     fontObject=&NotoSans
     letterSpacing=0
     monospacedWidth=1
     size=16
     skewAmount=0
     typefaceFontName=NotoSans-Bold
     outlineSettings {
      bApplyOutlineToDropShadows=false
      bMiteredCorners=false
      bSeparateFillAlpha=false
      outlineMaterial=null
      outlineSize=0
      outlineColor=<rgba 0 0 0 1>
     }
    }
    shadowOffset=<0 0>
   }
   slot {class=HorizontalBoxSlot, properties {verticalAlignment=VAlign_Center, size {sizeRule=Fill, value=1}}}
  }
  SizeBox_556 {
   widgetClassPath=&SizeBox
   parent=VerticalBox_66
   widget=SizeBox_556
   properties {bOverride_HeightOverride=true, heightOverride=4}
   slot {class=VerticalBoxSlot}
  }
  CommonBorder_522 {
   widgetClassPath=&CommonBorder
   parent=SizeBox_556
   widget=CommonBorder_522
   properties {style=&Border-ItemInfo-Locked_C, verticalAlignment=VAlign_Center, padding=<ltrb 0 0 56 0>}
   slot {class=SizeBoxSlot}
  }
  SizeBox_557 {
   widgetClassPath=&SizeBox
   parent=VerticalBox_66
   widget=SizeBox_557
   properties {bOverride_HeightOverride=true, heightOverride=44}
   slot {class=VerticalBoxSlot}
  }
  CommonBorder_523 {
   widgetClassPath=&CommonBorder
   parent=SizeBox_557
   widget=CommonBorder_523
   properties {
    style=&Border-SolidBG-StatItem_C
    verticalAlignment=VAlign_Center
    padding=<ltrb 15 0 32 0>
   }
   slot {class=SizeBoxSlot}
  }
  HorizontalBox_210 {
   widgetClassPath=&HorizontalBox
   parent=CommonBorder_523
   widget=HorizontalBox_210
   slot {class=BorderSlot, properties {verticalAlignment=VAlign_Center, padding=<ltrb 15 0 32 0>}}
  }
  FaceButtonBot {
   widgetClassPath=&KeybindWidget
   bIsVariable=true
   parent=HorizontalBox_210
   widget=FaceButtonBot
   properties {boundKeyFallback=Gamepad_FaceButton_Bottom, inputTypeOverride=Gamepad}
   slot {class=HorizontalBoxSlot}
  }
  Text_Gamepad_Face_Button_Bottom {
   widgetClassPath=&CommonTextBlock
   bIsVariable=true
   parent=HorizontalBox_210
   widget=Text_Gamepad_Face_Button_Bottom
   properties {
    scrollStyle=&TextScrollStyle-Base_C
    style=&TextStyle-Base-XS-B-DarkBlueFade_C
    text=Face Button Bottom
    colorAndOpacity {colorUseRule=UseColor_Specified, specifiedColor=<rgba 0.008 0.011 0.0262 0.8>}
    font {
     bForceMonospaced=false
     bMaterialIsStencil=false
     fontMaterial=null
     fontObject=&NotoSans
     letterSpacing=0
     monospacedWidth=1
     size=16
     skewAmount=0
     typefaceFontName=NotoSans-Bold
     outlineSettings {
      bApplyOutlineToDropShadows=false
      bMiteredCorners=false
      bSeparateFillAlpha=false
      outlineMaterial=null
      outlineSize=0
      outlineColor=<rgba 0 0 0 1>
     }
    }
    shadowOffset=<0 0>
   }
   slot {class=HorizontalBoxSlot, properties {verticalAlignment=VAlign_Center, padding=<ltrb 10 0 0 0>, size {sizeRule=Fill, value=1}}}
  }
  Spacer_49 {widgetClassPath=&Spacer, parent=VerticalBox_66, widget=Spacer_49, properties {size=<1 12>}, slot {class=VerticalBoxSlot}}
  SizeBox_558 {
   widgetClassPath=&SizeBox
   parent=VerticalBox_66
   widget=SizeBox_558
   properties {bOverride_HeightOverride=true, heightOverride=44}
   slot {class=VerticalBoxSlot}
  }
  CommonBorder_524 {
   widgetClassPath=&CommonBorder
   parent=SizeBox_558
   widget=CommonBorder_524
   properties {
    style=&Border-SolidBG-StatItem_C
    verticalAlignment=VAlign_Center
    padding=<ltrb 15 0 32 0>
   }
   slot {class=SizeBoxSlot}
  }
  HorizontalBox_211 {
   widgetClassPath=&HorizontalBox
   parent=CommonBorder_524
   widget=HorizontalBox_211
   slot {class=BorderSlot, properties {verticalAlignment=VAlign_Center, padding=<ltrb 15 0 32 0>}}
  }
  RStick {
   widgetClassPath=&KeybindWidget
   bIsVariable=true
   parent=HorizontalBox_211
   widget=RStick
   properties {boundKeyFallback=Gamepad_RightX, inputTypeOverride=Gamepad}
   slot {class=HorizontalBoxSlot}
  }
  Text_Gamepad_RS {
   widgetClassPath=&CommonTextBlock
   bIsVariable=true
   parent=HorizontalBox_211
   widget=Text_Gamepad_RS
   properties {
    scrollStyle=&TextScrollStyle-Base_C
    style=&TextStyle-Base-XS-B-DarkBlueFade_C
    text=Right Stick
    colorAndOpacity {colorUseRule=UseColor_Specified, specifiedColor=<rgba 0.008 0.011 0.0262 0.8>}
    font {
     bForceMonospaced=false
     bMaterialIsStencil=false
     fontMaterial=null
     fontObject=&NotoSans
     letterSpacing=0
     monospacedWidth=1
     size=16
     skewAmount=0
     typefaceFontName=NotoSans-Bold
     outlineSettings {
      bApplyOutlineToDropShadows=false
      bMiteredCorners=false
      bSeparateFillAlpha=false
      outlineMaterial=null
      outlineSize=0
      outlineColor=<rgba 0 0 0 1>
     }
    }
    shadowOffset=<0 0>
   }
   slot {class=HorizontalBoxSlot, properties {verticalAlignment=VAlign_Center, padding=<ltrb 10 0 0 0>, size {sizeRule=Fill, value=1}}}
  }
  SizeBox_559 {
   widgetClassPath=&SizeBox
   parent=VerticalBox_66
   widget=SizeBox_559
   properties {bOverride_HeightOverride=true, heightOverride=4}
   slot {class=VerticalBoxSlot}
  }
  CommonBorder_525 {
   widgetClassPath=&CommonBorder
   parent=SizeBox_559
   widget=CommonBorder_525
   properties {style=&Border-ItemInfo-Locked_C, verticalAlignment=VAlign_Center, padding=<ltrb 0 0 56 0>}
   slot {class=SizeBoxSlot}
  }
  SizeBox_560 {
   widgetClassPath=&SizeBox
   parent=VerticalBox_66
   widget=SizeBox_560
   properties {bOverride_HeightOverride=true, heightOverride=108}
   slot {class=VerticalBoxSlot}
  }
  CommonBorder_526 {
   widgetClassPath=&CommonBorder
   parent=SizeBox_560
   widget=CommonBorder_526
   properties {
    style=&Border-SolidBG-StatItem_C
    verticalAlignment=VAlign_Center
    padding=<ltrb 15 0 32 0>
   }
   slot {class=SizeBoxSlot}
  }
  HorizontalBox_18 {
   widgetClassPath=&HorizontalBox
   parent=CommonBorder_526
   widget=HorizontalBox_18
   slot {class=BorderSlot, properties {verticalAlignment=VAlign_Center, padding=<ltrb 15 0 32 0>}}
  }
  RStickClick {
   widgetClassPath=&KeybindWidget
   bIsVariable=true
   parent=HorizontalBox_18
   widget=RStickClick
   properties {boundKeyFallback=Gamepad_RightThumbstick, inputTypeOverride=Gamepad}
   slot {class=HorizontalBoxSlot}
  }
  VerticalBox_88 {
   widgetClassPath=&VerticalBox
   parent=HorizontalBox_18
   widget=VerticalBox_88
   slot {class=HorizontalBoxSlot, properties {padding=<ltrb 10 0 0 0>, size {sizeRule=Fill, value=1}}}
  }
  Text_Gamepad_RS_Click_1 {
   widgetClassPath=&CommonTextBlock
   bIsVariable=true
   parent=VerticalBox_88
   widget=Text_Gamepad_RS_Click_1
   properties {
    scrollStyle=&TextScrollStyle-Base_C
    style=&TextStyle-Base-XS-B-DarkBlueFade_C
    text=Right Stick Click 1
    colorAndOpacity {colorUseRule=UseColor_Specified, specifiedColor=<rgba 0.008 0.011 0.0262 0.8>}
    font {
     bForceMonospaced=false
     bMaterialIsStencil=false
     fontMaterial=null
     fontObject=&NotoSans
     letterSpacing=0
     monospacedWidth=1
     size=16
     skewAmount=0
     typefaceFontName=NotoSans-Bold
     outlineSettings {
      bApplyOutlineToDropShadows=false
      bMiteredCorners=false
      bSeparateFillAlpha=false
      outlineMaterial=null
      outlineSize=0
      outlineColor=<rgba 0 0 0 1>
     }
    }
    shadowOffset=<0 0>
   }
   slot {class=VerticalBoxSlot}
  }
  Text_Gamepad_RS_Click_2 {
   widgetClassPath=&CommonTextBlock
   bIsVariable=true
   parent=VerticalBox_88
   widget=Text_Gamepad_RS_Click_2
   properties {
    scrollStyle=&TextScrollStyle-Base_C
    style=&TextStyle-Base-XS-B-DarkBlueFade_C
    text=Right Stick Click 2
    colorAndOpacity {colorUseRule=UseColor_Specified, specifiedColor=<rgba 0.008 0.011 0.0262 0.8>}
    font {
     bForceMonospaced=false
     bMaterialIsStencil=false
     fontMaterial=null
     fontObject=&NotoSans
     letterSpacing=0
     monospacedWidth=1
     size=16
     skewAmount=0
     typefaceFontName=NotoSans-Bold
     outlineSettings {
      bApplyOutlineToDropShadows=false
      bMiteredCorners=false
      bSeparateFillAlpha=false
      outlineMaterial=null
      outlineSize=0
      outlineColor=<rgba 0 0 0 1>
     }
    }
    shadowOffset=<0 0>
   }
   slot {class=VerticalBoxSlot}
  }
  Text_Gamepad_RS_Click_3 {
   widgetClassPath=&CommonTextBlock
   bIsVariable=true
   parent=VerticalBox_88
   widget=Text_Gamepad_RS_Click_3
   properties {
    scrollStyle=&TextScrollStyle-Base_C
    style=&TextStyle-Base-XS-B-DarkBlueFade_C
    text=Right Stick Click 3
    colorAndOpacity {colorUseRule=UseColor_Specified, specifiedColor=<rgba 0.008 0.011 0.0262 0.8>}
    font {
     bForceMonospaced=false
     bMaterialIsStencil=false
     fontMaterial=null
     fontObject=&NotoSans
     letterSpacing=0
     monospacedWidth=1
     size=16
     skewAmount=0
     typefaceFontName=NotoSans-Bold
     outlineSettings {
      bApplyOutlineToDropShadows=false
      bMiteredCorners=false
      bSeparateFillAlpha=false
      outlineMaterial=null
      outlineSize=0
      outlineColor=<rgba 0 0 0 1>
     }
    }
    shadowOffset=<0 0>
   }
   slot {class=VerticalBoxSlot}
  }
  HorizontalBox_CustomButtons {
   widgetClassPath=&HorizontalBox
   bIsVariable=true
   parent=Overlay_4
   widget=HorizontalBox_CustomButtons
   slot {class=OverlaySlot, properties {horizontalAlignment=HAlign_Fill, verticalAlignment=VAlign_Fill}}
  }
  SizeBox_5 {
   widgetClassPath=&SizeBox
   parent=HorizontalBox_CustomButtons
   widget=SizeBox_5
   properties {widthOverride=400}
   slot {class=HorizontalBoxSlot, properties {verticalAlignment=VAlign_Bottom, padding=<ltrb 12 22 15 0>, size {sizeRule=Fill, value=1}}}
  }
  VerticalBox_3 {widgetClassPath=&VerticalBox, parent=SizeBox_5, widget=VerticalBox_3, slot {class=SizeBoxSlot}}
  LT_Custom {
   widgetClassPath=/FortUILibrary/Foundation/Buttons/GamepadKeyTextButton
   bIsVariable=true
   parent=VerticalBox_3
   widget=LT_Custom
   properties {
    buttonText="<Font.Bold>LT</>"
    keyData=Gamepad_LeftTrigger
    navigation=&self:WidgetTree.LT_Custom.WidgetNavigation_0
    onButtonBaseClicked=(&self:WidgetTree.LT_Custom.HandleOnButtonClicked)
    onButtonBaseDoubleClicked=(&self:WidgetTree.LT_Custom.HandleOnButtonDoubleClicked)
    onButtonBaseHovered=(&self:WidgetTree.LT_Custom.HandleOnButtonHovered)
    onButtonBaseUnhovered=(&self:WidgetTree.LT_Custom.HandleOnButtonUnhovered)
    onSelectedChangedBase=(&self:WidgetTree.LT_Custom.HandleOnSelectedChanged)
   }
   slot {class=VerticalBoxSlot, properties {padding=<ltrb 0 0 0 6>}}
  }
  LB_Custom {
   widgetClassPath=&GamepadKeyTextButton
   bIsVariable=true
   parent=VerticalBox_3
   widget=LB_Custom
   properties {
    buttonText="<Font.Bold>LB</>"
    keyData=Gamepad_LeftShoulder
    navigation=&self:WidgetTree.LB_Custom.WidgetNavigation_0
    onButtonBaseClicked=(&self:WidgetTree.LB_Custom.HandleOnButtonClicked)
    onButtonBaseDoubleClicked=(&self:WidgetTree.LB_Custom.HandleOnButtonDoubleClicked)
    onButtonBaseHovered=(&self:WidgetTree.LB_Custom.HandleOnButtonHovered)
    onButtonBaseUnhovered=(&self:WidgetTree.LB_Custom.HandleOnButtonUnhovered)
    onSelectedChangedBase=(&self:WidgetTree.LB_Custom.HandleOnSelectedChanged)
   }
   slot {class=VerticalBoxSlot, properties {padding=<ltrb 0 0 0 6>}}
  }
  DpadUp_Custom {
   widgetClassPath=&GamepadKeyTextButton
   bIsVariable=true
   parent=VerticalBox_3
   widget=DpadUp_Custom
   properties {
    buttonText="<Font.Bold>DpadUp</>"
    keyData=Gamepad_DPad_Up
    navigation=&self:WidgetTree.DpadUp_Custom.WidgetNavigation_0
    onButtonBaseClicked=(&self:WidgetTree.DpadUp_Custom.HandleOnButtonClicked)
    onButtonBaseDoubleClicked=(&self:WidgetTree.DpadUp_Custom.HandleOnButtonDoubleClicked)
    onButtonBaseHovered=(&self:WidgetTree.DpadUp_Custom.HandleOnButtonHovered)
    onButtonBaseUnhovered=(&self:WidgetTree.DpadUp_Custom.HandleOnButtonUnhovered)
    onSelectedChangedBase=(&self:WidgetTree.DpadUp_Custom.HandleOnSelectedChanged)
   }
   slot {class=VerticalBoxSlot, properties {padding=<ltrb 0 60 0 6>}}
  }
  DpadLeft_Custom {
   widgetClassPath=&GamepadKeyTextButton
   bIsVariable=true
   parent=VerticalBox_3
   widget=DpadLeft_Custom
   properties {
    buttonText="<Font.Bold>DpadLeft</>"
    keyData=Gamepad_DPad_Left
    navigation=&self:WidgetTree.DpadLeft_Custom.WidgetNavigation_0
    onButtonBaseClicked=(&self:WidgetTree.DpadLeft_Custom.HandleOnButtonClicked)
    onButtonBaseDoubleClicked=(&self:WidgetTree.DpadLeft_Custom.HandleOnButtonDoubleClicked)
    onButtonBaseHovered=(&self:WidgetTree.DpadLeft_Custom.HandleOnButtonHovered)
    onButtonBaseUnhovered=(&self:WidgetTree.DpadLeft_Custom.HandleOnButtonUnhovered)
    onSelectedChangedBase=(&self:WidgetTree.DpadLeft_Custom.HandleOnSelectedChanged)
   }
   slot {class=VerticalBoxSlot, properties {padding=<ltrb 0 0 0 6>}}
  }
  DpadRight_Custom {
   widgetClassPath=&GamepadKeyTextButton
   bIsVariable=true
   parent=VerticalBox_3
   widget=DpadRight_Custom
   properties {
    buttonText="<Font.Bold>DpadRight</>"
    keyData=Gamepad_DPad_Right
    navigation=&self:WidgetTree.DpadRight_Custom.WidgetNavigation_0
    onButtonBaseClicked=(&self:WidgetTree.DpadRight_Custom.HandleOnButtonClicked)
    onButtonBaseDoubleClicked=(&self:WidgetTree.DpadRight_Custom.HandleOnButtonDoubleClicked)
    onButtonBaseHovered=(&self:WidgetTree.DpadRight_Custom.HandleOnButtonHovered)
    onButtonBaseUnhovered=(&self:WidgetTree.DpadRight_Custom.HandleOnButtonUnhovered)
    onSelectedChangedBase=(&self:WidgetTree.DpadRight_Custom.HandleOnSelectedChanged)
   }
   slot {class=VerticalBoxSlot, properties {padding=<ltrb 0 0 0 6>}}
  }
  DpadDown_Custom {
   widgetClassPath=&GamepadKeyTextButton
   bIsVariable=true
   parent=VerticalBox_3
   widget=DpadDown_Custom
   properties {
    buttonText="<Font.Bold>DpadDown</>"
    keyData=Gamepad_DPad_Down
    navigation=&self:WidgetTree.DpadDown_Custom.WidgetNavigation_0
    onButtonBaseClicked=(&self:WidgetTree.DpadDown_Custom.HandleOnButtonClicked)
    onButtonBaseDoubleClicked=(&self:WidgetTree.DpadDown_Custom.HandleOnButtonDoubleClicked)
    onButtonBaseHovered=(&self:WidgetTree.DpadDown_Custom.HandleOnButtonHovered)
    onButtonBaseUnhovered=(&self:WidgetTree.DpadDown_Custom.HandleOnButtonUnhovered)
    onSelectedChangedBase=(&self:WidgetTree.DpadDown_Custom.HandleOnSelectedChanged)
   }
   slot {class=VerticalBoxSlot, properties {padding=<ltrb 0 0 0 6>}}
  }
  LS_Click_Custom {
   widgetClassPath=&GamepadKeyTextButton
   bIsVariable=true
   parent=VerticalBox_3
   widget=LS_Click_Custom
   properties {
    buttonText="<Font.Bold>LS_Click</>"
    keyData=Gamepad_LeftThumbstick
    navigation=&self:WidgetTree.LS_Click_Custom.WidgetNavigation_0
    onButtonBaseClicked=(&self:WidgetTree.LS_Click_Custom.HandleOnButtonClicked)
    onButtonBaseDoubleClicked=(&self:WidgetTree.LS_Click_Custom.HandleOnButtonDoubleClicked)
    onButtonBaseHovered=(&self:WidgetTree.LS_Click_Custom.HandleOnButtonHovered)
    onButtonBaseUnhovered=(&self:WidgetTree.LS_Click_Custom.HandleOnButtonUnhovered)
    onSelectedChangedBase=(&self:WidgetTree.LS_Click_Custom.HandleOnSelectedChanged)
   }
   slot {class=VerticalBoxSlot, properties {padding=<ltrb 0 60 0 0>}}
  }
  SizeBox_3 {
   widgetClassPath=&SizeBox
   parent=HorizontalBox_CustomButtons
   widget=SizeBox_3
   properties {bOverride_WidthOverride=true, minDesiredWidth=600, widthOverride=600}
   slot {class=HorizontalBoxSlot}
  }
  SizeBox_7 {
   widgetClassPath=&SizeBox
   parent=SizeBox_3
   widget=SizeBox_7
   properties {bOverride_WidthOverride=true, widthOverride=600}
   slot {class=SizeBoxSlot, properties {horizontalAlignment=HAlign_Center, verticalAlignment=VAlign_Center}}
  }
  VerticalBox_6 {
   widgetClassPath=&VerticalBox
   parent=SizeBox_7
   widget=VerticalBox_6
   slot {class=SizeBoxSlot, properties {padding=<ltrb 20 0 20 0>}}
  }
  Spacer_0 {widgetClassPath=&Spacer, parent=VerticalBox_6, widget=Spacer_0, properties {size=<1 647.2286>}, slot {class=VerticalBoxSlot}}
  Start_Custom {
   widgetClassPath=&GamepadKeyTextButton
   bIsVariable=true
   parent=VerticalBox_6
   widget=Start_Custom
   properties {
    bIsFocusable=false
    bSelectable=false
    buttonText="<Font.Bold>Start</>"
    keyData=Gamepad_Special_Right
    onButtonBaseClicked=(&self:WidgetTree.Start_Custom.HandleOnButtonClicked)
    onButtonBaseDoubleClicked=(&self:WidgetTree.Start_Custom.HandleOnButtonDoubleClicked)
    onButtonBaseHovered=(&self:WidgetTree.Start_Custom.HandleOnButtonHovered)
    onButtonBaseUnhovered=(&self:WidgetTree.Start_Custom.HandleOnButtonUnhovered)
    onSelectedChangedBase=(&self:WidgetTree.Start_Custom.HandleOnSelectedChanged)
    visibility=HitTestInvisible
   }
   slot {class=VerticalBoxSlot, properties {padding=<ltrb 0 0 0 6>}}
  }
  Select_Custom {
   widgetClassPath=&GamepadKeyTextButton
   bIsVariable=true
   parent=VerticalBox_6
   widget=Select_Custom
   properties {
    buttonText="<Font.Bold>Select</>"
    keyData=Gamepad_Special_Left
    navigation=&self:WidgetTree.Select_Custom.WidgetNavigation_0
    onButtonBaseClicked=(&self:WidgetTree.Select_Custom.HandleOnButtonClicked)
    onButtonBaseDoubleClicked=(&self:WidgetTree.Select_Custom.HandleOnButtonDoubleClicked)
    onButtonBaseHovered=(&self:WidgetTree.Select_Custom.HandleOnButtonHovered)
    onButtonBaseUnhovered=(&self:WidgetTree.Select_Custom.HandleOnButtonUnhovered)
    onSelectedChangedBase=(&self:WidgetTree.Select_Custom.HandleOnSelectedChanged)
   }
   slot {class=VerticalBoxSlot}
  }
  SizeBox_6 {
   widgetClassPath=&SizeBox
   parent=HorizontalBox_CustomButtons
   widget=SizeBox_6
   slot {class=HorizontalBoxSlot, properties {verticalAlignment=VAlign_Bottom, padding=<ltrb 15 22 12 0>, size {sizeRule=Fill, value=1}}}
  }
  VerticalBox_4 {widgetClassPath=&VerticalBox, parent=SizeBox_6, widget=VerticalBox_4, slot {class=SizeBoxSlot}}
  RT_Custom {
   widgetClassPath=&GamepadKeyTextButton
   bIsVariable=true
   parent=VerticalBox_4
   widget=RT_Custom
   properties {
    buttonText="<Font.Bold>RT</>"
    keyData=Gamepad_RightTrigger
    onButtonBaseClicked=(&self:WidgetTree.RT_Custom.HandleOnButtonClicked)
    onButtonBaseDoubleClicked=(&self:WidgetTree.RT_Custom.HandleOnButtonDoubleClicked)
    onButtonBaseHovered=(&self:WidgetTree.RT_Custom.HandleOnButtonHovered)
    onButtonBaseUnhovered=(&self:WidgetTree.RT_Custom.HandleOnButtonUnhovered)
    onSelectedChangedBase=(&self:WidgetTree.RT_Custom.HandleOnSelectedChanged)
   }
   slot {class=VerticalBoxSlot, properties {padding=<ltrb 0 0 0 6>}}
  }
  RB_Custom {
   widgetClassPath=&GamepadKeyTextButton
   bIsVariable=true
   parent=VerticalBox_4
   widget=RB_Custom
   properties {
    buttonText="<Font.Bold>RB</>"
    keyData=Gamepad_RightShoulder
    onButtonBaseClicked=(&self:WidgetTree.RB_Custom.HandleOnButtonClicked)
    onButtonBaseDoubleClicked=(&self:WidgetTree.RB_Custom.HandleOnButtonDoubleClicked)
    onButtonBaseHovered=(&self:WidgetTree.RB_Custom.HandleOnButtonHovered)
    onButtonBaseUnhovered=(&self:WidgetTree.RB_Custom.HandleOnButtonUnhovered)
    onSelectedChangedBase=(&self:WidgetTree.RB_Custom.HandleOnSelectedChanged)
   }
   slot {class=VerticalBoxSlot, properties {padding=<ltrb 0 0 0 6>}}
  }
  FaceTop_Custom {
   widgetClassPath=&GamepadKeyTextButton
   bIsVariable=true
   parent=VerticalBox_4
   widget=FaceTop_Custom
   properties {
    buttonText="<Font.Bold>FaceTop</>"
    keyData=Gamepad_FaceButton_Top
    onButtonBaseClicked=(&self:WidgetTree.FaceTop_Custom.HandleOnButtonClicked)
    onButtonBaseDoubleClicked=(&self:WidgetTree.FaceTop_Custom.HandleOnButtonDoubleClicked)
    onButtonBaseHovered=(&self:WidgetTree.FaceTop_Custom.HandleOnButtonHovered)
    onButtonBaseUnhovered=(&self:WidgetTree.FaceTop_Custom.HandleOnButtonUnhovered)
    onSelectedChangedBase=(&self:WidgetTree.FaceTop_Custom.HandleOnSelectedChanged)
   }
   slot {class=VerticalBoxSlot, properties {padding=<ltrb 0 60 0 6>}}
  }
  FaceLeft_Custom {
   widgetClassPath=&GamepadKeyTextButton
   bIsVariable=true
   parent=VerticalBox_4
   widget=FaceLeft_Custom
   properties {
    buttonText="<Font.Bold>FaceLeft</>"
    keyData=Gamepad_FaceButton_Left
    onButtonBaseClicked=(&self:WidgetTree.FaceLeft_Custom.HandleOnButtonClicked)
    onButtonBaseDoubleClicked=(&self:WidgetTree.FaceLeft_Custom.HandleOnButtonDoubleClicked)
    onButtonBaseHovered=(&self:WidgetTree.FaceLeft_Custom.HandleOnButtonHovered)
    onButtonBaseUnhovered=(&self:WidgetTree.FaceLeft_Custom.HandleOnButtonUnhovered)
    onSelectedChangedBase=(&self:WidgetTree.FaceLeft_Custom.HandleOnSelectedChanged)
   }
   slot {class=VerticalBoxSlot, properties {padding=<ltrb 0 0 0 6>}}
  }
  FaceRight_Custom {
   widgetClassPath=&GamepadKeyTextButton
   bIsVariable=true
   parent=VerticalBox_4
   widget=FaceRight_Custom
   properties {
    buttonText="<Font.Bold>FaceRight</>"
    keyData=Gamepad_FaceButton_Right
    onButtonBaseClicked=(&self:WidgetTree.FaceRight_Custom.HandleOnButtonClicked)
    onButtonBaseDoubleClicked=(&self:WidgetTree.FaceRight_Custom.HandleOnButtonDoubleClicked)
    onButtonBaseHovered=(&self:WidgetTree.FaceRight_Custom.HandleOnButtonHovered)
    onButtonBaseUnhovered=(&self:WidgetTree.FaceRight_Custom.HandleOnButtonUnhovered)
    onSelectedChangedBase=(&self:WidgetTree.FaceRight_Custom.HandleOnSelectedChanged)
   }
   slot {class=VerticalBoxSlot, properties {padding=<ltrb 0 0 0 6>}}
  }
  FaceBottom_Custom {
   widgetClassPath=&GamepadKeyTextButton
   bIsVariable=true
   parent=VerticalBox_4
   widget=FaceBottom_Custom
   properties {
    buttonText="<Font.Bold>FaceBottom</>"
    keyData=Gamepad_FaceButton_Bottom
    onButtonBaseClicked=(&self:WidgetTree.FaceBottom_Custom.HandleOnButtonClicked)
    onButtonBaseDoubleClicked=(&self:WidgetTree.FaceBottom_Custom.HandleOnButtonDoubleClicked)
    onButtonBaseHovered=(&self:WidgetTree.FaceBottom_Custom.HandleOnButtonHovered)
    onButtonBaseUnhovered=(&self:WidgetTree.FaceBottom_Custom.HandleOnButtonUnhovered)
    onSelectedChangedBase=(&self:WidgetTree.FaceBottom_Custom.HandleOnSelectedChanged)
   }
   slot {class=VerticalBoxSlot, properties {padding=<ltrb 0 0 0 6>}}
  }
  RS_Click_Custom {
   widgetClassPath=&GamepadKeyTextButton
   bIsVariable=true
   parent=VerticalBox_4
   widget=RS_Click_Custom
   properties {
    buttonText="<Font.Bold>RS_Click</>"
    keyData=Gamepad_RightThumbstick
    navigation=&self:WidgetTree.RS_Click_Custom.WidgetNavigation_0
    onButtonBaseClicked=(&self:WidgetTree.RS_Click_Custom.HandleOnButtonClicked)
    onButtonBaseDoubleClicked=(&self:WidgetTree.RS_Click_Custom.HandleOnButtonDoubleClicked)
    onButtonBaseHovered=(&self:WidgetTree.RS_Click_Custom.HandleOnButtonHovered)
    onButtonBaseUnhovered=(&self:WidgetTree.RS_Click_Custom.HandleOnButtonUnhovered)
    onSelectedChangedBase=(&self:WidgetTree.RS_Click_Custom.HandleOnSelectedChanged)
   }
   slot {class=VerticalBoxSlot, properties {padding=<ltrb 0 60 0 2>}}
  }
  CommonBorder_Legend {
   widgetClassPath=&CommonBorder
   bIsVariable=true
   parent=Overlay_4
   widget=CommonBorder_Legend
   properties {
    visibility=HitTestInvisible
    background {
     bIsDynamicallyLoaded=false
     drawAs=Box
     imageType=NoImage
     mirroring=NoMirror
     resourceName=None
     resourceObject=null
     tiling=NoTile
     imageSize=<32 32>
     margin=<ltrb 0 0 0 0>
     outlineSettings {
      bUseBrushTransparency=false
      roundingType=HalfHeightRadius
      width=0
      color {colorUseRule=UseColor_Specified, specifiedColor=<rgba 0 0 0 0>}
      cornerRadii=<0 0 0 0>
     }
     tintColor {colorUseRule=UseColor_Specified, specifiedColor=<rgba 0.1195 0.2122 0.3864 1>}
     uVRegion {bIsValid=false, max=<0 0>, min=<0 0>}
    }
   }
   slot {class=OverlaySlot, properties {horizontalAlignment=HAlign_Right, padding=<ltrb 0 10 10 0>}}
  }
  VerticalBox_7 {
   widgetClassPath=&VerticalBox
   parent=CommonBorder_Legend
   widget=VerticalBox_7
   properties {visibility=HitTestInvisible}
   slot {class=BorderSlot}
  }
  HorizontalBox_19 {
   widgetClassPath=&HorizontalBox
   parent=VerticalBox_7
   widget=HorizontalBox_19
   properties {visibility=HitTestInvisible}
   slot {class=VerticalBoxSlot}
  }
  Image_Hold {
   widgetClassPath=&Image
   bIsVariable=true
   parent=HorizontalBox_19
   widget=Image_Hold
   properties {
    visibility=SelfHitTestInvisible
    brush {
     bIsDynamicallyLoaded=false
     drawAs=Image
     imageType=NoImage
     mirroring=NoMirror
     resourceName=None
     resourceObject=null
     tiling=NoTile
     imageSize=<16 16>
     margin=<ltrb 0 0 0 0>
     outlineSettings {
      bUseBrushTransparency=false
      roundingType=HalfHeightRadius
      width=0
      color {colorUseRule=UseColor_Specified, specifiedColor=<rgba 0 0 0 0>}
      cornerRadii=<0 0 0 0>
     }
     tintColor {colorUseRule=UseColor_Specified, specifiedColor=<rgba 0 0.4676 1 1>}
     uVRegion {bIsValid=false, max=<0 0>, min=<0 0>}
    }
   }
   slot {class=HorizontalBoxSlot, properties {padding=<ltrb 4 10 4 10>}}
  }
  FortRichTextBlock_Hold {
   widgetClassPath=&FortRichTextBlockLegacy
   bIsVariable=true
   parent=HorizontalBox_19
   widget=FortRichTextBlock_Hold
   properties {
    styleSet=&TextStyle_Gamepad_Bindings
    text="<Font.Bold>Hold Action</>"
    visibility=SelfHitTestInvisible
   }
   slot {class=HorizontalBoxSlot, properties {horizontalAlignment=HAlign_Left, verticalAlignment=VAlign_Center, padding=<ltrb 10 0 0 0>}}
  }
  HorizontalBox_20 {
   widgetClassPath=&HorizontalBox
   parent=VerticalBox_7
   widget=HorizontalBox_20
   properties {visibility=HitTestInvisible}
   slot {class=VerticalBoxSlot}
  }
  Image_DoubleTap {
   widgetClassPath=&Image
   bIsVariable=true
   parent=HorizontalBox_20
   widget=Image_DoubleTap
   properties {
    visibility=SelfHitTestInvisible
    brush {
     bIsDynamicallyLoaded=false
     drawAs=Image
     imageType=NoImage
     mirroring=NoMirror
     resourceName=None
     resourceObject=null
     tiling=NoTile
     imageSize=<16 16>
     margin=<ltrb 0 0 0 0>
     outlineSettings {
      bUseBrushTransparency=false
      roundingType=HalfHeightRadius
      width=0
      color {colorUseRule=UseColor_Specified, specifiedColor=<rgba 0 0 0 0>}
      cornerRadii=<0 0 0 0>
     }
     tintColor {colorUseRule=UseColor_Specified, specifiedColor=<rgba 0.9387 0.2404 1 1>}
     uVRegion {bIsValid=false, max=<0 0>, min=<0 0>}
    }
   }
   slot {class=HorizontalBoxSlot, properties {padding=<ltrb 4 10 4 10>}}
  }
  FortRichTextBlock_DoubleTap {
   widgetClassPath=&FortRichTextBlockLegacy
   bIsVariable=true
   parent=HorizontalBox_20
   widget=FortRichTextBlock_DoubleTap
   properties {
    styleSet=&TextStyle_Gamepad_Bindings
    text="<Font.Bold>Double Tap Action</>"
    visibility=SelfHitTestInvisible
   }
   slot {class=HorizontalBoxSlot, properties {horizontalAlignment=HAlign_Left, verticalAlignment=VAlign_Center, padding=<ltrb 10 0 0 0>}}
  }
  Image_LightBox {
   widgetClassPath=&Image
   bIsVariable=true
   parent=Overlay_0
   widget=Image_LightBox
   properties {visibility=Hidden, colorAndOpacity=<rgba 0.0104 0.0104 0.0104 0.6>}
   slot {class=OverlaySlot, properties {horizontalAlignment=HAlign_Fill, verticalAlignment=VAlign_Fill, padding=<ltrb -500 -500 -500 -500>}}
  }
  Overlay_ActionsList {
   widgetClassPath=&Overlay
   parent=Overlay_0
   widget=Overlay_ActionsList
   slot {class=OverlaySlot, properties {horizontalAlignment=HAlign_Center, verticalAlignment=VAlign_Center, padding=<ltrb 0 -100 0 -100>}}
  }
  GamepadCustomActionSelectorPanel_v3 {
   widgetClassPath=&/GamepadCustomActionSelectorPanel_v3
   bIsVariable=true
   parent=Overlay_ActionsList
   widget=GamepadCustomActionSelectorPanel_v3
   properties {bIsModal=true, visibility=Collapsed}
   slot {class=OverlaySlot, properties {horizontalAlignment=HAlign_Fill, verticalAlignment=VAlign_Fill, padding=<ltrb 2 2 2 2>}}
  }
 ]
}
