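# Options tab widget for gamepad configuration: controller preset, controller platform, custom bindings and deadzones.
# Selector panel radio groups and mode tabs set preset, platform and action group indices that refresh the mapping view.
# For custom presets, picking a key opens an action list; rebinding rejects triggers and shoulders for full-screen map actions and confirms unbinding conflicts.
# Back is blocked while important actions are unbound, and reset offers non-custom presets.
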WidgetBlueprint /FortniteCoreUI/Settings/CustomBindings_v3/TabGamePadConfig_v3
graphs [
 AreWeOkayWithThisChange {nodes [
   n1 = AreWeOkayWithThisChange!() -> (then, ActionName:Name, Key:Key Structure) {is=title K2Node_FunctionEntry}
   > n2 = Sequence()
   Return Node!(execute=[@n2.then_1, @n5.then], IsOkay:bool=Get IsOkay_Local() -> (IsOkay_Local:bool))
   n3 = Contains(SearchIn="To String (Name)"(InName=@n1), Substring=Get FullScreenMap_Local() -> (FullScreenMap_Local:String))
   n4 = OR Boolean!(
    A="Equal (Key)"!(A:Key Structure=@n1.Key, B:Key Structure=Gamepad_LeftTrigger) -> (ReturnValue:bool)
    B="Equal (Key)"(A=@n1.Key, B=Gamepad_RightTrigger)
    C:bool="Equal (Key)"(A=@n1.Key, B=Gamepad_LeftShoulder)
    D:bool="Equal (Key)"(A=@n1.Key, B=Gamepad_RightShoulder)
   ) -> (ReturnValue)
   > n5 = Set IsOkay_Local(IsOkay_Local:bool=NOT Boolean(A=AND Boolean(A=@n4, B=@n3))) -> (Output_Get:bool)
  ]}
 ChangeCustomClearVisibilityForInputType {nodes [
   n1 = ChangeCustomClearVisibilityForInputType!() -> (then, InputType:ECommonInputType Enum) {is=title K2Node_FunctionEntry}
   > n2 = Switch on ECommonInputType!(execute, Selection:ECommonInputType Enum=@n1) -> (MouseAndKeyboard:Exec, Gamepad:Exec, Touch:Exec, Count:Exec) {is=title K2Node_SwitchEnum}
   n3 = Get CommonBorder_Clear!(self:Gamepad Custom Action Selector Panel V 3 Object Reference=Get GamepadCustomActionSelectorPanel_v3!(self:Gamepad Mapping Info V 3 Object Reference=Get GamepadMappingInfo_v3() -> (GamepadMappingInfo_v3:Gamepad Mapping Info V 3 Object Reference)) -> (GamepadCustomActionSelectorPanel_v3:Gamepad Custom Action Selector Panel V 3 Object Reference)) -> (CommonBorder_Clear:Common Border Object Reference)
   SetVisibility(execute=@n2.Gamepad, self=@n3, InVisibility=SelfHitTestInvisible)
   SetVisibility(execute=@n2.MouseAndKeyboard, self=@n3, InVisibility=Visible)
  ]}
 ChangeCustomResetVisibilityForInputType {nodes [
   n1 = ChangeCustomResetVisibilityForInputType!() -> (then, InputType:ECommonInputType Enum) {is=title K2Node_FunctionEntry}
   > n2 = Switch on ECommonInputType(Selection=@n1)
   n3 = Get CommonBorder_Reset!(self:Gamepad Info Selector Panel V 3 Object Reference=Get GamepadInfoSelectorPanel_v3() -> (GamepadInfoSelectorPanel_v3:Gamepad Info Selector Panel V 3 Object Reference)) -> (CommonBorder_Reset:Common Border Object Reference)
   SetVisibility(execute=@n2.Gamepad, self=@n3, InVisibility=SelfHitTestInvisible)
   SetVisibility(execute=@n2.MouseAndKeyboard, self=@n3, InVisibility=Visible)
  ]}
 ChangeCustomTabListVisibilityForInputType {nodes [
   n1 = ChangeCustomTabListVisibilityForInputType!() -> (then, Selection:ECommonInputType Enum) {is=title K2Node_FunctionEntry}
   > n2 = Switch on ECommonInputType(Selection=@n1)
   n3 = Get HorizontalTabList_CustomModes!(self:Gamepad Mapping Info V 3 Object Reference=Get GamepadMappingInfo_v3()) -> (HorizontalTabList_CustomModes:Horizontal Tab List Legacy Object Reference)
   SetVisibility(execute=@n2.Gamepad, self=@n3, InVisibility=SelfHitTestInvisible)
   SetVisibility(execute=@n2.MouseAndKeyboard, self=@n3, InVisibility=Visible)
  ]}
 ChangeTheBinding {nodes [
   n1 = Inputs!() -> (In:Exec, ActionName:Name, Key:Key Structure)
   n2 = GetInputDataIndexForActionName!(self:Fort Input Options Object Reference, ActionName:Name=@n1) -> (ReturnValue:int=0)
   n3 = Break Fort Action Being Unbound!(FortActionBeingUnbound:"Fort Action Being Unbound Structure (by ref)"=GetActionBeingUnbound!(self:Fort Input Options Object Reference, bIsPrimarySlot:bool=false, Input:int=@n2, NewKey:Key Structure=@n1.Key) -> (ReturnValue:Fort Action Being Unbound Structure)) -> (ActionName:Name, InputIndex:int) {is=title K2Node_BreakStruct}
   n4 = Branch(Condition="Not Equal (Name)"!(A:Name=@n3, B:Name=None) -> (ReturnValue:bool) {is=title K2Node_CallFunction})
   > n5 = SetupDisplayObject!(execute, self, ActionName:Name=@n3, Key:Key Structure=@n1.Key) -> (then, DisplayObject:Gamepad Key Text Button Object Reference)
   > n6 = ShowSimpleConfirmationDialog_NUI!(
    execute:Exec
    Icon:Slate Brush Structure
    Title:Text="LOCTEXT(\"Warning!\")"
    Message:Text="LOCTEXT(\"This change will unbind another action.\\r\\nAre you sure?\")"
    bShowConfirm:bool=true
    bShowDecline:bool=true
    AdditionalContent:Widget Object Reference=@n5
    LeftAdditionalContent:Widget Object Reference
    TargetPlayer:Local Player Object Reference
   ) -> (then:Exec, DialogResult:Exec, Result:EFortDialogResult Enum, ResultName:Name) {is=title K2Node_AsyncAction}
   n7 = Branch(execute=@n6.DialogResult, Condition="Equal (Enum)"(A:EFortDialogResult Enum=@n6, B:EFortDialogResult Enum=Confirmed))
   n8 = ChangeBinding!(
    execute=[@n7.then, @n4.else, @n1.In]
    self:Fort Input Options Object Reference
    bIsPrimarySlot:bool=false
    Input:int=@n2
    NewKey:Key Structure=@n1.Key
   ) -> (then)
   Outputs!(Out:Exec=[@n8.then, @n7.else])
  ]}
 CloseCustomActionsPanel {nodes [
   n1 = CloseCustomActionsPanel!() -> (then, FocusLastButton:bool=~) {is=title K2Node_FunctionEntry}
   n2 = Get GamepadMappingInfo_v3()
   > PopActivatablePanel!(execute, Panel:Common Activatable Panel Legacy Object Reference=Get GamepadCustomActionSelectorPanel_v3(self=@n2)) -> (then)
   > SetVisibility(self=Get Image_LightBox!(self:Gamepad Mapping Info V 3 Object Reference=@n2) -> (Image_LightBox:Image Object Reference), InVisibility=Hidden)
   > UpdateCustomButtons(self:Gamepad Mapping Info V 3 Object Reference=@n2)
   n3 = Get LastClickedButton!(self:Gamepad Mapping Info V 3 Object Reference=@n2) -> (LastClickedButton:Gamepad Key Text Button Object Reference)
   > ClearSelection(self:Common Button Base Object Reference=@n3)
   > Branch(Condition=@n1)
   > CenterWidget!(execute, ToCenter:Widget Object Reference=@n3) -> (then)
  ]}
 CreateCustomResetToDefaultList {nodes [
   CreateCustomResetToDefaultList() {is=title K2Node_FunctionEntry}
   > n1 = For Each Loop(Array:Array of Texts=Get InputPresetNames_NoCustom() -> (InputPresetNames_NoCustom:Array of Texts)) -> (Array Element:Text)
   Return Node!(
    execute=@n1.Completed
    ReturnValue:Array of Confirmation Dialog Action Structures=Get ReturnValue_Local() -> (ReturnValue_Local:Array of Confirmation Dialog Action Structures)
   )
   n2 = "Get (a ref)"(Array:Array of Strings=Get InputPresets() -> (InputPresets:Array of Strings), Dimension 1=@n1.Array Index) -> (Output:"String (by ref)")
   > Add(
    TargetArray:Array of Confirmation Dialog Action Structures=Get ReturnValue_Local()
    NewItem:"Confirmation Dialog Action Structure (by ref)"=Make Confirmation Dialog Action!(
     DisplayName:Text=@n1
     ResultName:Name=String To Name(InString=@n2)
     Icon:Slate Brush Structure=(TintColor=(SpecifiedColor=(R=1,G=1,B=1,A=1),ColorUseRule=UseColor_Specified),DrawAs=Image,Tiling=NoTile,Mirroring=NoMirror,ImageType=NoImage,ImageSize=(X=32,Y=32),Margin=(Left=0,Top=0,Right=0,Bottom=0),ResourceObject=None,OutlineSettings=(CornerRadii=(X=0,Y=0,Z=0,W=0),Color=(SpecifiedColor=(R=0,G=0,B=0,A=0),ColorUseRule=UseColor_Specified),Width=0,RoundingType=HalfHeightRadius,bUseBrushTransparency=False),UVRegion=(Min=(X=0,Y=0),Max=(X=0,Y=0),bIsValid=False),bIsDynamicallyLoaded=False,ResourceName="")
     ActionName:Name=None
    ) -> (ConfirmationDialogAction:Confirmation Dialog Action Structure) {is=title K2Node_MakeStruct}
   ) {is=title K2Node_CallArrayFunction}
  ]}
 EventGraph {nodes [
   DeadZoneLookSettingValueChanged_Bind!() -> (OutputDelegate, then, Value:float=~) {is=title K2Node_CustomEvent}
   > Call On Tab Setting Changed!(execute:Exec, self:Tab Game Pad Config V 3 Object Reference) -> (then:Exec) {is=title K2Node_CallDelegate}
   > Call Gamepad Mapping Info Changed!(execute:Exec, self:Tab Game Pad Config V 3 Object Reference) -> (then:Exec) {is=title K2Node_CallDelegate}
   CustomDoResetToDefault() {is=title K2Node_CustomEvent}
   > n1 = Branch(Condition=IsInSubGame!(SubGame:ESubGame Enum=Campaign) -> (ReturnValue:bool))
   > Call Gamepad Info Selector Reset!(execute:Exec, self:Tab Game Pad Config V 3 Object Reference) -> (then:Exec) {is=title K2Node_CallDelegate}
   Set CustomResetToDefaultConfig(execute=@n1.else, CustomResetToDefaultConfig:int=0) -> (Output_Get:int=0)
   > n2 = CreateCustomResetToDefaultList!(execute, self) -> (then, ReturnValue:Array of Confirmation Dialog Action Structures) {is=title K2Node_CallFunction}
   > n3 = ShowConfirmationDialog_NUI!(
    execute:Exec
    Icon:Slate Brush Structure
    Title:Text="LOCTEXT(\"Reset To Default\")"
    Message:Text="LOCTEXT(\"Which configuration would you like to reset to?\")"
    ConfirmButtonInputActions:Array of Confirmation Dialog Action Structures=@n2
    DeclineButtonInputAction:"Name (by ref)"=MakeLiteralName(Value=Cancel)
    AdditionalContent:Widget Object Reference
    LeftAdditionalContent:Widget Object Reference
    TargetPlayer:Local Player Object Reference
   ) -> (then:Exec, DialogResult:Exec, Result:EFortDialogResult Enum, ResultName:Name) {is=title K2Node_AsyncAction}
   n4 = Branch(execute=@n3.DialogResult, Condition="Equal (Enum)"(A=@n3, B=Confirmed))
   > SetCurrentCustomInputTemplatePresetName!(
    execute
    self:Fort Global UIContext Object Reference=Get FortGlobalUIContext!() -> (ReturnValue:Fort Global UIContext Object Reference) {is=title K2Node_GetBlueprintContextEx}
    InputPresetName:String="To String (Name)"(InName=@n3.ResultName)
   ) -> (then)
   > HandleApply!(execute, self:Gamepad Info Selector Panel V 3 Object Reference=Get GamepadInfoSelectorPanel_v3()) -> (then, PassThrough:bool)
   > ResetCustomGamepadToDefault(self:Fort Gamepad Input Options Object Reference)
   > n5 = UpdateCustomButtons(self=Get GamepadMappingInfo_v3())
   NavigateToSelectedConfig(execute=[@n5.then, @n4.else])
   > BroadcastModalClosed(self:Fort Gamepad Input Options Object Reference)
   DeadZoneMoveSettingValueChanged_Bind!() -> (OutputDelegate, then, Value:float=~) {is=title K2Node_CustomEvent}
   > Call On Tab Setting Changed()
   > Call Gamepad Mapping Info Changed()
   CustomClearButton_Bind() {is=title K2Node_CustomEvent}
   n6 = Get GamepadMappingInfo_v3()
   > Change the Binding!(
    In:Exec
    ActionName:Name=Get ActionNameData!(self:Gamepad Key Text Button Object Reference=Get LastClickedButton(self=@n6)) -> (ActionNameData:Name=None)
    Key:Key Structure=Get Key_None() -> (Key_None:Key Structure)
   ) -> (Out:Exec) {is=title K2Node_MacroInstance}
   > CloseCustomActionsPanel!(execute, self, FocusLastButton:bool=true) -> (then) {is=title K2Node_CallFunction}
   n7 = GamepadPlatformConfigChanged!() -> (OutputDelegate, then, AssociatedButton:Common Button Legacy Object Reference, ButtonIndex:int) {is=title K2Node_CustomEvent}
   > n8 = Set PlatformMappingIndex(PlatformMappingIndex:int=Select!(Option 0:int=0, Option 1:int=1, Option 2:int=2, Option 3:int=3, Option 4:int=4, Option 5:int=5, Option 6:int=6, Index:int=@n7.ButtonIndex) -> (ReturnValue:int)) -> (Output_Get:int=0)
   > Call On Tab Setting Changed()
   > SetControllerPlatformEnum!(
    execute
    self:Fort Gamepad Input Options Object Reference
    InControllerPlatform:ECommonGamepadType Enum=Select(
     Option 0:ECommonGamepadType Enum=XboxOneController
     Option 1:ECommonGamepadType Enum=PS4Controller
     Option 2:ECommonGamepadType Enum=GenericController
     Option 3:ECommonGamepadType Enum=XboxSeriesXController
     Option 4:ECommonGamepadType Enum=PS5Controller
     Option 5:ECommonGamepadType Enum=SwitchController
     Option 6:ECommonGamepadType Enum=Switch2Controller
     Index=@n8
    ) -> (ReturnValue:ECommonGamepadType Enum)
   ) -> (then)
   > HandleGamepadMappingInfoChange()
   n9 = ModesTabSelected_Bind!() -> (OutputDelegate, then, TabId:Name) {is=title K2Node_CustomEvent}
   n10 = Get GamepadMappingInfo_v3()
   > n11 = Branch(Condition="Equal (Name)"!(A:Name=@n9.TabId, B:Name=Get CombatModeIdName!(self:Gamepad Mapping Info V 3 Object Reference=@n10) -> (CombatModeIdName:Name=None)) -> (ReturnValue:bool) {is=title K2Node_CallFunction})
   > n12 = Set CurrentInputActionGroup(CurrentInputActionGroup:EFortInputActionGroup Enum=Combat) -> (Output_Get:EFortInputActionGroup Enum=AllModes)
   n13 = Get GamepadMappingInfo_v3()
   n14 = Set CurrentInputActionGroup!(
    execute=[@n20.then, @n23.then, @n17.then, @n12.then]
    CurrentInputActionGroup=Get CurrentInputActionGroup() -> (CurrentInputActionGroup:EFortInputActionGroup Enum=AllModes)
    self:Gamepad Mapping Info V 3 Object Reference=@n13
   ) -> (then, Output_Get=AllModes)
   > SetTextForMode!(execute, self:Gamepad Mapping Info V 3 Object Reference=@n13, Selection:EFortInputActionGroup Enum=@n14) -> (then)
   > DeadZoneTabDeselected(self:Gamepad Mapping Info V 3 Object Reference=@n13)
   > SetVisibility(self=Get GamepadInfoSelectorPanel_v3(), InVisibility=SelfHitTestInvisible)
   > HandleGamepadMappingInfoChange()
   > NavigateToSelectedConfig()
   n15 = Get GamepadMappingInfo_v3()
   n16 = Branch(
    execute=@n11.else
    Condition="Equal (Name)"(A=@n9.TabId, B=Get BuildModeIdName!(self:Gamepad Mapping Info V 3 Object Reference=@n15) -> (BuildModeIdName:Name=None))
   )
   > n17 = Set CurrentInputActionGroup!(execute, CurrentInputActionGroup=Building) -> (then, Output_Get=AllModes)
   n18 = Get GamepadMappingInfo_v3()
   n19 = Branch(
    execute=@n16.else
    Condition="Equal (Name)"(A=@n9.TabId, B=Get EditModeIdName!(self:Gamepad Mapping Info V 3 Object Reference=@n18) -> (EditModeIdName:Name=None))
   )
   > n20 = Set CurrentInputActionGroup(CurrentInputActionGroup=Edit) -> (Output_Get=AllModes)
   n21 = Get GamepadMappingInfo_v3()
   n22 = Branch(
    execute=@n19.else
    Condition="Equal (Name)"(A=@n9.TabId, B=Get LTMAbilitiesModeIdName!(self:Gamepad Mapping Info V 3 Object Reference=@n21) -> (LTMAbilitiesModeIdName:Name=None))
   )
   > n23 = Set CurrentInputActionGroup(CurrentInputActionGroup=AthenaLTMAbilities) -> (Output_Get=AllModes)
   n24 = Get GamepadMappingInfo_v3()
   Branch(
    execute=@n22.else
    Condition="Equal (Name)"(A=@n9.TabId, B=Get DeadZoneMoveIdName!(self:Gamepad Mapping Info V 3 Object Reference=@n24) -> (DeadZoneMoveIdName:Name=None))
   )
   > DeadZoneTabSelected(self:Gamepad Mapping Info V 3 Object Reference=Get GamepadMappingInfo_v3())
   > SetVisibility(self=Get GamepadInfoSelectorPanel_v3(), InVisibility=Collapsed)
   n25 = CustomKeySelected_Bind!() -> (OutputDelegate, then, Key:Key Structure=None) {is=title K2Node_CustomEvent}
   > Set CustomKeySelection(CustomKeySelection:Key Structure=@n25.Key) -> (Output_Get:Key Structure)
   > HandleApply(self=Get GamepadInfoSelectorPanel_v3())
   > Set InputDataForActionsPanel(InputDataForActionsPanel:Array of Fort Options Menu Input Data Object References=GetActionItemsInGroup!(
     self:Fort Input Options Object Reference
     InputActionGroup:EFortInputActionGroup Enum=Get CurrentInputActionGroup() -> (CurrentInputActionGroup=AllModes)
    ) -> (ReturnValue:Array of Fort Options Menu Input Data Object References)) -> (Output_Get:Array of Fort Options Menu Input Data Object References)
   n26 = Get GamepadMappingInfo_v3()
   n27 = Get GamepadCustomActionSelectorPanel_v3(self=@n26)
   > UpdatePanel!(
    execute
    self:Gamepad Custom Action Selector Panel V 3 Object Reference=@n27
    InputData:Array of Fort Options Menu Input Data Object References=Get InputDataForActionsPanel() -> (InputDataForActionsPanel:Array of Fort Options Menu Input Data Object References)
   ) -> (then)
   > PushActivatablePanel!(execute, Panel:Common Activatable Panel Legacy Object Reference=@n27) -> (then)
   > SelectSameItemInActionsList()
   > SetVisibility(self=Get Image_LightBox(self=@n26), InVisibility=Visible)
   > ChangeCustomClearVisibilityForInputType!(
    execute
    self
    InputType:ECommonInputType Enum=GetCurrentInputType(self=Get CommonInputSubsystem!() -> (ReturnValue:Common Input Object Reference) {is=title K2Node_GetBlueprintContextEx})
   ) -> (then) {is=title K2Node_CallFunction}
   n28 = Gamepad Config Changed!() -> (OutputDelegate, then, AssociatedButton:Common Button Legacy Object Reference, ButtonIndex:int) {is=title K2Node_CustomEvent}
   > n29 = Set PresetMappingIndex(PresetMappingIndex:int=@n28.ButtonIndex) -> (Output_Get:int=0)
   > Call On Tab Setting Changed()
   > HandleGamepadMappingInfoChange()
   > SetCurrentInputPresetName!(
    execute
    self:Fort Global UIContext Object Reference=Get FortGlobalUIContext()
    InputPresetName:String="Get (a copy)"(Array=Get InputPresets(), Dimension 1=@n29)
   ) -> (then)
   OnTabDeactivated_Bind() {is=title K2Node_CustomEvent}
   > PopActivatablePanel(Panel=Get GamepadInfoSelectorPanel_v3())
   > n30 = Switch on ESubGame!(
    execute
    Selection:ESubGame Enum=GetSubGame!(self:Fort Global UIContext Object Reference=Get FortGlobalUIContext()) -> (ReturnValue:ESubGame Enum=Campaign)
   ) -> (Campaign:Exec, Athena:Exec, Invalid:Exec, Count:Exec, Creative:Exec) {is=title K2Node_SwitchEnum}
   CloseCustomActionsPanel(execute=@n30.Athena, FocusLastButton=false)
   n31 = CustomActionSelected_Bind!() -> (OutputDelegate, then, SelectedIndex:int, ActionName:Name) {is=title K2Node_CustomEvent}
   > Set CustomActionSelectionIndex(CustomActionSelectionIndex:int=@n31.SelectedIndex) -> (Output_Get:int=0)
   > n32 = Set CustomActionSelection(CustomActionSelection:Name=@n31.ActionName) -> (Output_Get:Name=None)
   > n33 = AreWeOkayWithThisChange!(execute, self, ActionName:Name=@n32, Key:Key Structure=Get CustomKeySelection() -> (CustomKeySelection:Key Structure)) -> (then, IsOkay:bool) {is=title K2Node_CallFunction}
   > n34 = Branch(Condition=@n33)
   > n35 = Change the Binding(ActionName=Get CustomActionSelection() -> (CustomActionSelection:Name=None), Key=Get CustomKeySelection())
   CloseCustomActionsPanel(execute=[@n36.DialogResult, @n35.Out], FocusLastButton=true)
   > CenterOnTab(self:Fort Options Tab Object Reference)
   n36 = ShowSimpleConfirmationDialog_NUI(
    execute=@n34.else
    Title="LOCTEXT(\"Change Not Allowed\")"
    Message="LOCTEXT(\"We aren\\'t allowing this action to be bound to this key yet.\\r\\nThis may come in future updates.\")"
    bShowConfirm=true
    bShowDecline=false
   )
   OnTabActivated_Bind() {is=title K2Node_CustomEvent}
   n37 = Get GamepadInfoSelectorPanel_v3()
   > Bind Event to Gamepad Info Selector Focus(self:Gamepad Info Selector Panel V 3 Object Reference=@n37, Delegate=Create Event()) {is=title K2Node_AddDelegate}
   > PushActivatablePanel(Panel=@n37)
   CustomClosedUsingBack_Bind() {is=title K2Node_CustomEvent}
   > CloseCustomActionsPanel(FocusLastButton=true)
   Event Construct()
   > n38 = Sequence!(execute) -> (then_0, then_1, then_2:Exec, then_3:Exec)
   > InitializeData()
   > UpdateData()
   HandleGamepadMappingInfoChange(execute=@n38.then_1)
   > Bind Event to On Tab Activated(self:Tab Game Pad Config V 3 Object Reference, Delegate=Create Event()) {is=title K2Node_AddDelegate}
   > Bind Event to On Tab Deactivated(self:Tab Game Pad Config V 3 Object Reference, Delegate=Create Event()) {is=title K2Node_AddDelegate}
   n39 = Get GamepadInfoSelectorPanel_v3()
   > Bind Event to On Selected Button Changed(
    self:Common Button Group Legacy Object Reference=Get CreatedButtonsGroup!(self:Fort Radio Button Group Object Reference=Get FortRadioButtonGroup_Configs!(self:Gamepad Info Selector Panel V 3 Object Reference=@n39) -> (FortRadioButtonGroup_Configs:Fort Radio Button Group Object Reference)) -> (CreatedButtonsGroup:Common Button Group Legacy Object Reference)
    Delegate=Create Event()
   ) {is=title K2Node_AddDelegate}
   > Bind Event to Gamepad Info Selector Back(self:Gamepad Info Selector Panel V 3 Object Reference=@n39, Delegate=Create Event()) {is=title K2Node_AddDelegate}
   > Bind Event to Gamepad Info Selector Apply(self:Gamepad Info Selector Panel V 3 Object Reference=@n39, Delegate=Create Event()) {is=title K2Node_AddDelegate}
   > Bind Event to Gamepad Info Selector Reset(self:Gamepad Info Selector Panel V 3 Object Reference=@n39, Delegate=Create Event()) {is=title K2Node_AddDelegate}
   n40 = Get GamepadMappingInfo_v3()
   > Bind Event to Custom Key Selected(self:Gamepad Mapping Info V 3 Object Reference=@n40, Delegate=Create Event()) {is=title K2Node_AddDelegate}
   n41 = Get GamepadCustomActionSelectorPanel_v3(self=@n40)
   > Bind Event to Custom Action Selected(self:Gamepad Custom Action Selector Panel V 3 Object Reference=@n41, Delegate=Create Event()) {is=title K2Node_AddDelegate}
   > Bind Event to Custom Closed Using Back(self:Gamepad Custom Action Selector Panel V 3 Object Reference=@n41, Delegate=Create Event()) {is=title K2Node_AddDelegate}
   > Bind Event to Custom Clear Button(self:Gamepad Custom Action Selector Panel V 3 Object Reference=@n41, Delegate=Create Event()) {is=title K2Node_AddDelegate}
   > Bind Event to On Tab Selected(self:Common Tab List Widget Base Object Reference=Get HorizontalTabList_CustomModes(self=@n40), Delegate=Create Event()) {is=title K2Node_AddDelegate}
   > Bind Event to Image Lightbox Clicked(self:Gamepad Mapping Info V 3 Object Reference=@n40, Delegate=Create Event()) {is=title K2Node_AddDelegate}
   > Bind Event to Custom Button Nav Left(self:Gamepad Mapping Info V 3 Object Reference=@n40, Delegate=Create Event()) {is=title K2Node_AddDelegate}
   > Bind Event to Setting Value Changed(
    self:Fort Options Menu Setting Object Reference=Get Deadzone_Adjustment_Slider!(self:Deadzone Adjustment Object Reference=Get DeadzoneAdjustment_Look!(self:Gamepad Mapping Info V 3 Object Reference=@n40) -> (DeadzoneAdjustment_Look:Deadzone Adjustment Object Reference)) -> (Deadzone_Adjustment_Slider:Options Menu Setting Object Reference)
    Delegate=Create Event()
   ) {is=title K2Node_AddDelegate}
   > Bind Event to Setting Value Changed(
    self=Get Deadzone_Adjustment_Slider(self=Get DeadzoneAdjustment_Move!(self:Gamepad Mapping Info V 3 Object Reference=@n40) -> (DeadzoneAdjustment_Move:Deadzone Adjustment Object Reference))
    Delegate=Create Event()
   )
   ChangeCustomTabListVisibilityForInputType!(execute=@n38.then_3, self, Selection:ECommonInputType Enum=GetCurrentInputType(self=Get CommonInputSubsystem())) -> (then) {is=title K2Node_CallFunction}
   n42 = Branch(execute=@n38.then_2, Condition=IsControllerDefaultInputType!(self:Fort Options Tab Object Reference) -> (ReturnValue:bool))
   n43 = Get VerticalBox_Platforms!(self:Gamepad Info Selector Panel V 3 Object Reference=Get GamepadInfoSelectorPanel_v3()) -> (VerticalBox_Platforms:Vertical Box Object Reference)
   > SetVisibility(self=@n43, InVisibility=Collapsed)
   SetVisibility(execute=@n42.else, self=@n43, InVisibility=Visible)
   n44 = Get GamepadInfoSelectorPanel_v3()
   > Bind Event to On Selected Button Changed(
    self=Get CreatedButtonsGroup(self=Get FortRadioButtonGroup_Platforms!(self:Gamepad Info Selector Panel V 3 Object Reference=@n44) -> (FortRadioButtonGroup_Platforms:Fort Radio Button Group Object Reference))
    Delegate=Create Event()
   )
   n45 = Event PreConstruct()
   > Branch(Condition=@n45.IsDesignTime)
   n46 = Event HandleUsingGamepadChanged!() -> (OutputDelegate, then, NewInputType:ECommonInputType Enum) {is=title K2Node_Event}
   > n47 = Sequence!(execute) -> (then_0, then_1, then_2)
   > ChangeCustomTabListVisibilityForInputType(Selection=@n46.NewInputType)
   n48 = IsCustomConfig!(execute=@n47.then_1, self, PresetIndex:int=Get PresetMappingIndex() -> (PresetMappingIndex:int=0)) -> (then, ReturnValue:bool)
   > Branch(Condition=@n48)
   > n49 = Sequence!(execute) -> (then_0, then_1)
   > ChangeCustomResetVisibilityForInputType!(execute, self, InputType:ECommonInputType Enum=@n46.NewInputType) -> (then) {is=title K2Node_CallFunction}
   n50 = Get GamepadMappingInfo_v3()
   Branch(execute=@n49.then_1, Condition=IsVisible(self=Get GamepadCustomActionSelectorPanel_v3(self=@n50)))
   > ChangeCustomClearVisibilityForInputType(InputType=@n46.NewInputType)
   Event CenterOnTab() {is=title K2Node_Event}
   n51 = Get GamepadMappingInfo_v3()
   n52 = Get HorizontalTabList_CustomModes(self=@n51)
   > n53 = Branch(Condition="Equal (Name)"(
     A=GetSelectedTabId!(self:Common Tab List Widget Base Object Reference=@n52) -> (ReturnValue:Name=None)
     B=Get DeadZoneMoveIdName(self=@n51) -> (DeadZoneMoveIdName=None)
    ))
   > CenterOnWidget(self:Fort Options Menu Setting Object Reference=Get Deadzone_Adjustment_Slider(self=Get DeadzoneAdjustment_Move(self=@n51)))
   CenterOnWidget(execute=@n53.else, self:Fort Radio Button Group Object Reference=Get FortRadioButtonGroup_Configs(self=Get GamepadInfoSelectorPanel_v3()))
   Event UpdateOptionsTab() {is=title K2Node_Event}
   > UpdateData()
   > HandleGamepadMappingInfoChange()
   PreConstructData()
  ]}
 GamepadInfoSelectorNavRight {nodes [
   GamepadInfoSelectorNavRight() {is=title K2Node_FunctionEntry}
   > CenterWidget(ToCenter=Get LT_Custom!(self:Gamepad Mapping Info V 3 Object Reference=Get GamepadMappingInfo_v3()) -> (LT_Custom:Gamepad Key Text Button Object Reference))
  ]}
 GamepadInfoSelectorReset {nodes [GamepadInfoSelectorReset() {is=title K2Node_FunctionEntry}]}
 GamepadMappingInfoChanged {nodes [GamepadMappingInfoChanged() {is=title K2Node_FunctionEntry}]}
 GetCurrentPlatformIndex {nodes [
   GetCurrentPlatformIndex() {is=title K2Node_FunctionEntry}
   > n1 = Switch on String!(execute:Exec, Selection:String=GetControllerPlatform!(self:Fort Gamepad Input Options Object Reference) -> (ReturnValue:String)) -> (Default:Exec, XboxOne:Exec, PS4:Exec, Generic:Exec, XBX:Exec, PS5:Exec, Switch:Exec, Switch2:Exec) {pinNames [XboxOne,PS4,Generic,XBX,PS5,Switch,Switch2]}
   Return Node!(execute=[@n1.Default, @n1.Generic], PlatformIndex:int=2)
   Return Node(execute=@n1.PS4, PlatformIndex=1)
   Return Node(execute=@n1.PS5, PlatformIndex=4)
   Return Node(execute=@n1.Switch, PlatformIndex=5)
   Return Node(execute=@n1.Switch2, PlatformIndex=6)
   Return Node(execute=@n1.XBX, PlatformIndex=3)
   Return Node(execute=@n1.XboxOne, PlatformIndex=0)
  ]}
 HandleGamepadInfoSelectorApply {nodes [
   HandleGamepadInfoSelectorApply() {is=title K2Node_FunctionEntry}
   > Call Tab Game Pad Config Apply!(execute:Exec, self:Tab Game Pad Config V 3 Object Reference) -> (then:Exec) {is=title K2Node_CallDelegate}
  ]}
 HandleGamepadInfoSelectorBack {nodes [
   HandleGamepadInfoSelectorBack() {is=title K2Node_FunctionEntry}
   > n1 = AreAllImportantActionsBound!(execute, self:Fort Gamepad Input Options Object Reference, bRichText:bool=false) -> (then, OutUnboundScreenLabels:Array of Texts, ReturnValue:bool)
   > n2 = Branch(Condition=@n1.ReturnValue)
   > Call Tab Game Pad Config Back!(execute:Exec, self:Tab Game Pad Config V 3 Object Reference) -> (then:Exec) {is=title K2Node_CallDelegate}
   Call Tab Game Pad Actions That Need Binds!(execute:Exec=@n2.else, self:Tab Game Pad Config V 3 Object Reference, ScreenLabels:Array of Texts=@n1) -> (then:Exec) {is=title K2Node_CallDelegate}
  ]}
 HandleGamepadInfoSelectorFocus {nodes [
   HandleGamepadInfoSelectorFocus() {is=title K2Node_FunctionEntry}
   > Unbind all Events from Gamepad Info Selector Focus!(execute:Exec, self:Gamepad Info Selector Panel V 3 Object Reference=Get GamepadInfoSelectorPanel_v3()) -> (then:Exec) {is=title K2Node_ClearDelegate}
   > NavigateToSelectedConfig()
  ]}
 HandleGamepadMappingInfoChange {nodes [
   HandleGamepadMappingInfoChange() {is=title K2Node_FunctionEntry}
   > Call Gamepad Mapping Info Changed()
   > n1 = Sequence()
   n2 = Get GamepadMappingInfo_v3()
   n3 = Get PresetMappingIndex() -> (PresetMappingIndex=0)
   n4 = Get CurrentInputActionGroup() -> (CurrentInputActionGroup=AllModes)
   Update Displayed Mapping!(
    execute=@n1.then_1
    self:Gamepad Mapping Info V 3 Object Reference=@n2
    Platform:int=Get PlatformMappingIndex() -> (PlatformMappingIndex:int=0)
    Mode:EFortInputActionGroup Enum=@n4
    Configuration:int=@n3
   ) -> (then)
   > OnConfigChanged!(execute, self:Gamepad Mapping Info V 3 Object Reference=@n2, Mode:EFortInputActionGroup Enum=@n4, Config:int=@n3) -> (then)
   > n5 = IsCustomConfig(PresetIndex=@n3)
   > n6 = Branch(Condition=@n5)
   n7 = Get HorizontalBox_AllText!(self:Gamepad Mapping Info V 3 Object Reference=Get GamepadMappingInfo_v3()) -> (HorizontalBox_AllText:Horizontal Box Object Reference)
   > SetVisibility(self=@n7, InVisibility=Collapsed)
   n8 = Get GamepadMappingInfo_v3()
   n9 = Get CommonBorder_Legend!(self:Gamepad Mapping Info V 3 Object Reference=@n8) -> (CommonBorder_Legend:Common Border Object Reference)
   n10 = Get HorizontalBox_CustomButtons!(self:Gamepad Mapping Info V 3 Object Reference=@n8) -> (HorizontalBox_CustomButtons:Horizontal Box Object Reference)
   > SetVisibility(self=[@n9, @n10], InVisibility=SelfHitTestInvisible)
   > ChangeCustomResetVisibilityForInputType(InputType=GetCurrentInputType(self=Get CommonInputSubsystem()))
   > Call Override Reset Modal!(execute:Exec, self:Tab Game Pad Config V 3 Object Reference, bOverride:bool=true) -> (then:Exec) {is=title K2Node_CallDelegate}
   > UpdateCustomButtons(self=Get GamepadMappingInfo_v3())
   n11 = Get GamepadMappingInfo_v3()
   > ShowEditModeTab(self:Gamepad Mapping Info V 3 Object Reference=@n11)
   > n12 = Switch on ESubGame(Selection=GetSubGame(self=Get FortGlobalUIContext()) -> (ReturnValue=Campaign))
   n13 = Get GamepadMappingInfo_v3()
   ShowAthenaLTMAbilitiesTab(execute=@n12.Athena, self:Gamepad Mapping Info V 3 Object Reference=@n13)
   SetVisibility(execute=@n6.else, self=@n7, InVisibility=SelfHitTestInvisible)
   > SetVisibility(self=[@n9, @n10], InVisibility=Collapsed)
   > SetVisibility(self=Get CommonBorder_Reset(self=Get GamepadInfoSelectorPanel_v3()), InVisibility=Collapsed)
   > Call Override Reset Modal()
   > HideEditModeTab(self:Gamepad Mapping Info V 3 Object Reference=@n11)
   > HideAthenaLTMAbilitiesTab(self:Gamepad Mapping Info V 3 Object Reference=@n13)
   > n14 = Switch on ESubGame(Selection=GetSubGame(self=Get FortGlobalUIContext()) -> (ReturnValue=Campaign))
   CloseCustomActionsPanel(execute=@n14.Athena, FocusLastButton=false)
   ShowDeadZoneTab(execute=@n1.then_0, self:Gamepad Mapping Info V 3 Object Reference=Get GamepadMappingInfo_v3())
  ]}
 InitializeData {nodes [
   InitializeData() {is=title K2Node_FunctionEntry}
   > n1 = Sequence!(execute) -> (then_0, then_1, then_2:Exec, then_3:Exec, then_4:Exec, then_5:Exec, then_6:Exec, then_7:Exec, then_8:Exec)
   ConstructKeyBindList(execute=@n1.then_6, self:Fort Input Options Object Reference)
   n2 = GetPlatformName!() -> (ReturnValue:String)
   n3 = Get GamepadInfoSelectorPanel_v3()
   "SetText (Text)"(
    execute=@n1.then_0
    self=Get TextHeader!(self:Options Menu Header Object Reference=Get OptionsMenuHeader_Configs!(self:Gamepad Info Selector Panel V 3 Object Reference=@n3) -> (OptionsMenuHeader_Configs:Options Menu Header Object Reference)) -> (TextHeader:Common Text Object Reference)
    InText=Select!(
     Option 0:Text="LOCTEXT(\"Controller Configurations\")"
     Option 1:Text="LOCTEXT(\"Wireless Controller Configurations\")"
     Index:bool="Equal Exactly (String)"(A=@n2, B=PS4)
    ) -> (ReturnValue:Text)
   )
   n4 = Get FortRadioButtonGroup_Configs(self=Get GamepadInfoSelectorPanel_v3())
   Clear(
    execute=@n1.then_1
    TargetArray:Array of Fort Radio Button Item Struct Structures=Get Buttons!(self:Fort Radio Button Group Object Reference=@n4) -> (Buttons:Array of Fort Radio Button Item Struct Structures)
   )
   n5 = Get FortGlobalUIContext()
   n6 = GetAllPlayerInputPresetNamesForSubGame!(self:Fort Global UIContext Object Reference=@n5, SubGame:ESubGame Enum=GetSubGame(self=@n5) -> (ReturnValue=Campaign)) -> (InputPresetNames:Array of Strings, InputPresetFriendlyNames:Array of Texts)
   > n7 = Set InputPresets(InputPresets:Array of Strings=@n6) -> (Output_Get:Array of Strings)
   > n8 = For Each Loop(Array:Array of Strings=@n7) -> (Array Element:String)
   UpdateButtons(execute=@n8.Completed, self:Fort Radio Button Group Object Reference=@n4)
   n9 = Branch(
    execute=@n8.LoopBody
    Condition=IsCustomGamepadConfig!(self:Fort Gamepad Input Options Object Reference, ConfigName:String=@n8) -> (ReturnValue:bool)
   )
   n10 = "Get (a ref)"(Array:Array of Texts=@n6.InputPresetFriendlyNames, Dimension 1=@n8.Array Index) -> (Output:"Text (by ref)")
   Add(
    execute=[@n11.then, @n9.then]
    TargetArray:Array of Fort Radio Button Item Struct Structures=Get Buttons(self=@n4)
    NewItem:"Fort Radio Button Item Struct Structure (by ref)"=Make Fort Radio Button Item Struct!(
     ButtonClass_16_#1:Icon Text Button Class Reference=Get RadioButtonClass() -> (RadioButtonClass:Icon Text Button Class Reference)
     ButtonStyle_12_#2:Common Button Style Class Reference=Get RadioButtonStyle() -> (RadioButtonStyle:Common Button Style Class Reference)
     ButtonText_8_#3:Text=@n10
    ) -> (FortRadioButtonItemStruct:Fort Radio Button Item Struct Structure) {is=title K2Node_MakeStruct}
   )
   n11 = Add(execute=@n9.else, TargetArray:Array of Texts=Get InputPresetNames_NoCustom(), NewItem:"Text (by ref)"=@n10)
   n12 = Set CurrentInputActionGroup(execute=@n1.then_2, CurrentInputActionGroup=Combat) -> (Output_Get=AllModes)
   > Set CurrentInputActionGroup!(execute, CurrentInputActionGroup=@n12, self:Gamepad Mapping Info V 3 Object Reference=Get GamepadMappingInfo_v3()) -> (then, Output_Get=AllModes)
   n13 = Get FortRadioButtonGroup_Platforms(self=Get GamepadInfoSelectorPanel_v3())
   Clear(execute=@n1.then_5, TargetArray=Get Buttons(self=@n13))
   > n14 = For Each Loop(Array:Array of Texts=Make Array!(
     "[0]":Text="LOCTEXT(\"XBox One\")"
     "[1]":Text="LOCTEXT(\"PlayStation®4\")"
     "[2]":Text="LOCTEXT(\"Generic\")"
     "[3]":Text="LOCTEXT(\"Xbox Series X|S\")"
     "[4]":Text="LOCTEXT(\"PlayStation®5\")"
     "[5]":Text="LOCTEXT(\"Switch\")"
     "[6]":Text="LOCTEXT(\"Switch2\")"
    ) -> (Array:Array of Texts)) -> (Array Element:Text)
   UpdateButtons(execute=@n14.Completed, self=@n13)
   Add(
    execute=@n14.LoopBody
    TargetArray:Array of Fort Radio Button Item Struct Structures=Get Buttons(self=@n13)
    NewItem:"Fort Radio Button Item Struct Structure (by ref)"=Make Fort Radio Button Item Struct(
     ButtonClass_16_#1=Get RadioButtonClass()
     ButtonStyle_12_#2=Get RadioButtonStyle()
     ButtonText_8_#3=@n14
    )
   )
   n15 = GetPlatformName()
   n16 = Get GamepadInfoSelectorPanel_v3()
   "SetText (Text)"(
    execute=@n1.then_4
    self=Get TextHeader(self=Get OptionsMenuHeader_Platforms!(self:Gamepad Info Selector Panel V 3 Object Reference=@n16) -> (OptionsMenuHeader_Platforms:Options Menu Header Object Reference))
    InText=Select(
     Option 0="LOCTEXT(\"Controller Platform\")"
     Option 1="LOCTEXT(\"Wireless Controller Platform\")"
     Index="Equal Exactly (String)"(A=@n15, B=PS4)
    )
   )
  ]}
 IsCustomConfig {nodes [
   n1 = IsCustomConfig!() -> (then, PresetIndex:int) {is=title K2Node_FunctionEntry}
   n2 = Get InputPresets()
   > Return Node!(
    execute
    ReturnValue:bool=IsCustomGamepadConfig(ConfigName="Get (a ref)"(Array:Array of Strings=@n2, Dimension 1=@n1) -> (Output:"String (by ref)"))
   )
  ]}
 NavigateToSelectedConfig {nodes [
   NavigateToSelectedConfig() {is=title K2Node_FunctionEntry}
   n1 = Get CreatedButtonsGroup(self=Get FortRadioButtonGroup_Configs(self=Get GamepadInfoSelectorPanel_v3()))
   > CenterWidget(ToCenter=GetButtonAtIndex!(
     self:Common Button Group Legacy Object Reference=@n1
     Index:int=GetSelectedButtonIndex!(self:Common Button Group Base Object Reference=@n1) -> (ReturnValue:int=0)
    ) -> (ReturnValue:Common Button Legacy Object Reference))
  ]}
 OverrideResetModal {nodes [OverrideResetModal!() -> (then, bOverride:bool=~) {is=title K2Node_FunctionEntry}]}
 PreConstructData {nodes [
   PreConstructData() {is=title K2Node_FunctionEntry}
   > n1 = Sequence!(execute) -> (then_0, then_1, then_2, then_3, then_4, then_5, then_6, then_7)
   Reroute Node(InputPin=@n1.then_2)
   Reroute Node(InputPin=@n1.then_3)
   n2 = Get FortRadioButtonGroup_Configs(self=Get GamepadInfoSelectorPanel_v3())
   Clear(execute=@n1.then_1, TargetArray=Get Buttons(self=@n2))
   > n3 = For Each Loop(Array=Make Array!(
     "[0]"="LOCTEXT(\"Old School\")"
     "[1]"="LOCTEXT(\"Quick Builder\")"
     "[2]"="LOCTEXT(\"Combat Pro\")"
     "[3]"="LOCTEXT(\"Builder Pro\")"
     "[4]"="LOCTEXT(\"Custom\")"
    ) -> (Array))
   UpdateButtons(execute=@n3.Completed, self=@n2)
   Add(
    execute=@n3.LoopBody
    TargetArray=Get Buttons(self=@n2)
    NewItem=Make Fort Radio Button Item Struct(
     ButtonClass_16_#1=Get RadioButtonClass()
     ButtonStyle_12_#2=Get RadioButtonStyle()
     ButtonText_8_#3=@n3
    )
   )
   n4 = Get GamepadInfoSelectorPanel_v3()
   "SetText (Text)"(execute=@n1.then_0, self=Get TextHeader(self=Get OptionsMenuHeader_Configs(self=@n4)), InText="LOCTEXT(\"Configurations\")")
   n5 = Get FortRadioButtonGroup_Platforms(self=Get GamepadInfoSelectorPanel_v3())
   Clear(execute=@n1.then_5, TargetArray=Get Buttons(self=@n5))
   > n6 = For Each Loop(Array=Make Array!("[0]"="LOCTEXT(\"XBox One\")", "[1]"="LOCTEXT(\"PlayStation®4\")") -> (Array))
   UpdateButtons(execute=@n6.Completed, self=@n5)
   Add(
    execute=@n6.LoopBody
    TargetArray=Get Buttons(self=@n5)
    NewItem=Make Fort Radio Button Item Struct(
     ButtonClass_16_#1=Get RadioButtonClass()
     ButtonStyle_12_#2=Get RadioButtonStyle()
     ButtonText_8_#3=@n6
    )
   )
   n7 = Get GamepadInfoSelectorPanel_v3()
   "SetText (Text)"(execute=@n1.then_4, self=Get TextHeader(self=Get OptionsMenuHeader_Platforms(self=@n7)), InText="LOCTEXT(\"Controller Platform\")")
  ]}
 SelectSameItemInActionsList {nodes [
   SelectSameItemInActionsList() {is=title K2Node_FunctionEntry}
   > Set IndexToSelect_Local(IndexToSelect_Local:int=0) -> (Output_Get:int=0)
   n1 = Get GamepadMappingInfo_v3()
   > n2 = For Each Loop with Break!(
    Exec:Exec
    Array:Array of Fort Options Menu Input Data Object References=Get CurrentInputData!(self:Gamepad Custom Action Selector Panel V 3 Object Reference=Get GamepadCustomActionSelectorPanel_v3(self=@n1)) -> (CurrentInputData:Array of Fort Options Menu Input Data Object References)
    Break:Exec=@n6.then
   ) -> (LoopBody:Exec, Array Element:Fort Options Menu Input Data Object Reference, Array Index:int, Completed:Exec)
   n3 = Get FortGlobalUIContext()
   n4 = String To Name(InString=GetCustomGamepadInputPresetName!(self:Fort Global UIContext Object Reference=@n3) -> (ReturnValue:String))
   n5 = Get ActionName!(self:Fort Options Menu Input Data Object Reference=@n2) -> (ActionName:Name=None)
   > Branch(Condition="Equal (Key)"(
     A=Get CustomKeySelection()
     B=GetKeyForAction!(self:Fort Global UIContext Object Reference=@n3, Action:Name=@n5, bForceGamepadKey:bool=true, PresetNameOverride:Name=@n4) -> (ReturnValue:Key Structure)
    ))
   > n6 = Set IndexToSelect_Local(IndexToSelect_Local=@n2.Array Index) -> (Output_Get=0)
   n7 = Get CommonListView_GamepadActions!(self:Gamepad Custom Action Selector Panel V 3 Object Reference=Get GamepadCustomActionSelectorPanel_v3(self=Get GamepadMappingInfo_v3())) -> (CommonListView_GamepadActions:Common List View Object Reference)
   CenterWidget(execute=@n2.Completed, ToCenter=@n7)
   n8 = Get IndexToSelect_Local() -> (IndexToSelect_Local:int=0)
   > SetSelectedIndex!(execute, self:List View Object Reference=@n7, Index:int=@n8) -> (then)
   > NavigateToIndex!(execute, self:List View Object Reference=@n7, Index:int=@n8) -> (then)
  ]}
 SetupDisplayObject {nodes [
   n1 = SetupDisplayObject!() -> (then, ActionName:Name, Key:Key Structure) {is=title K2Node_FunctionEntry}
   n2 = GetCustomGamepadInputPresetName(self=Get FortGlobalUIContext())
   n3 = Get DisplayObject() -> (DisplayObject:Gamepad Key Text Button Object Reference)
   > SetText(
    self:Fort Rich Text Block Legacy Object Reference=Get FortRichTextBlock_Desc!(self:Gamepad Key Text Button Object Reference=@n3) -> (FortRichTextBlock_Desc:Fort Rich Text Block Legacy Object Reference)
    InText:Text=GetScreenLabel!(
     self:Fort Input Data Object Reference=GetFortInputDataForBlueprint!() -> (ReturnValue:Fort Input Data Object Reference)
     PresetName:Name=String To Name(InString=@n2)
     InputActionGroup:EFortInputActionGroup Enum=Get CurrentInputActionGroup() -> (CurrentInputActionGroup=AllModes)
     ActionName:Name=@n1
     bRichText:bool=true
    ) -> (ReturnValue:Text)
   )
   > SetBoundKey!(
    execute
    self:Fort Keybind Widget Object Reference=Get BoundKey!(self:Gamepad Key Text Button Object Reference=@n3) -> (BoundKey:Keybind Widget Object Reference)
    NewBoundAction:Key Structure=@n1.Key
   ) -> (then)
   > Return Node!(execute, DisplayObject:Gamepad Key Text Button Object Reference=Get DisplayObject())
  ]}
 TabGamePadActionsThatNeedBinds {nodes [TabGamePadActionsThatNeedBinds!() -> (then, ScreenLabels:Array of Texts) {is=title K2Node_FunctionEntry}]}
 TabGamePadConfigApply {nodes [TabGamePadConfigApply() {is=title K2Node_FunctionEntry}]}
 TabGamePadConfigBack {nodes [TabGamePadConfigBack() {is=title K2Node_FunctionEntry}]}
 UpdateData {nodes [
   UpdateData() {is=title K2Node_FunctionEntry}
   > n1 = Sequence!(execute) -> (then_0, then_1, then_2)
   n2 = Get FortGlobalUIContext()
   > n3 = Set PresetMappingIndex(PresetMappingIndex=Find Item!(
     TargetArray=Get InputPresets()
     ItemToFind:"String (by ref)"=GetCurrentInputPresetName!(self:Fort Global UIContext Object Reference=@n2) -> (ReturnValue:String)
    ) -> (ReturnValue)) -> (Output_Get=0)
   n4 = Get GamepadInfoSelectorPanel_v3()
   > SelectButtonAtIndex!(
    execute
    self:Common Button Group Base Object Reference=Get CreatedButtonsGroup(self=Get FortRadioButtonGroup_Configs(self=@n4))
    ButtonIndex:int=@n3
    bAllowSound:bool=true
   ) -> (then)
   n5 = Set PlatformMappingIndex(execute=@n1.then_2, PlatformMappingIndex=GetCurrentPlatformIndex!() -> (PlatformIndex:int=0) {is=title K2Node_CallFunction}) -> (Output_Get=0)
   n6 = Get GamepadInfoSelectorPanel_v3()
   > SelectButtonAtIndex(
    self=Get CreatedButtonsGroup(self=Get FortRadioButtonGroup_Platforms(self=@n6))
    ButtonIndex=Select!(Option 0:int=0, Option 1:int=1, Option 2:int=2, Option 3:int=3, Option 4:int=4, Option 5:int=5, Option 6:int=6, Index:int=@n5) -> (ReturnValue:int)
    bAllowSound=true
   )
  ]}
]
variables [
 CurrentInputActionGroup {type=EFortInputActionGroup, default=AllModes}
 CustomActionSelection {type=string, default=None}
 CustomActionSelectionIndex {type=integer}
 CustomInputActionGroupSelectionIndex {type=integer}
 CustomInputTemplatePresetIndex {type=integer}
 CustomKeySelection {type=Key, default=None}
 CustomResetToDefaultConfig {type=integer}
 DisplayObject {type=GamepadKeyTextButton_C, default=null}
 GamepadInfoSelectorReset {}
 GamepadMappingInfoChanged {}
 InputDataForActionsPanel {type="array<FortOptionsMenuInputData>", default []}
 InputPresetNames_NoCustom {type="array<string>", default []}
 InputPresets {type="array<string>", default []}
 Key_None {type=Key, default=None}
 OverrideResetModal {}
 PlatformMappingIndex {type=integer}
 PresetMappingIndex {type=integer}
 RadioButtonClass {type=Class, default=/FortUILibrary/Foundation/Buttons/IconTextButton.IconTextButton_C}
 RadioButtonStyle {type=Class, default=/FortUILibrary/Foundation/Buttons/ButtonStyles/ButtonStyle-GamepadBindings.ButtonStyle-GamepadBindings_C}
 TabGamePadActionsThatNeedBinds {}
 TabGamePadConfigApply {}
 TabGamePadConfigBack {}
]
widgets {
 info {inheritedWidgetCount=2, parentClass=/Script/FortniteUI.FortGamepadInputOptions, rootWidgetClass=&ScaleBox, widgetCount=6}
 widgets [
  ScaleBox_0 {widgetClassPath=&ScaleBox, widget=ScaleBox_0, properties {stretch=Fill}}
  HorizontalBox_0 {widgetClassPath=&HorizontalBox, parent=ScaleBox_0, widget=HorizontalBox_0, slot {class=ScaleBoxSlot}}
  GamepadInfoSelectorPanel_v3 {
   widgetClassPath=&/GamepadInfoSelectorPanel_v3
   bIsVariable=true
   parent=HorizontalBox_0
   widget=GamepadInfoSelectorPanel_v3
   properties {navigation=&self:WidgetTree.GamepadInfoSelectorPanel_v3.WidgetNavigation_0}
   slot {class=HorizontalBoxSlot, properties {verticalAlignment=VAlign_Center}}
  }
  GamepadMappingInfo_v3 {
   widgetClassPath=&/GamepadMappingInfo_v3
   bIsVariable=true
   parent=HorizontalBox_0
   widget=GamepadMappingInfo_v3
   properties {customKeyButtons []}
   slot {class=HorizontalBoxSlot, properties {size {sizeRule=Fill, value=1}}}
  }
  InputCommonListView {widgetClassPath=&CommonListView, bInherited=true, widget=null}
  TooltipDisplay {widgetClassPath=&CommonTextBlock, bInherited=true, widget=null}
 ]
}
